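# Tab container for the Quest screen side navigation, a child of the UIKit tab container base with no widget tree of its own.
# On collapsed-state change it sets expansion from whether an inner tab is selected, then updates the arrow instantly or by playing its hide animation forward or in reverse.
# It then passes the collapsed state to the tab button and inner tab children through the collapsible interface.
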WidgetBlueprint /QuestUI/QuestScreen/V3/WBP_QuestSideNav_TabContainer
graphs [
 EventGraph {nodes [
   n1 = Event PreConstruct()
   > Parent: PreConstruct(IsDesignTime=@n1.IsDesignTime)
   Event Construct()
   Event Tick()
  ]}
 HandleCollapsedStateChanged {nodes [
   HandleCollapsedStateChanged!() -> (then, IsCollapsed:bool, SkipAnimation:bool) {is=title K2Node_FunctionEntry}
   > n1 = Sequence!(execute) -> (then_0, then_1, then_2:Exec, then_3:Exec)
   n2 = GetInnerTabGroupWidget!(self:WBP UIKit Tab Container No Layout Object Reference) -> (InnerTabGroupWidget:WBP UIKit Tab Group Compact Base Object Reference)
   n3 = For Each Loop(
    Exec=@n1.then_3
    Array:Array of Widget Object References=GetAllChildren(self=Get TabsContainer!(self:WBP UIKit Tab Group Compact Base Object Reference=@n2) -> (TabsContainer:Grid Panel Object Reference))
   ) -> (Array Element:Widget Object Reference)
   > n4 = Cast To BPI_UIKit_Collapsible!(execute, Object=@n3) -> (then, CastFailed, AsBPI UIKit Collapsible:BPI UIKit Collapsible Interface) {is=title K2Node_DynamicCast}
   > SetIsCollapsed!(
    execute
    self:BPI UIKit Collapsible Interface=@n4
    IsCollapsed:bool=Get IsCollapsed() -> (IsCollapsed:bool)
    SkipAnimation:bool=Get SkipAnimation() -> (SkipAnimation:bool)
   ) -> (then)
   n5 = Cast To BPI_UIKit_Collapsible(
    execute=@n1.then_2
    Object=GetTabButton!(self:WBP UIKit Tab Container No Layout Object Reference) -> (TabButton:Common Button Base Object Reference)
   )
   > SetIsCollapsed(self=@n5, IsCollapsed=Get IsCollapsed(), SkipAnimation=Get SkipAnimation())
   n6 = Branch(execute=@n1.then_1, Condition=Get SkipAnimation())
   > SetArrowVisibilty!(execute, self:WBP UIKit Tab Container No Layout Object Reference, IsVisible:bool=NOT Boolean(A=Get IsCollapsed())) -> (then)
   n7 = Override_GetHideArrowAnimation!(self:WBP UIKit Tab Container No Layout Object Reference) -> (Animation:Widget Animation Object Reference)
   Is Valid(exec=@n6.else, InputObject=@n7)
   > n8 = Branch(Condition=Get IsCollapsed())
   > PlayAnimationForward(InAnimation=@n7)
   PlayAnimationReverse(execute=@n8.else, InAnimation=@n7)
   n9 = "Not Equal (Name)"(A=GetSelectedTabId!(self:Common Tab List Widget Base Object Reference=GetInnerTabGroupWidget()) -> (ReturnValue:Name=None)) -> (ReturnValue=~)
   SetIsExpanded!(execute=@n1.then_0, self:WBP UIKit Tab Container No Layout Object Reference, InIsExpanded:bool=@n9, SkipAnim:bool=Get SkipAnimation()) -> (then)
   Select(Option 0:bool=@n9, Option 1:bool=false, Index=Get IsCollapsed()) -> (ReturnValue:bool)
  ]}
]
variables []
widgets {
 info {
  inheritedWidgetCount=0
  parentClass=/FortUILibrary/Systems/Tabs/WBP_UIKit_TabContainer_Base.WBP_UIKit_TabContainer_Base_C
  rootWidgetClass=null
  widgetCount=0
 }
 widgets []
}
