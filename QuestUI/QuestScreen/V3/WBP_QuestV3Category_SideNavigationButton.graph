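# Quest screen side-navigation button for a quest category, built on the UIKit generic button and able to collapse.
# SetViewModel sets the label and async-loads the category icon, with a fallback material and a stale-load check before setting icon material parameters.
# SetIsCollapsed plays or mid-way reverses Anim_Collapse, which fades and shifts content and animates a Test value whose setter eases width and slot padding.
# GetRemainingAnimTime reports that animation's timing, adjusted for playback speed.
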WidgetBlueprint /QuestUI/QuestScreen/V3/WBP_QuestV3Category_SideNavigationButton
animations [Anim_Collapse {
  binding_count=2
  length_seconds=1
  object=&self:Anim_Collapse
  sequence {bindings [
    Main_Stack {
     id=#1
     possessed_class=StackBox
     tracks [
      {
       type=MovieSceneFloatTrack
       display_name=Render Opacity
       sections [{
         type=MovieSceneFloatSection
         end_frame=20
         start_frame=0
         channels [None {type=MovieSceneScriptingFloatChannel, keys [[0,1],[4,0.896],[8,0],[20,0]]}]
        }]
      }
      {
       type=MovieScene2DTransformTrack
       sections [{
         type=MovieScene2DTransformSection
         end_frame=20
         start_frame=0
         channels [
          Translation.X {type=MovieSceneScriptingFloatChannel, keys [[0,0],[4,-8]]}
          Translation.Y {type=MovieSceneScriptingFloatChannel, const [0,1]}
          Angle {type=MovieSceneScriptingFloatChannel, const [0,1]}
          Scale.X {type=MovieSceneScriptingFloatChannel}
          Scale.Y {type=MovieSceneScriptingFloatChannel}
          Shear.X {type=MovieSceneScriptingFloatChannel, const [0,1]}
          Shear.Y {type=MovieSceneScriptingFloatChannel, const [0,1]}
         ]
        }]
      }
     ]
    }
    WBP_QuestCategory_SideNavigationButton_C_7 {
     id=#2
     possessed_class=WBP_QuestCategory_SideNavigationButton_C
     tracks [{
       type=MovieSceneDoubleTrack
       display_name=Test
       sections [{type=MovieSceneDoubleSection, end_frame=20, start_frame=0, channels [None {type=MovieSceneScriptingDoubleChannel, keys [[0,0],[20,1]]}]}]
      }]
    }
   ]}
 }]
graphs [
 EventGraph {nodes [
   n1 = LoadDisplayAsset!() -> (OutputDelegate, then, DisplayAsset:Soft Object Reference) {is=title K2Node_CustomEvent}
   > SetIcon!(execute, self:WBP UIKit Button Generic Base Object Reference, Icon:Object Reference) -> (then)
   > n2 = Set RequiredDisplayAsset(RequiredDisplayAsset:Soft Object Reference=Select(
     Option 0:Soft Object Reference=/QuestUI/QuestScreen/Materials/MI_UI_QuestPin_Tile
     Option 1:Soft Object Reference=@n1.DisplayAsset
     Index=IsValidSoftObjectReference(SoftObjectReference=@n1.DisplayAsset)
    ) -> (ReturnValue:Soft Object Reference)) -> (Output_Get:Soft Object Reference)
   > n3 = Async Load Asset(Asset=@n2)
   n4 = Get RequiredDisplayAsset() -> (RequiredDisplayAsset:Soft Object Reference)
   n5 = To Soft Object Reference(SoftObjectPath=GetSoftObjectPath(Object=@n3))
   Branch(
    execute=@n3.Completed
    Condition="Equal Exactly (String)"(
     A="To String (SoftObjectReference)"!(SoftObjectReference:"Soft Object Reference (by ref)"=@n5) -> (ReturnValue:String)
     B="To String (SoftObjectReference)"(SoftObjectReference=@n4)
    )
   )
   > SetIcon(Icon=@n3)
   > n6 = Get!(execute:Exec, self:WBP UIKit Block Image Object Reference=Get Block_Icon() -> (Block_Icon:WBP UIKit Block Image Object Reference)) -> (then:Exec, else:Exec, ImageMID:Material Instance Dynamic Object Reference)
   > SetScalarParameterValue(self=@n6, ParameterName=IsSideNav, Value=1)
   > SetScalarParameterValue(self=@n6, ParameterName=CollectionsQuestV3, Value=1)
   > SetScalarParameterValue(self=@n6, ParameterName=SideNav_RadiusInPixels, Value=1)
   > SetScalarParameterValue(self=@n6, ParameterName=SideNav_CornerRadius, Value=12)
   n7 = Event PreConstruct()
   > Parent: PreConstruct(IsDesignTime=@n7.IsDesignTime)
   > SetWidthOverride(self=Get Main_StackSize() -> (Main_StackSize:Size Box Object Reference), InWidthOverride=304)
   Event Construct()
   > Parent: Construct()
   > n8 = Cast To WBP_UIKit_BlockBackground_Tab_Enhanced!(execute, Object=Get Block_Background() -> (Block_Background:WBP UIKit Block Background Base Object Reference)) -> (then, CastFailed, AsWBP UIKit Block Background Tab Enhanced:WBP UIKit Block Background Tab Enhanced Object Reference) {is=title K2Node_DynamicCast}
   > SetPadding!(
    execute
    self:Grid Slot Object Reference=SlotAsGridSlot(Widget=Get Backing!(self:WBP UIKit Block Background Tab Enhanced Object Reference=@n8) -> (Backing:Image Object Reference))
    InPadding:Margin Structure=Get BackgroundPadding() -> (BackgroundPadding:Margin Structure)
   ) -> (then)
   > SetCornerRadiuses!(
    execute
    self:WBP UIKit Block Background Tab Enhanced Object Reference=@n8
    CornerRadius:float=Get BackgroundCornerRadius() -> (BackgroundCornerRadius:float)
    RadiusInPixels:bool=true
   ) -> (then)
   > SetPadding(
    self=SlotAsGridSlot(Widget=Get FocusOutline!(self:WBP UIKit Block Background Tab Enhanced Object Reference=@n8) -> (FocusOutline:Image Object Reference))
    InPadding=Get FocusOutlinePadding() -> (FocusOutlinePadding:Margin Structure)
   )
   Event OnInitialized()
   > SetCompactUIStyle() {is=title K2Node_CallFunction}
   n9 = Event SetIsCollapsed!() -> (OutputDelegate, then, IsCollapsed:bool, SkipAnimation:bool) {is=title K2Node_Event}
   > n10 = Branch(Condition=@n9.SkipAnimation)
   n11 = Get Anim_Collapse() -> (Anim_Collapse:Widget Animation Object Reference)
   n12 = Select(Option 0:EUMGSequencePlayMode Enum=Reverse, Option 1:EUMGSequencePlayMode Enum, Index=@n9.IsCollapsed) -> (ReturnValue:EUMGSequencePlayMode Enum)
   > PlayAnimation(InAnimation=@n11, StartAtTime=GetEndTime(self=@n11), PlayMode=@n12)
   n13 = Branch(execute=@n10.else, Condition=IsAnimationPlaying(InAnimation=Get Anim_Collapse()))
   n14 = Get Anim_Collapse()
   > n15 = PauseAnimation!(execute, self:User Widget Object Reference, InAnimation:Widget Animation Object Reference=@n14) -> (then, ReturnValue:float)
   n16 = GetEndTime(self=@n14)
   n17 = Get Playback Speed() -> (Playback Speed:float)
   > PlayAnimation(
    InAnimation=@n14
    StartAtTime=SelectFloat(A=@n15, B=float - float(A=@n16, B=@n15), bPickA=@n9.IsCollapsed)
    PlayMode=@n12
    PlaybackSpeed=@n17
   )
   PlayAnimation(execute=@n13.else, InAnimation=Get Anim_Collapse(), PlayMode=@n12, PlaybackSpeed=@n17)
  ]}
 GetRemainingAnimTime {nodes [
   GetRemainingAnimTime() {is=title K2Node_FunctionEntry}
   n1 = Get Anim_Collapse()
   > n2 = Branch(Condition=IsAnimationPlaying(InAnimation=@n1))
   > n3 = IsAnimationPlayingForward!(execute, self:User Widget Object Reference, InAnimation:Widget Animation Object Reference=@n1) -> (then, ReturnValue:bool)
   > n4 = Branch(Condition=@n3)
   n5 = float / float(A=GetEndTime(self=Get Anim_Collapse()), B=Get Playback Speed())
   n6 = Get Anim_Collapse()
   n7 = GetEndTime(self=@n6)
   n8 = GetAnimationCurrentTime!(self:User Widget Object Reference, InAnimation:Widget Animation Object Reference=@n6) -> (ReturnValue:float)
   > Return Node!(
    execute
    FunctionIsImplemented:bool=true
    RemainingTime:double=float / float(A=float - float(A=@n7, B=@n8), B=Get Playback Speed())
    TotalTime:double=@n5
   )
   n9 = Get Anim_Collapse()
   n10 = GetAnimationCurrentTime(InAnimation=@n9)
   n11 = GetStartTime!(self:Widget Animation Object Reference=@n9) -> (ReturnValue:float)
   Return Node(
    execute=@n4.else
    FunctionIsImplemented=true
    RemainingTime=float / float(A=float - float(A=@n10, B=@n11), B=Get Playback Speed())
    TotalTime=@n5
   )
   Return Node(execute=@n2.else, FunctionIsImplemented=true, RemainingTime=0, TotalTime=0)
  ]}
 IsContentEnabled {nodes [
   n1 = IsContentEnabled!() -> (then, IsEnabled:bool=~) {is=title K2Node_FunctionEntry}
   > SetRenderOpacity(InOpacity=Select(Option 0:float=0.5, Option 1:float=1, Index=@n1) -> (ReturnValue:float))
  ]}
 Sequencer Events {links [], nodes []}
 SetCompactUIStyle {nodes [
   SetCompactUIStyle() {is=title K2Node_FunctionEntry}
   > Branch(Condition=IsUsingCompactUI())
  ]}
 SetModalTabName {nodes [
   n1 = SetModalTabName!() -> (then, Text:Text) {is=title K2Node_FunctionEntry}
   > SetText!(execute, self:WBP UIKit Block Text Object Reference=Get Block_Text() -> (Block_Text:WBP UIKit Block Text Object Reference), Text:Text=@n1) -> (then)
  ]}
 SetTest {nodes [
   n1 = SetTest!() -> (then, Value:double=~) {is=title K2Node_FunctionEntry}
   n2 = Get!() -> (IconSize_X:double, IconSize_Y:double)
   n3 = Get!() -> (CachedContentSize_X:double, CachedContentSize_Y:double)
   > SetWidthOverride(
    self=Get Main_StackSize()
    InWidthOverride=Ease!(Function:EEasingFunc Enum=Linear, Alpha:double=@n1, A:double=float - float!(A=@n3, B=@n2, C:double=36) -> (ReturnValue), B:double) -> (Result:double=~) {is=title K2Node_EaseFunction}
   )
   n4 = SlotAsGridSlot(Widget=Self-Reference())
   > Is Valid(InputObject=@n4)
   n5 = Ease(Function=Linear, Alpha=@n1, A:float, B:float=-4) -> (Result:float=~)
   n6 = Get!(self:Grid Slot Object Reference=@n4) -> (Padding_Left:float=~, Padding_Top:float=~, Padding_Right:float=~, Padding_Bottom:float=~)
   > SetPadding!(
    execute
    self=@n4
    InPadding_Left:float=@n5
    InPadding_Top:float=@n6.Padding_Top
    InPadding_Right:float=@n5
    InPadding_Bottom:float=@n6.Padding_Bottom
   ) -> (then)
  ]}
 SetViewModel {nodes [
   n1 = SetViewModel!() -> (then, QuestUICategoryInterfaceVM:Quest UI Category Interface View Model Object Reference) {is=title K2Node_FunctionEntry}
   > n2 = Set QuestUICategoryInterfaceVM(QuestUICategoryInterfaceVM:Quest UI Category Interface View Model Object Reference=@n1) -> (Output_Get:Quest UI Category Interface View Model Object Reference)
   > SetText(self=Get Block_Text(), Text=GetModalTabName!(self:Quest UI Category Interface View Model Object Reference=@n2) -> (ReturnValue:Text))
   > LoadDisplayAsset!(
    execute
    self
    DisplayAsset:Soft Object Reference=GetCategoryIcon!(self:Quest UI Category Interface View Model Object Reference=Get QuestUICategoryInterfaceVM() -> (QuestUICategoryInterfaceVM:Quest UI Category Interface View Model Object Reference)) -> (ReturnValue:Soft Object Reference)
   ) -> (then) {is=title K2Node_CallFunction}
  ]}
]
variables [
 BackgroundCornerRadius {type=number, default=12}
 BackgroundPadding {type=Margin, default=<ltrb 0 0 0 0>}
 CachedContentSize {type=Vector2D, default=<344 48>}
 FocusOutlinePadding {type=Margin, default=<ltrb -15 -15 -15 -15>}
 IsTrackedCategory {type=boolean}
 Playback Speed {type=number, default=3}
 RequiredDisplayAsset {type=Object, default=null}
 test {type=number}
]
widgets {
 info {
  inheritedWidgetCount=0
  parentClass=/FortUILibrary/Systems/ModularButton/WBP_UIKit_ButtonGeneric_Base.WBP_UIKit_ButtonGeneric_Base_C
  rootWidgetClass=null
  widgetCount=0
 }
 widgets []
}
