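# Header widget for the quests modal, showing product icon and name, category name, an arrow, a debug category ID and a lazy background image.
# Setter functions fill the texts and icon; SetCompactStyle hides the background, product info and arrow.
# SetBackground async-loads a material, sets its IsHeader parameter and broadcasts its BackgroundColor through OnBackgroundChanged.
# Construct sets the debug ID to a literal and shows a spacer only for right-to-left cultures.
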WidgetBlueprint /QuestUI/QuestScreen/Modal/WBP_QuestsModal_Header
graphs [
 EventGraph {nodes [
   n1 = LoadBackground!() -> (OutputDelegate, then, Background:Soft Object Reference) {is=title K2Node_CustomEvent}
   > n2 = Async Load Asset(Asset=@n1.Background)
   n3 = Cast To MaterialInterface!(execute=@n2.Completed, Object=@n2) -> (then, CastFailed, AsMaterial Interface:Material Interface Object Reference) {is=title K2Node_DynamicCast}
   > SetBrushFromMaterial(self=Get Lazy_Background() -> (Lazy_Background:Fort Lazy Image Object Reference), Material=@n3)
   > n4 = GetDynamicMaterial(self=Get Lazy_Background())
   > SetScalarParameterValue(self=@n4, ParameterName=IsHeader, Value=1)
   > n5 = Get Vector Parameter Value!(execute, self:Material Instance Dynamic Object Reference=@n4, ParameterName:Name=BackgroundColor) -> (then, ReturnValue:Linear Color Structure) {is=title K2Node_CallFunction}
   > Call On Background Changed!(execute:Exec, self:WBP Quests Modal Header Object Reference, Color:Linear Color Structure=@n5) -> (then:Exec) {is=title K2Node_CallDelegate}
   Event Construct()
   > SetDebugCategoryID!(execute, self, DebugCategoryID:String=MakeLiteralString()) -> (then)
   n6 = GetCurrentCulture!() -> (ReturnValue:String)
   > SetVisibility(
    self=Get Spacer_Product() -> (Spacer_Product:Spacer Object Reference)
    InVisibility=Select(Option 0=Collapsed, Option 1=SelfHitTestInvisible, Index=IsCultureRightToLeft!(Culture:String=@n6) -> (ReturnValue:bool))
   )
  ]}
 OnBackgroundChanged {nodes [OnBackgroundChanged!() -> (then, Color:Linear Color Structure) {is=title K2Node_FunctionEntry}]}
 SetBackground {nodes [
   n1 = SetBackground!() -> (then, Background:Soft Object Reference) {is=title K2Node_FunctionEntry}
   > LoadBackground!(execute, self, Background:Soft Object Reference=@n1) -> (then) {is=title K2Node_CallFunction}
  ]}
 SetCategoryName {nodes [
   n1 = SetCategoryName!() -> (then, Category:Text) {is=title K2Node_FunctionEntry}
   > "SetText (Text)"(self=Get QuestCategoryName() -> (QuestCategoryName:Common Text Object Reference), InText=@n1)
  ]}
 SetCompactStyle {nodes [
   n1 = SetCompactStyle!() -> (then, IsCompact:bool=~) {is=title K2Node_FunctionEntry}
   > SetVisibility(
    self=[
     Get Lazy_Background()
     Get ProductIcon() -> (ProductIcon:Image Object Reference)
     Get ProductName() -> (ProductName:Common Text Object Reference)
    ]
    InVisibility=Select(Option 0=SelfHitTestInvisible, Option 1=Collapsed, Index=@n1)
   )
   > SetVisibility(self=Get Arrow() -> (Arrow:Image Object Reference), InVisibility=Select(Option 0=SelfHitTestInvisible, Option 1=Hidden, Index=@n1))
  ]}
 SetDebugCategoryID {nodes [
   n1 = SetDebugCategoryID!() -> (then, DebugCategoryID:String) {is=title K2Node_FunctionEntry}
   > "SetText (Text)"(self=Get DebugCategoryID() -> (DebugCategoryID:Common Text Object Reference), InText="To Text (String)"(InString=@n1))
  ]}
 SetProductIcon {nodes [
   n1 = SetProductIcon!() -> (then, Icon:Material Instance Soft Object Reference) {is=title K2Node_FunctionEntry}
   > SetBrushFromSoftMaterial!(execute, self:Image Object Reference=Get ProductIcon(), SoftMaterial:Material Interface Soft Object Reference=@n1) -> (then)
  ]}
 SetProductName {nodes [
   n1 = SetProductName!() -> (then, Product:Text) {is=title K2Node_FunctionEntry}
   > "SetText (Text)"(self=Get ProductName(), InText=@n1)
  ]}
]
variables [
 Background {type=Object, default=null}
 OnBackgroundChanged {}
 Scrolled {type=boolean}
]
widgets {
 info {inheritedWidgetCount=0, parentClass=&CommonUserWidget, rootWidgetClass=&SizeBox, widgetCount=10}
 widgets [
  SizeBox_1 {widgetClassPath=&SizeBox, widget=SizeBox_1, properties {bOverride_WidthOverride=true, widthOverride=872}}
  Root {
   widgetClassPath=&GridPanel
   parent=SizeBox_1
   widget=Root
   properties {columnFill [0,1], rowFill [0,0,1]}
   slot {class=SizeBoxSlot}
  }
  ProductName {
   widgetClassPath=&CommonTextBlock
   bIsVariable=true
   parent=Root
   widget=ProductName
   properties {
    text=Battle Royale Quests
    textTransformPolicy=ToUpper
    colorAndOpacity {colorUseRule=UseColor_Specified, specifiedColor=<rgba 0.5776 0.5776 0.5776 1>}
    font {
     bForceMonospaced=false
     bMaterialIsStencil=false
     fontMaterial=null
     fontObject=&HeadingNow
     letterSpacing=0
     monospacedWidth=1
     size=15.75
     skewAmount=0
     typefaceFontName=76 Bold
     outlineSettings {
      bApplyOutlineToDropShadows=false
      bMiteredCorners=false
      bSeparateFillAlpha=false
      outlineMaterial=null
      outlineSize=0
      outlineColor=<rgba 0 0 0 1>
     }
    }
    shadowOffset=<0 0>
   }
   slot {class=GridSlot, properties {column=1, layer=1, verticalAlignment=VAlign_Center}}
  }
  HorizontalBox_1 {
   widgetClassPath=&HorizontalBox
   parent=Root
   widget=HorizontalBox_1
   slot {class=GridSlot, properties {column=1, row=1, verticalAlignment=VAlign_Center, padding=<ltrb 0 0 18 0>}}
  }
  QuestCategoryName {
   widgetClassPath=&CommonTextBlock
   bIsVariable=true
   parent=HorizontalBox_1
   widget=QuestCategoryName
   properties {
    scrollStyle=&TextScrollStyle-Base_C
    text=SUPERNOVA ACADEMY
    textTransformPolicy=ToUpper
    font {
     bForceMonospaced=false
     bMaterialIsStencil=false
     fontMaterial=null
     fontObject=&HeadingNow
     letterSpacing=0
     monospacedWidth=1
     size=18
     skewAmount=0
     typefaceFontName=86 Bold
     outlineSettings {
      bApplyOutlineToDropShadows=false
      bMiteredCorners=false
      bSeparateFillAlpha=false
      outlineMaterial=null
      outlineSize=0
      outlineColor=<rgba 0 0 0 1>
     }
    }
    shadowOffset=<0 0>
   }
   slot {class=HorizontalBoxSlot, properties {size {sizeRule=Fill, value=1}}}
  }
  DebugCategoryID {
   widgetClassPath=&CommonTextBlock
   bIsVariable=true
   parent=HorizontalBox_1
   widget=DebugCategoryID
   properties {
    bAutoCollapseWithEmptyText=true
    text=Debug Category ID
    colorAndOpacity {colorUseRule=UseColor_Specified, specifiedColor=<rgba 0.5776 0.5776 0.5776 1>}
    font {
     bForceMonospaced=false
     bMaterialIsStencil=false
     fontMaterial=null
     fontObject=&HeadingNow
     letterSpacing=0
     monospacedWidth=1
     size=10.5
     skewAmount=0
     typefaceFontName=74 Regular
     outlineSettings {
      bApplyOutlineToDropShadows=false
      bMiteredCorners=false
      bSeparateFillAlpha=false
      outlineMaterial=null
      outlineSize=0
      outlineColor=<rgba 0 0 0 1>
     }
    }
    shadowOffset=<0 0>
   }
   slot {class=HorizontalBoxSlot, properties {horizontalAlignment=HAlign_Left, verticalAlignment=VAlign_Bottom, padding=<ltrb 10 0 0 0>}}
  }
  Arrow {
   widgetClassPath=&Image
   bIsVariable=true
   parent=Root
   widget=Arrow
   properties {
    brush {
     bIsDynamicallyLoaded=false
     drawAs=Image
     imageType=FullColor
     mirroring=NoMirror
     resourceName=None
     resourceObject=&/T_UI_Arrow
     tiling=NoTile
     imageSize=<32 32>
     margin=<ltrb 0 0 0 0>
     outlineSettings {
      bUseBrushTransparency=false
      roundingType=HalfHeightRadius
      width=0
      color {colorUseRule=UseColor_Specified, specifiedColor=<rgba 0 0 0 0>}
      cornerRadii=<0 0 0 0>
     }
     tintColor {colorUseRule=UseColor_Specified, specifiedColor=<rgba 1 1 1 1>}
     uVRegion {bIsValid=false, max=<0 0>, min=<0 0>}
    }
    colorAndOpacity=<rgba 0.5776 0.5776 0.5776 1>
   }
   slot {
    class=GridSlot
    properties {horizontalAlignment=HAlign_Left, layer=1, row=1, verticalAlignment=VAlign_Center, nudge=<7 -3>, padding=<ltrb 0 0 8 0>}
   }
  }
  ProductIcon {
   widgetClassPath=&Image
   bIsVariable=true
   parent=Root
   widget=ProductIcon
   properties {
    brush {
     bIsDynamicallyLoaded=false
     drawAs=Image
     imageType=FullColor
     mirroring=NoMirror
     resourceName=None
     resourceObject=/FortUILibrary/Textures/MaterialInstances/MI_UIKit_Icon_BattlePass
     tiling=NoTile
     imageSize=<32 32>
     margin=<ltrb 0 0 0 0>
     outlineSettings {
      bUseBrushTransparency=false
      roundingType=HalfHeightRadius
      width=0
      color {colorUseRule=UseColor_Specified, specifiedColor=<rgba 0 0 0 0>}
      cornerRadii=<0 0 0 0>
     }
     tintColor {colorUseRule=UseColor_Specified, specifiedColor=<rgba 1 1 1 1>}
     uVRegion {bIsValid=false, max=<0 0>, min=<0 0>}
    }
    colorAndOpacity=<rgba 0.5776 0.5776 0.5776 1>
   }
   slot {class=GridSlot, properties {horizontalAlignment=HAlign_Left, layer=1, verticalAlignment=VAlign_Center, padding=<ltrb 0 0 8 0>}}
  }
  Lazy_Background {
   widgetClassPath=&FortLazyImage
   bIsVariable=true
   parent=Root
   widget=Lazy_Background
   properties {
    brush {
     bIsDynamicallyLoaded=false
     drawAs=Image
     imageType=FullColor
     mirroring=NoMirror
     resourceName=None
     resourceObject=/FortniteCoreUI/Items/QuestItems/Categories/Art/MI_UI_CategoryLoading
     tiling=NoTile
     imageSize=<0 0>
     margin=<ltrb 0 0 0 0>
     outlineSettings {
      bUseBrushTransparency=false
      roundingType=HalfHeightRadius
      width=0
      color {colorUseRule=UseColor_Specified, specifiedColor=<rgba 0 0 0 0>}
      cornerRadii=<0 0 0 0>
     }
     tintColor {colorUseRule=UseColor_Specified, specifiedColor=<rgba 1 1 1 1>}
     uVRegion {bIsValid=false, max=<0 0>, min=<0 0>}
    }
    loadingBackgroundBrush {
     bIsDynamicallyLoaded=false
     drawAs=Image
     imageType=FullColor
     mirroring=NoMirror
     resourceName=None
     resourceObject=&MI_UI_CategoryLoading
     tiling=NoTile
     imageSize=<32 32>
     margin=<ltrb 0 0 0 0>
     outlineSettings {
      bUseBrushTransparency=false
      roundingType=HalfHeightRadius
      width=0
      color {colorUseRule=UseColor_Specified, specifiedColor=<rgba 0 0 0 0>}
      cornerRadii=<0 0 0 0>
     }
     tintColor {colorUseRule=UseColor_Specified, specifiedColor=<rgba 1 1 1 0.5>}
     uVRegion {bIsValid=false, max=<0 0>, min=<0 0>}
    }
   }
   slot {class=GridSlot, properties {column=1, columnSpan=5, rowSpan=2, padding=<ltrb 0 0 0 -8>}}
  }
  Spacer_Product {
   widgetClassPath=&Spacer
   bIsVariable=true
   parent=Root
   widget=Spacer_Product
   properties {visibility=Collapsed, size=<165 25>}
   slot {class=GridSlot, properties {column=3, horizontalAlignment=HAlign_Right, verticalAlignment=VAlign_Center}}
  }
 ]
}
