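# Quest screen panel that shows a custom widget, side rewards or the quest map for the active category and selection.
# Setters feed Refresh, which derives content from view models and passes it to CommitState, which switches views in that priority order.
# Custom widgets and backgrounds load asynchronously, skipping already-current ones.
# Map controls follow activation, visible view, sidebar collapse and move mode; custom categories can offer a map/character toggle.
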WidgetBlueprint /QuestUI/QuestScreen/Modal/WBP_SelectionPanel
animations [HideVisibilitySwitcher {
  binding_count=2
  length_seconds=1
  object=&self:HideVisibilitySwitcher
  sequence {bindings [
    VisibilitySwitcher {
     id=#1
     possessed_class=CommonVisibilitySwitcher
     tracks [{
       type=MovieSceneFloatTrack
       display_name=Render Opacity
       sections [{type=MovieSceneFloatSection, end_frame=20, start_frame=0, channels [None {type=MovieSceneScriptingFloatChannel, keys [[0,1],[20,0]]}]}]
      }]
    }
    QuestMap {
     id=#2
     possessed_class=WBP_QuestMap_C
     tracks [{
       type=MovieSceneByteTrack
       display_name=Visibility
       sections [{type=MovieSceneByteSection, end_frame=20, start_frame=0, channels [None {type=MovieSceneScriptingByteChannel, keys [[0,4],[20,1]]}]}]
      }]
    }
   ]}
 }]
graphs [
 BP_GetDesiredInputConfig {nodes [
   Get Desired Input Config() {is=title K2Node_FunctionEntry}
   > Return Node!(
    execute
    ReturnValue:UIInput Config Structure=Get DesiredInputConfig!(self:WBP Quest Map Object Reference=Get QuestMap() -> (QuestMap:WBP Quest Map Object Reference)) -> (DesiredInputConfig:UIInput Config Structure)
   )
  ]}
 CommitState {nodes [
   CommitState!() -> (
    then
    CustomMapWidget:"User Widget Soft Class Reference (by ref)"
    Background:Soft Object Reference
    Rewards:Array of Quest UI Reward VM Object References
    RewardSourceVM:Quest UI Base Content View Model Object Reference
    AllowMap:bool=~
   ) {is=title K2Node_FunctionEntry}
   > n1 = Sequence!(execute) -> (then_0, then_1, then_2:Exec)
   n2 = Branch(execute=@n1.then_2, Condition=Get ApplyRewards() -> (ApplyRewards:bool))
   > n3 = UpdateRewards!(
    execute
    self:WBP Side Rewards Panel Object Reference=Get SideRewardsPanel() -> (SideRewardsPanel:WBP Side Rewards Panel Object Reference)
    RewardsCustomWidgets:Array of Quest UI Reward VM Object References=Get Rewards() -> (Rewards:Array of Quest UI Reward VM Object References)
    QuestContentVM:Quest UI Base Content View Model Object Reference=Get RewardSourceVM() -> (RewardSourceVM:Quest UI Base Content View Model Object Reference)
   ) -> (then)
   UpdateMapContainerActivation(execute=[@n3.then, @n2.else])
   > UpdateInspectRewards()
   > UpdateToggleMap()
   > LoadBackground!(execute, self, Background:Soft Object Reference=Get Background() -> (Background:Soft Object Reference)) -> (then)
   > LoadCustomWidget!(
    execute
    self
    Custom Map Widget:User Widget Soft Class Reference=Get CustomMapWidget() -> (CustomMapWidget:"User Widget Soft Class Reference (by ref)")
   ) -> (then)
   n4 = Branch(
    execute=@n1.then_0
    Condition=IsValidSoftClassReference!(SoftClassReference:"Object Soft Class Reference (by ref)"=Get CustomMapWidget()) -> (ReturnValue:bool)
   )
   > Set WidgetToSelect(WidgetToSelect:Widget Object Reference=Get SB_CustomWidget() -> (SB_CustomWidget:Size Box Object Reference)) -> (Output_Get:Widget Object Reference)
   n5 = Get Rewards()
   n6 = Branch(execute=@n4.else, Condition="integer > integer"(A=Length(TargetArray:Array of Quest UI Reward VM Object References=@n5)))
   > Set WidgetToSelect(WidgetToSelect=Get Border_SideReward() -> (Border_SideReward:Border Object Reference))
   > Set ApplyRewards(ApplyRewards:bool=true) -> (Output_Get:bool)
   Branch(execute=@n6.else, Condition=Get AllowMap() -> (AllowMap:bool))
   > Set WidgetToSelect(WidgetToSelect=Get QuestMap())
   > Set ApplyMap(ApplyMap:bool=true) -> (Output_Get:bool)
   n7 = Is Valid(exec=@n1.then_1, InputObject=Get WidgetToSelect() -> (WidgetToSelect:Widget Object Reference))
   n8 = Get VisibilitySwitcher() -> (VisibilitySwitcher:Common Visibility Switcher Object Reference)
   SetVisibility(execute=@n7.Is Not Valid, self=@n8, InVisibility=Collapsed)
   SetVisibility(execute=@n7.Is Valid, self=@n8, InVisibility=SelfHitTestInvisible)
   > SetActiveWidget(self=@n8, Widget=Get WidgetToSelect())
  ]}
 EventGraph {nodes [
   "On Clicked (DecoyButtonToggleCharacterMap)"!() -> (OutputDelegate, then, Button:Common Button Base Object Reference) {is=title K2Node_ComponentBoundEvent}
   n1 = Get VisibilitySwitcher()
   > Set IsMapVisibleForCustomCategory(IsMapVisibleForCustomCategory:bool="Equal (Object)"(A=GetActiveWidget!(self:Common Visibility Switcher Object Reference=@n1) -> (ReturnValue:Widget Object Reference), B=Get SB_CustomWidget())) -> (Output_Get:bool)
   > Refresh()
   n2 = LoadBackground!() -> (OutputDelegate, then, Background:Soft Object Reference) {is=title K2Node_CustomEvent}
   > n3 = Branch(Condition=IsCurrentBackground!(Background:Soft Object Reference=@n2.Background) -> (IsMatch:bool))
   n4 = Branch(execute=@n3.else, Condition=IsValidSoftObjectReference(SoftObjectReference=@n2.Background))
   > n5 = Async Load Asset(Asset=@n2.Background)
   HandleBackgroundLoaded!(execute=@n5.Completed, self, Background:Object Reference=@n5) -> (then)
   HandleBackgroundLoaded(execute=@n4.else)
   n6 = LoadCustomWidget!() -> (OutputDelegate, then, Custom Map Widget:User Widget Soft Class Reference) {is=title K2Node_CustomEvent}
   > n7 = Branch(Condition=IsCurrentCustomWidget!(CustomWidget:User Widget Soft Class Reference=@n6.Custom Map Widget) -> (IsMatch:bool))
   > UpdateCustomWidgetVM()
   n8 = Branch(execute=@n7.else, Condition=IsValidSoftClassReference(SoftClassReference=@n6.Custom Map Widget))
   > n9 = CreateWidgetAsync!(
    execute:Exec
    UserWidgetSoftClass:User Widget Soft Class Reference=@n6.Custom Map Widget
    OwningPlayer:Player Controller Object Reference
    bSuspendInputUntilComplete:bool=true
   ) -> (then:Exec, OnComplete:Exec, OnFailure:Exec, UserWidget:User Widget Object Reference) {is=title K2Node_AsyncAction}
   HandleCustomWidgetLoaded!(execute=@n9.OnComplete, self, CustomWidget:User Widget Object Reference=@n9) -> (then)
   HandleCustomWidgetLoaded(execute=@n8.else)
   Event On Deactivated()
   > Parent: On Deactivated() {is=title K2Node_CallParentFunction}
   > UpdateMapContainerActivation()
   n10 = Event PreConstruct()
   > Parent: PreConstruct(IsDesignTime=@n10.IsDesignTime)
   > SetCompactUIStyle() {is=title K2Node_CallFunction}
   > SetBottomButtonVisibility!(
    execute
    self:WBP Bottom Bar Decoy Button Object Reference=Get DecoyButtonToggleCharacterMap() -> (DecoyButtonToggleCharacterMap:WBP Bottom Bar Decoy Button Object Reference)
    ShouldButtonBeVisible:bool=false
   ) -> (then)
   Event On Activated()
   > Parent: On Activated() {is=title K2Node_CallParentFunction}
   > Refresh()
  ]}
 GetAllowMap {nodes [
   GetAllowMap() {is=title K2Node_FunctionEntry}
   > n1 = Get!(execute:Exec) -> (then:Exec, else:Exec, Category Interface VM:Quest UI Category Interface View Model Object Reference)
   n2 = IsValid(Object=@n1)
   n3 = GetAllowsMap!(self:Quest UI Category Interface View Model Object Reference=@n1) -> (ReturnValue:bool)
   n4 = NOT Boolean(A=IsToggleMapAllowed!() -> (IsAllowed:bool))
   n5 = Get IsMapVisibleForCustomCategory() -> (IsMapVisibleForCustomCategory:bool)
   > Return Node!(
    execute
    AllowMap:bool=AND Boolean!(
     A=AND Boolean!(A=@n2, B=@n3) -> (ReturnValue)
     B=Get IsSidebarCollapsed() -> (IsSidebarCollapsed:bool)
     C:bool=GetCurrentSelectionAllowsMap!() -> (ReturnValue:bool)
     D:bool=OR Boolean(A=@n4, B=@n5)
    ) -> (ReturnValue)
   )
   Return Node(execute=@n1.else, AllowMap=false)
  ]}
 GetCurrentSelectionAllowsMap {nodes [
   GetCurrentSelectionAllowsMap() {is=title K2Node_FunctionEntry}
   > Return Node!(
    execute
    ReturnValue:bool=Get bMapLayerContainerEnabled!(self:Athena Full Screen Map Base Object Reference=Get QuestMap()) -> (bMapLayerContainerEnabled:bool)
   )
  ]}
 GetCustomWidgetAndBackground {nodes [
   GetCustomWidgetAndBackground() {is=title K2Node_FunctionEntry}
   > n1 = Branch(Condition=NOT Boolean(A=Get IsInspectRewardsActive() -> (IsInspectRewardsActive:bool)))
   > n2 = Sequence()
   n3 = Get SelectionVM() -> (SelectionVM:Object Reference)
   > Is Valid(InputObject=@n3)
   > n4 = Cast To QuestUIQuestVM!(execute, Object=@n3) -> (then, CastFailed, AsQuest UI Quest View Model:Quest UI Quest View Model Object Reference) {is=title K2Node_DynamicCast}
   > n5 = Branch(Condition=Get bContentEnabled!(self:Quest UI Base Content View Model Object Reference=@n4) -> (bContentEnabled:bool))
   > Set ActiveCustomMapWidget(ActiveCustomMapWidget:User Widget Soft Class Reference=Get CustomMapWidget!(self:Quest UI Quest View Model Object Reference=@n4) -> (CustomMapWidget:User Widget Soft Class Reference)) -> (Output_Get:User Widget Soft Class Reference)
   Set ActiveCustomMapWidget(
    execute=@n5.else
    ActiveCustomMapWidget=&/WBP_CustomHoverContent_DisabledQuest.WBP_CustomHoverContent_DisabledQuest_C
   )
   n6 = Cast To QuestUIHorizontalPunchCardVM!(execute=@n4.CastFailed, Object=@n3) -> (then, CastFailed, AsQuest UI Horizontal Punch Card View Model:Quest UI Horizontal Punch Card View Model Object Reference) {is=title K2Node_DynamicCast}
   > Set ActiveCustomMapWidget(ActiveCustomMapWidget=Get CustomMapWidget(self:Quest UI Punch Card View Model Object Reference=@n6))
   > Set ActiveBackground(ActiveBackground:Soft Object Reference=Get CustomMapBackground!(self:Quest UI Punch Card View Model Object Reference=@n6) -> (CustomMapBackground:Soft Object Reference)) -> (Output_Get:Soft Object Reference)
   n7 = Cast To MVVM_PunchCardQuest!(execute=@n6.CastFailed, Object=@n3) -> (then, CastFailed, AsMVVM Punch Card Quest:MVVM Punch Card Quest Object Reference) {is=title K2Node_DynamicCast}
   > Set ActiveCustomMapWidget(ActiveCustomMapWidget=Get CustomWidget!(self:MVVM Punch Card Quest Object Reference=@n7) -> (CustomWidget:User Widget Soft Class Reference))
   n8 = Cast To QuestUICustomRowVM!(execute=@n7.CastFailed, Object=@n3) -> (then, CastFailed, AsQuest UI Custom Row View Model:Quest UI Custom Row View Model Object Reference) {is=title K2Node_DynamicCast}
   > n9 = Branch(Condition=AND Boolean(A=IsToggleMapAllowed(), B=Get IsMapVisibleForCustomCategory()))
   n10 = Get SelectionWidgetOverride!(self:Quest UI Custom Row View Model Object Reference=@n8) -> (SelectionWidgetOverride:Common Activatable Widget Soft Class Reference)
   n11 = Branch(execute=@n9.else, Condition=IsValidSoftClassReference(SoftClassReference=@n10))
   > Set ActiveCustomMapWidget(ActiveCustomMapWidget=@n10)
   Set ActiveCustomMapWidget(execute=@n11.else, ActiveCustomMapWidget=/QuestUI/QuestScreen/V3/WBP_CustomSelectionPanel.WBP_CustomSelectionPanel_C)
   Return Node!(
    execute=[@n2.then_2, @n1.else]
    CustomMapWidget:User Widget Soft Class Reference=Get ActiveCustomMapWidget() -> (ActiveCustomMapWidget:User Widget Soft Class Reference)
    Background:Soft Object Reference=Get ActiveBackground() -> (ActiveBackground:Soft Object Reference)
   )
   n12 = Cast To QuestUICategoryVM!(execute=@n2.then_1, Object=Get Category Interface VM() -> (Category Interface VM:Quest UI Category Interface View Model Object Reference)) -> (then, CastFailed, AsQuest UI Category View Model:Quest UI Category View Model Object Reference) {is=title K2Node_DynamicCast}
   > n13 = Sequence!(execute) -> (then_0, then_1)
   > n14 = Branch(Condition=IsValidSoftObjectReference(SoftObjectReference=Get ActiveBackground()))
   Set ActiveBackground(execute=@n14.else, ActiveBackground=Get CustomMapBackground(self:Quest UI Category View Model Object Reference=@n12))
   n15 = Branch(execute=@n13.then_1, Condition=IsValidSoftClassReference(SoftClassReference=Get ActiveCustomMapWidget()))
   Branch(
    execute=@n15.else
    Condition=Get bPropagateCustomMapWidgetForAllContent!(self:Quest UI Category View Model Object Reference=@n12) -> (bPropagateCustomMapWidgetForAllContent:bool)
   )
   > Set ActiveCustomMapWidget(ActiveCustomMapWidget=Get CustomMapWidget(self:Quest UI Category View Model Object Reference=@n12))
  ]}
 GetRewards {nodes [
   GetRewards() {is=title K2Node_FunctionEntry}
   > n1 = Sequence()
   n2 = Get SelectionVM()
   > Is Valid(InputObject=@n2)
   > n3 = Cast To QuestUIHorizontalPunchCardVM(Object=@n2)
   > n4 = Branch(Condition=CanShowMap!(self:Quest UI Punch Card View Model Object Reference=@n3) -> (ReturnValue:bool))
   Set SourceVM(execute=@n4.else, SourceVM:Quest UI Base Content View Model Object Reference=@n3) -> (Output_Get:Quest UI Base Content View Model Object Reference)
   > n5 = Cast To QuestUIQuestVM(Object="Get (a copy)"(
     Array:Array of Quest UI Base Content View Model Object References=Get OrderedContentVMs!(self:Quest UI Punch Card View Model Object Reference=@n3) -> (OrderedContentVMs:Array of Quest UI Base Content View Model Object References)
     Dimension 1=Get BonusGoalRewardIndex() -> (BonusGoalRewardIndex:int=0)
    ) -> (Output:Quest UI Base Content View Model Object Reference))
   > Append Array!(
    execute
    TargetArray:Array of Quest UI Reward VM Object References=Get SelectionRewards() -> (SelectionRewards:Array of Quest UI Reward VM Object References)
    SourceArray:Array of Quest UI Reward VM Object References=Get RewardVMs!(self:Quest UI Quest View Model Object Reference=@n5) -> (RewardVMs:Array of Quest UI Reward VM Object References)
   ) -> (then)
   n6 = Cast To MVVM_PunchCardQuest(execute=@n3.CastFailed, Object=@n2)
   > n7 = Set QuestVM(QuestVM:Quest UI Quest View Model Object Reference=Get QuestData!(self:MVVM Punch Card Quest Object Reference=@n6) -> (QuestData:Quest UI Quest View Model Object Reference)) -> (Output_Get:Quest UI Quest View Model Object Reference)
   n8 = Branch(execute=[@n7.then, @n17.then], Condition=GetCurrentSelectionAllowsMap() {is=title K2Node_CallFunction})
   n9 = Get QuestVM() -> (QuestVM:Quest UI Quest View Model Object Reference)
   n10 = Get RewardVMs(self=@n9)
   > n11 = For Each Loop(Array:Array of Quest UI Reward VM Object References=@n10) -> (Array Element:Quest UI Reward VM Object Reference)
   n12 = Branch(execute=@n11.Completed, Condition=Get RewardsAllowed() -> (RewardsAllowed:bool))
   Set SourceVM(execute=[@n12.then, @n8.else], SourceVM=@n9)
   > Append Array(TargetArray=Get SelectionRewards(), SourceArray=@n10)
   n13 = Cast To QuestUIRewardCinematicVM!(execute=@n11.LoopBody, Object=@n11) -> (then, CastFailed, AsQuest UI Reward Cinematic VM:Quest UI Reward Cinematic VM Object Reference) {is=title K2Node_DynamicCast}
   n14 = Get RequiredCount!(self:Quest UI Quest View Model Object Reference=@n9) -> (RequiredCount:int=0)
   n15 = Get AchievedCount!(self:Quest UI Quest View Model Object Reference=@n9) -> (AchievedCount:int=0)
   > Branch(Condition=OR Boolean(
     A=Get bIsRewardContentAvailable!(self:Quest UI Reward VM Object Reference=@n13) -> (bIsRewardContentAvailable:bool)
     B="integer >= integer"(A=@n15, B=@n14)
    ))
   > Set RewardsAllowed(RewardsAllowed:bool=true) -> (Output_Get:bool)
   n16 = Cast To QuestUIQuestVM(execute=@n6.CastFailed, Object=@n2)
   > n17 = Set QuestVM(QuestVM=@n16)
   Return Node!(
    execute=@n1.then_1
    CurrentRewards:Array of Quest UI Reward VM Object References=Get SelectionRewards()
    RewardSourceVM:Quest UI Base Content View Model Object Reference=Get SourceVM() -> (SourceVM:Quest UI Base Content View Model Object Reference)
   )
  ]}
 HandleBackgroundLoaded {nodes [
   n1 = HandleBackgroundLoaded!() -> (then, Background:Object Reference) {is=title K2Node_FunctionEntry}
   > n2 = Cast To MaterialInstance!(execute, Object=@n1) -> (then, CastFailed, AsMaterial Instance:Material Instance Object Reference) {is=title K2Node_DynamicCast}
   n3 = Get CustomMapBackground!() -> (CustomMapBackground:Image Object Reference)
   > SetVisibility(self=@n3, InVisibility=HitTestInvisible)
   > SetBrushFromMaterial(self=@n3, Material=@n2)
   SetVisibility(execute=@n2.CastFailed, self=@n3, InVisibility=Collapsed)
  ]}
 HandleCustomWidgetLoaded {nodes [
   n1 = HandleCustomWidgetLoaded!() -> (then, CustomWidget:User Widget Object Reference) {is=title K2Node_FunctionEntry}
   n2 = Get CustomWidget_Slot() -> (CustomWidget_Slot:Named Slot Object Reference)
   > ClearChildren(self=@n2)
   > Is Valid(InputObject=@n1)
   > AddChild(self=@n2, Content=@n1)
   > UpdateCustomWidgetVM()
  ]}
 HandleFocusRequest {nodes [
   HandleFocusRequest() {is=title K2Node_FunctionEntry}
   n1 = Get VisibilitySwitcher()
   > Branch(Condition="Equal (Object)"(A=GetActiveWidget(self=@n1), B=Get SB_CustomWidget()))
   > n2 = Cast To WBP_CustomHoverContent_DisabledQuest!(
    execute
    Object=GetChildAt!(self:Panel Widget Object Reference=Get CustomWidget_Slot(), Index:int=0) -> (ReturnValue:Widget Object Reference)
   ) -> (then, CastFailed, AsWBP Custom Hover Content Disabled Quest:WBP Custom Hover Content Disabled Quest Object Reference) {is=title K2Node_DynamicCast}
   > Branch(Condition=Get IsShowingSelectButton!(self:WBP Custom Hover Content Disabled Quest Object Reference=@n2) -> (IsShowingSelectButton:bool))
   > SetFocus(self=@n2)
  ]}
 HandleInspectRewardsToggled {nodes [
   HandleInspectRewardsToggled() {is=title K2Node_FunctionEntry}
   > Refresh()
   > Return Node!(execute, IsInspectRewardsActive:bool=Get IsInspectRewardsActive())
  ]}
 HandleSidebarCollapsedStateChanged {nodes [
   n1 = HandleSidebarCollapsedStateChanged!() -> (then, Collapsed:bool=~) {is=title K2Node_FunctionEntry}
   > n2 = Branch(Condition="Not Equal (Boolean)"(A=@n1, B=Get IsSidebarCollapsed()) -> (ReturnValue=~))
   > Set IsSidebarCollapsed(IsSidebarCollapsed:bool=@n1) -> (Output_Get:bool)
   > UpdateMapContainerActivation()
   n3 = Select(Option 0:EUMGSequencePlayMode Enum, Option 1:EUMGSequencePlayMode Enum=Reverse, Index=@n1) -> (ReturnValue:EUMGSequencePlayMode Enum)
   > PlayAnimation(InAnimation=Get HideVisibilitySwitcher() -> (HideVisibilitySwitcher:Widget Animation Object Reference), PlayMode=@n3, PlaybackSpeed=4)
   n4 = Get HideVisibilitySwitcher()
   PlayAnimation(execute=@n2.else, InAnimation=@n4, StartAtTime=GetEndTime(self=@n4), PlayMode=@n3)
  ]}
 IsCurrentBackground {nodes [
   n1 = IsCurrentBackground!() -> (then, Background:Soft Object Reference) {is=title K2Node_FunctionEntry}
   > n2 = Branch(Condition=IsValidSoftObjectReference(SoftObjectReference=@n1))
   n3 = Get CustomMapBackground()
   n4 = Get Brush!(self:Image Object Reference=@n3) -> (Brush:Slate Brush Structure)
   > Return Node!(
    execute
    IsMatch:bool="Equal (Object)"(
     A=Break Slate Brush!(SlateBrush:"Slate Brush Structure (by ref)"=@n4) -> (ResourceObject:Object Reference) {is=title K2Node_BreakStruct}
     B:Soft Object Reference=@n1
    )
   )
   Return Node(
    execute=@n2.else
    IsMatch="Equal (Enum)"(
     A:ESlateVisibility Enum=GetVisibility!(self:Widget Object Reference=@n3) -> (ReturnValue:ESlateVisibility Enum=Visible)
     B:ESlateVisibility Enum=Collapsed
    )
   )
  ]}
 IsCurrentCustomWidget {nodes [
   n1 = IsCurrentCustomWidget!() -> (then, CustomWidget:User Widget Soft Class Reference) {is=title K2Node_FunctionEntry}
   n2 = GetChildAt(self=Get CustomWidget_Slot(), Index=0)
   > n3 = Is Valid(InputObject=@n2)
   Return Node(execute=@n3.Is Not Valid, IsMatch=NOT Boolean(A=IsValidSoftClassReference(SoftClassReference=@n1)))
   > Return Node(IsMatch="Equal (Class)"(A=Get Class(Object=@n2) -> (ReturnValue:Widget Class Reference), B:User Widget Soft Class Reference=@n1))
  ]}
 IsInFrontend {nodes [
   IsInFrontend() {is=title K2Node_FunctionEntry}
   > n1 = Cast To FortPlayerControllerFrontEnd!(execute, Object=GetOwningPlayer()) -> (then, CastFailed, AsFort Player Controller Front End:Fort Player Controller Front End Object Reference) {is=title K2Node_DynamicCast}
   > Return Node!(execute, InFrontend:bool=true)
   Return Node(execute=@n1.CastFailed, InFrontend=false)
  ]}
 IsToggleMapAllowed {nodes [
   IsToggleMapAllowed() {is=title K2Node_FunctionEntry}
   n1 = Get Category Interface VM()
   > n2 = Cast To QuestUICustomCategoryVM!(execute, Object=@n1) -> (then, CastFailed, AsQuest UI Custom Category View Model:Quest UI Custom Category View Model Object Reference) {is=title K2Node_DynamicCast}
   > n3 = Branch(Condition=GetAllowsMap(self=@n1))
   > Return Node!(execute, IsAllowed:bool=true)
   Return Node(execute=[@n3.else, @n2.CastFailed], IsAllowed=false)
  ]}
 OnInspectRewardsAllowed {nodes [OnInspectRewardsAllowed!() -> (then, IsAllowed:bool=~) {is=title K2Node_FunctionEntry}]}
 Refresh {nodes [
   Refresh() {is=title K2Node_FunctionEntry}
   n1 = GetRewards!() -> (CurrentRewards:Array of Quest UI Reward VM Object References, RewardSourceVM:Quest UI Base Content View Model Object Reference) {is=title K2Node_CallFunction}
   n2 = GetCustomWidgetAndBackground!() -> (CustomMapWidget:User Widget Soft Class Reference, Background:Soft Object Reference) {is=title K2Node_CallFunction}
   > CommitState!(
    execute
    self
    CustomMapWidget:"User Widget Soft Class Reference (by ref)"=@n2
    Background:Soft Object Reference=@n2.Background
    Rewards:Array of Quest UI Reward VM Object References=@n1
    RewardSourceVM:Quest UI Base Content View Model Object Reference=@n1.RewardSourceVM
    AllowMap:bool=GetAllowMap!() -> (AllowMap:bool) {is=title K2Node_CallFunction}
   ) -> (then) {is=title K2Node_CallFunction}
  ]}
 SetActiveCategory {nodes [
   n1 = SetActiveCategory!() -> (then, CategoryInterfaceVM:Quest UI Category Interface View Model Object Reference) {is=title K2Node_FunctionEntry}
   > Set Category Interface VM(Category Interface VM:Quest UI Category Interface View Model Object Reference=@n1) -> (Output_Get:Quest UI Category Interface View Model Object Reference)
   > UpdateMapData()
   > Refresh() {is=title K2Node_CallFunction}
  ]}
 SetActiveSelection {nodes [
   n1 = SetActiveSelection!() -> (then, SelectionVM:Object Reference) {is=title K2Node_FunctionEntry}
   > Is Valid(InputObject=@n1)
   > n2 = Sequence()
   > n3 = Set SelectionVM(SelectionVM:Object Reference=@n1) -> (Output_Get:Object Reference)
   > n4 = Cast To QuestUIHorizontalPunchCardVM(Object=@n3)
   > n5 = Set Quest UI Horizontal Punch Card View Model(Quest UI Horizontal Punch Card View Model:Quest UI Horizontal Punch Card View Model Object Reference=@n4) -> (Output_Get:Quest UI Horizontal Punch Card View Model Object Reference)
   Set IsInspectRewardsActive(execute=[@n5.then, @n7.then], IsInspectRewardsActive:bool=false) -> (Output_Get:bool)
   > n6 = Cast To QuestUICustomRowVM(Object=Get SelectionVM())
   Set RowSelectionVM(execute=@n6.CastFailed, RowSelectionVM:Quest UI Base Content View Model Object Reference) -> (Output_Get:Quest UI Base Content View Model Object Reference)
   n7 = Set Quest UI Horizontal Punch Card View Model(execute=@n4.CastFailed)
   UpdateMapData(execute=@n2.then_1)
   > Refresh()
  ]}
 SetCompactUIStyle {nodes [
   SetCompactUIStyle() {is=title K2Node_FunctionEntry}
   > Branch(Condition=IsUsingCompactUI())
   n1 = SlotAsBorderSlot!(Widget:Widget Object Reference=Get SideRewardsPanel()) -> (ReturnValue:Border Slot Object Reference)
   n2 = Get!(self:Border Slot Object Reference=@n1) -> (Padding_Left:float=~, Padding_Top:float=~, Padding_Right:float=~, Padding_Bottom:float=~)
   > SetPadding!(
    execute
    self:Border Slot Object Reference=@n1
    InPadding_Left:float=@n2
    InPadding_Top:float=0
    InPadding_Right:float=@n2.Padding_Right
    InPadding_Bottom:float=@n2.Padding_Bottom
   ) -> (then)
   n3 = IsTabletAspectRatio!() -> (ReturnValue:bool)
   n4 = Get SB_CustomWidget()
   > SetWidthOverride(self=@n4, InWidthOverride=SelectFloat(A=650, B=800, bPickA=@n3))
   > SetHeightOverride(self=@n4, InHeightOverride=SelectFloat(A=900, B=700, bPickA=@n3))
   n5 = IsInFrontend!() -> (InFrontend:bool) {is=title K2Node_CallFunction}
   n6 = SlotAsOverlaySlot(Widget=Get VisibilitySwitcher())
   n7 = Get(self:Overlay Slot Object Reference=@n6) -> (Padding_Left=~, Padding_Top=~, Padding_Right=~, Padding_Bottom=~)
   > SetPadding!(
    execute
    self:Overlay Slot Object Reference=@n6
    InPadding:Margin Structure=Make Margin(Left=1200, Top=SelectFloat(A=50, B=0, bPickA=@n5), Right=@n7.Padding_Right, Bottom=@n7.Padding_Bottom)
   ) -> (then)
  ]}
 SetIsMoveMode {nodes [
   n1 = SetIsMoveMode!() -> (then, IsMoving:bool=~) {is=title K2Node_FunctionEntry}
   > Set IsMoving(IsMoving:bool=@n1) -> (Output_Get:bool)
   > UpdateMapContainerActivation()
  ]}
 SetRewardPage {nodes [
   n1 = SetRewardPage!() -> (then, Index:int) {is=title K2Node_FunctionEntry}
   > Set BonusGoalRewardIndex(BonusGoalRewardIndex:int=@n1) -> (Output_Get:int=0)
   > Refresh()
  ]}
 SetRowSelection {nodes [
   n1 = SetRowSelection!() -> (then, RowSelectionVM:Quest UI Base Content View Model Object Reference) {is=title K2Node_FunctionEntry}
   > Set RowSelectionVM(RowSelectionVM=@n1)
   > UpdateMapData()
   > Refresh()
  ]}
 UpdateCustomWidgetVM {nodes [
   UpdateCustomWidgetVM() {is=title K2Node_FunctionEntry}
   n1 = GetChildAt(self=Get CustomWidget_Slot(), Index=0)
   > n2 = Cast To WBP_CustomHoverContent_DisabledQuest(Object=@n1)
   > n3 = Cast To QuestUIQuestVM(Object=Get SelectionVM())
   > SetQuestUIQuestVM!(execute, self:WBP Custom Hover Content Disabled Quest Object Reference=@n2, Viewmodel:Quest UI Quest View Model Object Reference=@n3) -> (then)
   Cast To QuestUICustomRowVM(execute=@n2.CastFailed, Object=Get SelectionVM())
   > n4 = Cast To UserWidget!(execute, Object=@n1) -> (then, CastFailed, AsUser Widget:User Widget Object Reference) {is=title K2Node_DynamicCast}
   > SetViewModelByClass!(
    execute
    Widget:User Widget Object Reference=@n4
    ViewModel:Notify Field Value Changed Interface=Get RowSelectionVM() -> (RowSelectionVM:Quest UI Base Content View Model Object Reference)
   ) -> (then)
  ]}
 UpdateInspectRewards {nodes [
   UpdateInspectRewards() {is=title K2Node_FunctionEntry}
   n1 = Get Quest UI Horizontal Punch Card View Model() -> (Quest UI Horizontal Punch Card View Model:Quest UI Horizontal Punch Card View Model Object Reference)
   > n2 = Is Valid(InputObject=@n1)
   > n3 = Is Valid(InputObject=@n1)
   n4 = Get CustomMapWidget(self:Quest UI Punch Card View Model Object Reference=@n1)
   > n5 = Branch(Condition=AND Boolean(
     A=Get bAllowInspectItems!(self:Quest UI Punch Card View Model Object Reference=@n1) -> (bAllowInspectItems:bool)
     B=IsValidSoftClassReference(SoftClassReference=@n4)
    ))
   > Call On Inspect Rewards Allowed!(execute:Exec, self:WBP Selection Panel Object Reference, IsAllowed:bool=true) -> (then:Exec) {is=title K2Node_CallDelegate}
   Call On Inspect Rewards Allowed(execute=[@n5.else, @n3.Is Not Valid, @n2.Is Not Valid])
  ]}
 UpdateIsShowingRewards {nodes [
   n1 = UpdateIsShowingRewards!() -> (then, InShowingRewards:bool=~) {is=title K2Node_FunctionEntry}
   > Set IsInspectRewardsActive(IsInspectRewardsActive=@n1)
   > HandleInspectRewardsToggled!(execute, self) -> (then, IsInspectRewardsActive:bool) {is=title K2Node_CallFunction}
  ]}
 UpdateMapContainerActivation {nodes [
   UpdateMapContainerActivation() {is=title K2Node_FunctionEntry}
   n1 = Get QuestMap()
   n2 = Get VisibilitySwitcher()
   n3 = AND Boolean!(
    A=IsActivated!(self:Common Activatable Widget Object Reference) -> (ReturnValue:bool)
    B="Equal (Object)"(A=GetActiveWidget(self=@n2), B:Object Reference=Get QuestMap())
    C:bool=Get IsSidebarCollapsed()
    D:bool=NOT Boolean(A=Get IsMoving() -> (IsMoving:bool))
   ) -> (ReturnValue)
   > Branch(Condition="Not Equal (Boolean)"(A=Get bIsActive!(self:Common Activatable Widget Object Reference=@n1) -> (bIsActive:bool), B=@n3) -> (ReturnValue=~))
   > Set IsMapEnabled(IsMapEnabled:bool=@n3) -> (Output_Get:bool)
   > EnableBrowsingControls!(
    execute
    self:Athena Map Layer Container Object Reference=GetMapLayerContainer!(self:Athena Full Screen Map Base Object Reference=Get QuestMap()) -> (ReturnValue:Athena Map Layer Container Object Reference)
    bEnabled:bool=@n3
   ) -> (then)
   > n4 = Branch(Condition=@n3)
   n5 = Get QuestMap()
   > ActivateWidget(self=@n5)
   DeactivateWidget(execute=@n4.else, self=@n5)
  ]}
 UpdateMapData {nodes [
   UpdateMapData() {is=title K2Node_FunctionEntry}
   n1 = Get RowSelectionVM()
   > n2 = Cast To QuestUIBaseContentVM!(
    execute
    Object=SelectObject!(A:Object Reference=@n1, B:Object Reference=Get SelectionVM(), bSelectA:bool=IsValid(Object=@n1)) -> (ReturnValue:Object Reference)
   ) -> (then, CastFailed, AsQuest UI Base Content View Model:Quest UI Base Content View Model Object Reference) {is=title K2Node_DynamicCast}
   n3 = Get QuestMap()
   n4 = Get Category Interface VM()
   > HandleSelectionChanged!(
    execute
    self:Quest UIMap Object Reference=@n3
    CategoryVM:Quest UI Category Interface View Model Object Reference=@n4
    ContentVM:Quest UI Base Content View Model Object Reference=@n2
   ) -> (then)
   HandleSelectionChanged(execute=@n2.CastFailed, self=@n3, CategoryVM=@n4)
  ]}
 UpdateToggleMap {nodes [
   UpdateToggleMap() {is=title K2Node_FunctionEntry}
   > n1 = Branch(Condition=IsToggleMapAllowed() {is=title K2Node_CallFunction})
   n2 = Get DecoyButtonToggleCharacterMap()
   > SetBottomButtonVisibility(self=@n2, ShouldButtonBeVisible=true)
   > SetTriggeringInputAction!(
    execute
    self:Common Button Base Object Reference=Get DecoyButtonToggleCharacterMap()
    InputActionRow_DataTable:Data Table Object Reference=/QuestUI/QuestScreen/RewardsPanel/QuestInputActionDataTable
    InputActionRow_RowName:Name=Select(Option 0:Name=ShowMap, Option 1:Name=ShowCharacter, Index=Get IsMapVisibleForCustomCategory()) -> (ReturnValue:Name)
   ) -> (then)
   SetBottomButtonVisibility(execute=@n1.else, self=@n2, ShouldButtonBeVisible=false)
  ]}
]
variables [
 BonusGoalRewardIndex {type=integer}
 Category Interface VM {type=QuestUICategoryInterfaceVM, default=null}
 IsInspectRewardsActive {type=boolean}
 IsMapEnabled {type=boolean}
 IsMapVisibleForCustomCategory {type=boolean}
 IsMoving {type=boolean}
 IsSidebarCollapsed {type=boolean}
 OnInspectRewardsAllowed {}
 Quest UI Horizontal Punch Card View Model {type=QuestUIHorizontalPunchCardVM, default=null}
 RowSelectionVM {type=QuestUIBaseContentVM, default=null}
 SelectionVM {type=Object, default=null}
]
widgets {
 info {inheritedWidgetCount=0, parentClass=&CommonActivatableWidget, rootWidgetClass=&Overlay, widgetCount=10}
 widgets [
  Overlay_0 {widgetClassPath=&Overlay, widget=Overlay_0}
  CustomMapBackground {
   widgetClassPath=&Image
   bIsVariable=true
   parent=Overlay_0
   widget=CustomMapBackground
   properties {
    visibility=Collapsed
    brush {
     bIsDynamicallyLoaded=false
     drawAs=Image
     imageType=FullColor
     mirroring=NoMirror
     resourceName=None
     resourceObject=/FortniteCoreUI/UI/Quests/MaterialInstances/MI_UI_Clouds_RebootRally
     tiling=NoTile
     imageSize=<32 32>
     margin=<ltrb 0 0 0 0>
     outlineSettings {
      bUseBrushTransparency=false
      roundingType=HalfHeightRadius
      width=0
      color {colorUseRule=UseColor_Specified, specifiedColor=<rgba 0 0 0 0>}
      cornerRadii=<0 0 0 0>
     }
     tintColor {colorUseRule=UseColor_Specified, specifiedColor=<rgba 1 1 1 1>}
     uVRegion {bIsValid=false, max=<0 0>, min=<0 0>}
    }
   }
   slot {class=OverlaySlot, properties {horizontalAlignment=HAlign_Fill, verticalAlignment=VAlign_Fill}}
  }
  BackgroundGradient {
   widgetClassPath=&Image
   bIsVariable=true
   parent=Overlay_0
   widget=BackgroundGradient
   properties {
    visibility=SelfHitTestInvisible
    brush {
     bIsDynamicallyLoaded=false
     drawAs=Image
     imageType=FullColor
     mirroring=NoMirror
     resourceName=None
     resourceObject=/QuestUI/QuestScreen/Materials/MI_UI_QuestsModalScreen_BackgroundGradient
     tiling=NoTile
     imageSize=<32 32>
     margin=<ltrb 0 0 0 0>
     outlineSettings {
      bUseBrushTransparency=false
      roundingType=HalfHeightRadius
      width=0
      color {colorUseRule=UseColor_Specified, specifiedColor=<rgba 0 0 0 0>}
      cornerRadii=<0 0 0 0>
     }
     tintColor {colorUseRule=UseColor_Specified, specifiedColor=<rgba 1 1 1 1>}
     uVRegion {bIsValid=false, max=<0 0>, min=<0 0>}
    }
   }
   slot {class=OverlaySlot, properties {horizontalAlignment=HAlign_Fill, verticalAlignment=VAlign_Fill}}
  }
  VisibilitySwitcher {
   widgetClassPath=&CommonVisibilitySwitcher
   bIsVariable=true
   parent=Overlay_0
   widget=VisibilitySwitcher
   slot {class=OverlaySlot, properties {horizontalAlignment=HAlign_Fill, verticalAlignment=VAlign_Fill, padding=<ltrb 1100 136 20 0>}}
  }
  QuestMap {
   widgetClassPath=/QuestUI/QuestScreen/Map/WBP_QuestMap
   bIsVariable=true
   parent=VisibilitySwitcher
   widget=QuestMap
   slot {class=CommonVisibilitySwitcherSlot}
  }
  Border_SideReward {
   widgetClassPath=&Border
   bIsVariable=true
   parent=VisibilitySwitcher
   widget=Border_SideReward
   properties {
    horizontalAlignment=HAlign_Center
    verticalAlignment=VAlign_Center
    background {
     bIsDynamicallyLoaded=false
     drawAs=Border
     imageType=NoImage
     mirroring=NoMirror
     resourceName=None
     resourceObject=null
     tiling=NoTile
     imageSize=<32 32>
     margin=<ltrb 0 0 0 0>
     outlineSettings {
      bUseBrushTransparency=false
      roundingType=HalfHeightRadius
      width=0
      color {colorUseRule=UseColor_Specified, specifiedColor=<rgba 0 0 0 0>}
      cornerRadii=<0 0 0 0>
     }
     tintColor {colorUseRule=UseColor_Specified, specifiedColor=<rgba 1 1 1 1>}
     uVRegion {bIsValid=false, max=<0 0>, min=<0 0>}
    }
    padding=<ltrb 0 50 0 0>
   }
   slot {class=CommonVisibilitySwitcherSlot, properties {padding=<ltrb 64 0 0 0>}}
  }
  SideRewardsPanel {
   widgetClassPath=/QuestUI/QuestScreen/RewardsPanel/WBP_SideRewardsPanel
   bIsVariable=true
   parent=Border_SideReward
   widget=SideRewardsPanel
   slot {class=BorderSlot, properties {horizontalAlignment=HAlign_Center, verticalAlignment=VAlign_Center, padding=<ltrb 0 50 0 0>}}
  }
  SB_CustomWidget {
   widgetClassPath=&SizeBox
   bIsVariable=true
   parent=VisibilitySwitcher
   widget=SB_CustomWidget
   properties {bOverride_HeightOverride=true, bOverride_WidthOverride=true, heightOverride=900, widthOverride=900}
   slot {class=CommonVisibilitySwitcherSlot, properties {horizontalAlignment=HAlign_Center, verticalAlignment=VAlign_Center}}
  }
  CustomWidget_Slot {
   widgetClassPath=&NamedSlot
   bIsVariable=true
   parent=SB_CustomWidget
   widget=CustomWidget_Slot
   slot {class=SizeBoxSlot, properties {padding=<ltrb 0 0 0 50>}}
  }
  DecoyButtonToggleCharacterMap {
   widgetClassPath=/Game/UI/Frontend/BottomBar/WBP_BottomBarDecoyButton
   bIsVariable=true
   parent=Overlay_0
   widget=DecoyButtonToggleCharacterMap
   properties {triggeringInputAction {dataTable=&QuestInputActionDataTable, rowName=ShowMap}}
   slot {class=OverlaySlot}
  }
 ]
}
