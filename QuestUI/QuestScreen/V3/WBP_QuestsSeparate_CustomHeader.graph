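# Custom category header row for the quest screen, showing an uppercase title and an icon over a collections-material background.
# When a poly view item is set, it casts and stores the QuestUICustomCategoryVM.
# SetDescription fills the text, and UpdateIcon lazy-loads the icon asset and sets its dynamic material's BaseRGB to white.
# PreConstruct shows a top spacer when IsEntry is true (the default) and collapses it otherwise.
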WidgetBlueprint /QuestUI/QuestScreen/V3/WBP_QuestsSeparate_CustomHeader
graphs [
 EventGraph {nodes [
   Event PreConstruct()
   > SetVisibility(
    self=Get AdditionalSpacer() -> (AdditionalSpacer:Spacer Object Reference)
    InVisibility=Select(Option 0=Collapsed, Option 1=SelfHitTestInvisible, Index=Get IsEntry() -> (IsEntry:bool))
   )
   n1 = Event On Poly View Item Set!() -> (OutputDelegate, then, Item:Object Reference, SourceIndex:int=0) {is=title K2Node_Event}
   > n2 = Cast To QuestUICustomCategoryVM!(execute, Object=@n1.Item) -> (then, CastFailed, AsQuest UI Custom Category View Model:Quest UI Custom Category View Model Object Reference) {is=title K2Node_DynamicCast}
   > Set QuestUICustomCategoryVM(QuestUICustomCategoryVM:Quest UI Custom Category View Model Object Reference=@n2) -> (Output_Get:Quest UI Custom Category View Model Object Reference)
  ]}
 SetDescription {nodes [
   n1 = SetDescription!() -> (then, Description:Text) {is=title K2Node_FunctionEntry}
   > "SetText (Text)"(self=Get Description() -> (Description:Common Text Object Reference), InText=@n1)
  ]}
 UpdateIcon {nodes [
   n1 = UpdateIcon!() -> (then, Icon:Soft Object Reference) {is=title K2Node_FunctionEntry}
   n2 = Get Lazy_Icon() -> (Lazy_Icon:Fort Lazy Image Object Reference)
   > SetBrushFromLazyDisplayAsset!(execute, self:Common Lazy Image Object Reference=@n2, LazyObject:"Soft Object Reference (by ref)"=@n1, bMatchTextureSize:bool=false) -> (then)
   > n3 = GetDynamicMaterial(self=@n2)
   > SetVectorParameterValue(self=@n3, ParameterName=BaseRGB, Value=(R=1,G=1,B=1,A=1))
  ]}
]
variables [IsEntry {type=boolean, default=true}]
widgets {
 info {inheritedWidgetCount=0, parentClass=&CommonUserWidget, rootWidgetClass=&GridPanel, widgetCount=6}
 widgets [
  Root {widgetClassPath=&GridPanel, widget=Root, properties {columnFill [1], rowFill [0,1]}}
  AdditionalSpacer {
   widgetClassPath=&Spacer
   bIsVariable=true
   parent=Root
   widget=AdditionalSpacer
   properties {size=<0 48>}
   slot {class=GridSlot, properties {columnSpan=5}}
  }
  Description {
   widgetClassPath=&CommonTextBlock
   bIsVariable=true
   parent=Root
   widget=Description
   properties {
    autoWrapText=true
    bAutoCollapseWithEmptyText=true
    scrollStyle=&TextScrollStyle-Base_C
    text=Custom Header
    textTransformPolicy=ToUpper
    font {
     bForceMonospaced=false
     bMaterialIsStencil=false
     fontMaterial=null
     fontObject=&HeadingNow
     letterSpacing=0
     monospacedWidth=1
     size=21
     skewAmount=0
     typefaceFontName=86 Bold Italic
     outlineSettings {
      bApplyOutlineToDropShadows=false
      bMiteredCorners=false
      bSeparateFillAlpha=false
      outlineMaterial=null
      outlineSize=0
      outlineColor=<rgba 0 0 0 1>
     }
    }
    shadowOffset=<0 0>
   }
   slot {class=GridSlot, properties {layer=1, row=1, verticalAlignment=VAlign_Center, padding=<ltrb 20 0 0 0>}}
  }
  Lazy_Background {
   widgetClassPath=&FortLazyImage
   bIsVariable=true
   parent=Root
   widget=Lazy_Background
   properties {
    materialTextureParamName=Texture
    brush {
     bIsDynamicallyLoaded=false
     drawAs=Image
     imageType=FullColor
     mirroring=NoMirror
     resourceName=None
     resourceObject=/FortniteCoreUI/Items/QuestItems/Categories/Art/M_UI_Collections
     tiling=NoTile
     imageSize=<0 60>
     margin=<ltrb 0 0 0 0>
     outlineSettings {
      bUseBrushTransparency=false
      roundingType=HalfHeightRadius
      width=0
      color {colorUseRule=UseColor_Specified, specifiedColor=<rgba 0 0 0 0>}
      cornerRadii=<0 0 0 0>
     }
     tintColor {colorUseRule=UseColor_Specified, specifiedColor=<rgba 1 1 1 1>}
     uVRegion {bIsValid=false, max=<0 0>, min=<0 0>}
    }
    loadingBackgroundBrush {
     bIsDynamicallyLoaded=false
     drawAs=Image
     imageType=FullColor
     mirroring=NoMirror
     resourceName=None
     resourceObject=null
     tiling=NoTile
     imageSize=<32 32>
     margin=<ltrb 0 0 0 0>
     outlineSettings {
      bUseBrushTransparency=false
      roundingType=HalfHeightRadius
      width=0
      color {colorUseRule=UseColor_Specified, specifiedColor=<rgba 0 0 0 0>}
      cornerRadii=<0 0 0 0>
     }
     tintColor {colorUseRule=UseColor_Specified, specifiedColor=<rgba 1 1 1 0.5>}
     uVRegion {bIsValid=false, max=<0 0>, min=<0 0>}
    }
   }
   slot {class=GridSlot, properties {columnSpan=5, row=1}}
  }
  Lazy_Icon {
   widgetClassPath=&FortLazyImage
   bIsVariable=true
   parent=Root
   widget=Lazy_Icon
   properties {
    brush {
     bIsDynamicallyLoaded=false
     drawAs=Image
     imageType=NoImage
     mirroring=NoMirror
     resourceName=None
     resourceObject=null
     tiling=NoTile
     imageSize=<96 96>
     margin=<ltrb 0 0 0 0>
     outlineSettings {
      bUseBrushTransparency=false
      roundingType=HalfHeightRadius
      width=0
      color {colorUseRule=UseColor_Specified, specifiedColor=<rgba 0 0 0 0>}
      cornerRadii=<0 0 0 0>
     }
     tintColor {colorUseRule=UseColor_Specified, specifiedColor=<rgba 1 1 1 1>}
     uVRegion {bIsValid=false, max=<0 0>, min=<0 0>}
    }
    loadingBackgroundBrush {
     bIsDynamicallyLoaded=false
     drawAs=Image
     imageType=FullColor
     mirroring=NoMirror
     resourceName=None
     resourceObject=/FortUILibrary/UIKit/Throbber/MI_UIKit_Monochrome_Throbber
     tiling=NoTile
     imageSize=<32 32>
     margin=<ltrb 0 0 0 0>
     outlineSettings {
      bUseBrushTransparency=false
      roundingType=HalfHeightRadius
      width=0
      color {colorUseRule=UseColor_Specified, specifiedColor=<rgba 0 0 0 0>}
      cornerRadii=<0 0 0 0>
     }
     tintColor {colorUseRule=UseColor_Specified, specifiedColor=<rgba 1 1 1 1>}
     uVRegion {bIsValid=false, max=<0 0>, min=<0 0>}
    }
    renderTransform {angle=0, scale=<1 1>, shear=<0 0>, translation=<0 10>}
   }
   slot {class=GridSlot, properties {column=1, horizontalAlignment=HAlign_Right, row=1, verticalAlignment=VAlign_Top, padding=<ltrb 0 -36 0 0>}}
  }
  Spacer_Product {
   widgetClassPath=&Spacer
   bIsVariable=true
   parent=Root
   widget=Spacer_Product
   properties {size=<30 1>}
   slot {class=GridSlot, properties {column=2, horizontalAlignment=HAlign_Right, row=1, verticalAlignment=VAlign_Center}}
  }
 ]
}
