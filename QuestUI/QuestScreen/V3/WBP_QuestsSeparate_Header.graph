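# Quest screen V3 category header widget showing a title, an expiration pill and a background image.
# It caches the QuestUICategoryVM from its list item, while separate functions set title, expiration and background.
# Expiration shows a countdown for a future time, custom text when no expiration exists, or collapses.
# Backgrounds async-load as materials with header shader parameters set; PreConstruct toggles a top spacer by IsEntry and applies those parameters at design time.
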WidgetBlueprint /QuestUI/QuestScreen/V3/WBP_QuestsSeparate_Header
graphs [
 EventGraph {nodes [
   n1 = LoadBackground!() -> (OutputDelegate, then, Background:Soft Object Reference) {is=title K2Node_CustomEvent}
   > n2 = Async Load Asset(Asset=@n1.Background)
   n3 = Cast To MaterialInterface!(execute=@n2.Completed, Object=@n2) -> (then, CastFailed, AsMaterial Interface:Material Interface Object Reference) {is=title K2Node_DynamicCast}
   > SetBrushFromMaterial(self=Get Lazy_Background() -> (Lazy_Background:Fort Lazy Image Object Reference), Material=@n3)
   > n4 = GetDynamicMaterial(self=Get Lazy_Background())
   > SetScalarParameterValue(self=@n4, ParameterName=IsHeader, Value=0)
   > SetScalarParameterValue(self=@n4, ParameterName=QuestV3, Value=1)
   n5 = Event PreConstruct()
   > SetVisibility(
    self=Get AdditionalSpacer() -> (AdditionalSpacer:Spacer Object Reference)
    InVisibility=Select(Option 0=Collapsed, Option 1=SelfHitTestInvisible, Index=Get IsEntry() -> (IsEntry:bool))
   )
   n6 = Event On Poly View Item Set!() -> (OutputDelegate, then, Item:Object Reference, SourceIndex:int=0) {is=title K2Node_Event}
   > n7 = Cast To QuestUICategoryVM!(execute, Object=@n6.Item) -> (then, CastFailed, AsQuest UI Category View Model:Quest UI Category View Model Object Reference) {is=title K2Node_DynamicCast}
   > Set QuestUICategoryVM(QuestUICategoryVM:Quest UI Category View Model Object Reference=@n7) -> (Output_Get:Quest UI Category View Model Object Reference)
   Branch(Condition=@n5.IsDesignTime)
   > n8 = GetDynamicMaterial(self=Get Lazy_Background())
   > SetScalarParameterValue(self=@n8, ParameterName=IsHeader, Value=0)
   > SetScalarParameterValue(self=@n8, ParameterName=QuestV3, Value=1)
  ]}
 SetDescription {nodes [
   n1 = SetDescription!() -> (then, Description:Text) {is=title K2Node_FunctionEntry}
   > "SetText (Text)"(self=Get Description() -> (Description:Common Text Object Reference), InText=@n1)
  ]}
 SetExpirationTime {nodes [
   n1 = SetExpirationTime!() -> (then, Time:Date Time Structure) {is=title K2Node_FunctionEntry}
   > n2 = Branch(Condition=Get bHasExpirationTime!(self:Quest UI Category View Model Object Reference=Get QuestUICategoryVM() -> (QuestUICategoryVM:Quest UI Category View Model Object Reference)) -> (bHasExpirationTime:bool))
   > n3 = GetMcpRelativeUtcNow!(
    execute
    FortPC:Fort Player Controller Object Reference=GetFirstLocalFortPlayerController!(bIncludeNonSpectators:bool=true, bIncludeSpectators:bool=false) -> (ReturnValue:Fort Player Controller Object Reference)
   ) -> (then, ReturnValue:Date Time Structure)
   > n4 = Branch(Condition="Timespan > Timespan"(
     A:Timespan Structure=DateTime - DateTime!(A:Date Time Structure=@n1, B:Date Time Structure=@n3) -> (ReturnValue:Timespan Structure)
     B:Timespan Structure=MakeTimespan!(Days:int=0, Hours:int=0, Minutes:int=0, Seconds:int=1, Milliseconds:int=0) -> (ReturnValue:Timespan Structure)
    ) -> (ReturnValue:bool=~) {is=title K2Node_PromotableOperator})
   > SetVisibility(self=Get OverlayExpiration() -> (OverlayExpiration:Overlay Object Reference), InVisibility=SelfHitTestInvisible)
   > n5 = SetDateTimeValue!(
    execute
    self:Common Date Time Text Block Object Reference=Get ExpirationTime() -> (ExpirationTime:Athena Challenge Date Time Text Block Object Reference)
    InDateTime:Date Time Structure=@n1
    bShowAsCountdown:bool=true
    InRefreshDelay:float=1
   ) -> (then)
   SetVisibility(execute=[@n9.then, @n5.then], InVisibility=SelfHitTestInvisible)
   SetVisibility(execute=[@n4.else, @n7.then], self=Get OverlayExpiration(), InVisibility=Collapsed)
   n6 = Get CustomExpirationText!(self:Quest UI Category View Model Object Reference=Get QuestUICategoryVM()) -> (CustomExpirationText:Text)
   n7 = Branch(execute=@n2.else, Condition=TextIsEmpty(InText=@n6))
   SetVisibility(execute=@n7.else, self=Get OverlayExpiration(), InVisibility=SelfHitTestInvisible)
   n8 = Get ExpirationTime()
   > SetDateTimeValue(self=@n8, bShowAsCountdown=false, InRefreshDelay=0)
   > n9 = "SetText (Text)"(self=@n8, InText=@n6)
  ]}
 UpdateBackground {nodes [
   n1 = UpdateBackground!() -> (then, Background:Soft Object Reference) {is=title K2Node_FunctionEntry}
   > LoadBackground!(execute, self, Background:Soft Object Reference=@n1) -> (then) {is=title K2Node_CallFunction}
  ]}
]
variables [IsEntry {type=boolean, default=true}]
widgets {
 info {inheritedWidgetCount=0, parentClass=&CommonUserWidget, rootWidgetClass=&GridPanel, widgetCount=10}
 widgets [
  Root {widgetClassPath=&GridPanel, widget=Root, properties {columnFill [1], rowFill [0,1]}}
  AdditionalSpacer {
   widgetClassPath=&Spacer
   bIsVariable=true
   parent=Root
   widget=AdditionalSpacer
   properties {size=<0 48>}
   slot {class=GridSlot, properties {columnSpan=5}}
  }
  Description {
   widgetClassPath=&CommonTextBlock
   bIsVariable=true
   parent=Root
   widget=Description
   properties {
    autoWrapText=true
    bAutoCollapseWithEmptyText=true
    scrollStyle=&TextScrollStyle-Base_C
    text=Quest
    textTransformPolicy=ToUpper
    font {
     bForceMonospaced=false
     bMaterialIsStencil=false
     fontMaterial=null
     fontObject=&HeadingNow
     letterSpacing=0
     monospacedWidth=1
     size=21
     skewAmount=0
     typefaceFontName=86 Bold Italic
     outlineSettings {
      bApplyOutlineToDropShadows=false
      bMiteredCorners=false
      bSeparateFillAlpha=false
      outlineMaterial=null
      outlineSize=0
      outlineColor=<rgba 0 0 0 1>
     }
    }
    shadowOffset=<0 0>
   }
   slot {class=GridSlot, properties {layer=1, row=1, verticalAlignment=VAlign_Center, padding=<ltrb 20 0 0 0>}}
  }
  OverlayExpiration {
   widgetClassPath=&Overlay
   bIsVariable=true
   parent=Root
   widget=OverlayExpiration
   slot {
    class=GridSlot
    properties {column=1, horizontalAlignment=HAlign_Left, layer=1, row=1, verticalAlignment=VAlign_Center, padding=<ltrb 20 0 0 0>}
   }
  }
  Backplate {
   widgetClassPath=&WBP_UIKit_Backplate
   bIsVariable=true
   parent=OverlayExpiration
   widget=Backplate
   properties {brightness=Medium, cornerRadius=Circle, backplateSize=<0 26>}
   slot {class=OverlaySlot, properties {horizontalAlignment=HAlign_Fill, verticalAlignment=VAlign_Fill}}
  }
  HorizontalBox_0 {
   widgetClassPath=&HorizontalBox
   parent=OverlayExpiration
   widget=HorizontalBox_0
   slot {class=OverlaySlot, properties {verticalAlignment=VAlign_Center, padding=<ltrb 12 0 12 0>}}
  }
  Icon {
   widgetClassPath=&Image
   bIsVariable=true
   parent=HorizontalBox_0
   widget=Icon
   properties {brush {
     bIsDynamicallyLoaded=false
     drawAs=Image
     imageType=FullColor
     mirroring=NoMirror
     resourceName=None
     resourceObject=/FortUILibrary/Textures/MaterialInstances/MI_UIKit_Icon_UpdatePending
     tiling=NoTile
     imageSize=<16 16>
     margin=<ltrb 0 0 0 0>
     outlineSettings {
      bUseBrushTransparency=false
      roundingType=HalfHeightRadius
      width=0
      color {colorUseRule=UseColor_Specified, specifiedColor=<rgba 0 0 0 0>}
      cornerRadii=<0 0 0 0>
     }
     tintColor {colorUseRule=UseColor_Specified, specifiedColor=<rgba 1 1 1 1>}
     uVRegion {bIsValid=false, max=<0 0>, min=<0 0>}
    }}
   slot {class=HorizontalBoxSlot, properties {horizontalAlignment=HAlign_Left, verticalAlignment=VAlign_Center, padding=<ltrb 0 0 10 0>}}
  }
  ExpirationTime {
   widgetClassPath=/Script/FortniteUI.AthenaChallengeDateTimeTextBlock
   bIsVariable=true
   parent=HorizontalBox_0
   widget=ExpirationTime
   properties {
    bAllowSeconds=false
    bAutoCollapseWithEmptyText=true
    bIsScrollingEnabled=false
    completionText=EXPIRED
    labelFormat="Ends: {0}"
    text=Text Block
    font {
     bForceMonospaced=false
     bMaterialIsStencil=false
     fontMaterial=null
     fontObject=&HeadingNow
     letterSpacing=0
     monospacedWidth=1
     size=12.75
     skewAmount=0
     typefaceFontName=74 Regular
     outlineSettings {
      bApplyOutlineToDropShadows=false
      bMiteredCorners=false
      bSeparateFillAlpha=false
      outlineMaterial=null
      outlineSize=0
      outlineColor=<rgba 0 0 0 1>
     }
    }
    shadowOffset=<0 0>
   }
   slot {class=HorizontalBoxSlot}
  }
  Lazy_Background {
   widgetClassPath=&FortLazyImage
   bIsVariable=true
   parent=Root
   widget=Lazy_Background
   properties {
    brush {
     bIsDynamicallyLoaded=false
     drawAs=Image
     imageType=FullColor
     mirroring=NoMirror
     resourceName=None
     resourceObject=/FortniteCoreUI/Items/QuestItems/Categories/Art/MI_UI_CategoryLoading
     tiling=NoTile
     imageSize=<0 60>
     margin=<ltrb 0 0 0 0>
     outlineSettings {
      bUseBrushTransparency=false
      roundingType=HalfHeightRadius
      width=0
      color {colorUseRule=UseColor_Specified, specifiedColor=<rgba 0 0 0 0>}
      cornerRadii=<0 0 0 0>
     }
     tintColor {colorUseRule=UseColor_Specified, specifiedColor=<rgba 1 1 1 1>}
     uVRegion {bIsValid=false, max=<0 0>, min=<0 0>}
    }
    loadingBackgroundBrush {
     bIsDynamicallyLoaded=false
     drawAs=Image
     imageType=FullColor
     mirroring=NoMirror
     resourceName=None
     resourceObject=&MI_UI_CategoryLoading
     tiling=NoTile
     imageSize=<32 32>
     margin=<ltrb 0 0 0 0>
     outlineSettings {
      bUseBrushTransparency=false
      roundingType=HalfHeightRadius
      width=0
      color {colorUseRule=UseColor_Specified, specifiedColor=<rgba 0 0 0 0>}
      cornerRadii=<0 0 0 0>
     }
     tintColor {colorUseRule=UseColor_Specified, specifiedColor=<rgba 1 1 1 0.5>}
     uVRegion {bIsValid=false, max=<0 0>, min=<0 0>}
    }
   }
   slot {class=GridSlot, properties {columnSpan=5, row=1}}
  }
  Spacer_Product {
   widgetClassPath=&Spacer
   bIsVariable=true
   parent=Root
   widget=Spacer_Product
   properties {size=<165 25>}
   slot {class=GridSlot, properties {column=3, horizontalAlignment=HAlign_Right, row=1, verticalAlignment=VAlign_Center}}
  }
 ]
}
