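# UI material for a translucent quest entry background that switches between locked and granted looks.
# An IsDiscovered scalar blends a flat locked color toward a two-color linear gradient with an adjustable rotation.
# The same blend runs on the colors' alpha values, which a rounded-corner mask then multiplies to shape the background.
# Corner radii are parameterized, with a parameter for treating them as pixels.
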Material /QuestUI/QuestScreen/V3/M_Quest_CustomEntryBackground
graph {
 properties {bAutomaticallySetUsageInEditor=false, blendMode=BLEND_Translucent, materialDomain=MD_UI, usedShadingModels=Default Lit}
 nodes [
  n1 = VectorParameter() {parameterName=ColorLocked, defaultValue=<rgba 0.0046 0.0094 0.0522 0.7>}
  n2 = VectorParameter() {parameterName=ColorGranted_02, defaultValue=<rgba 0.2016 0.5395 1 1>}
  n3 = VectorParameter() {parameterName=ColorGranted_01, defaultValue=<rgba 0 0.1878 1 1>}
  n4 = MaterialFunctionCall(RotationAngle=ScalarParameter() {defaultValue=0.25, parameterName=Rotation}) {materialFunction=&MF_UI_LinearGradient}
  n5 = ScalarParameter() {parameterName=IsDiscovered}
  LinearInterpolate(A=@n1.RGB, B=LinearInterpolate(A=@n2.RGB, B=@n3.RGB, Alpha=@n4.Result), Alpha=@n5)
  Multiply(
   A=LinearInterpolate(A=@n1.A, B=LinearInterpolate(A=@n2.A, B=@n3.A, Alpha=@n4.Result), Alpha=@n5)
   B=MaterialFunctionCall(
    Corner Radius=VectorParameter() {parameterName=CornerRadiusV4, defaultValue=<rgba 20 20 20 20>}.RGBA
    Radius in Pixels=ScalarParameter() {defaultValue=1, parameterName=RadiusInPixels}
   ) {materialFunction=&MF_UI_CoreUI_RoundedMask}.Shape Mask
  )
 ]
}
