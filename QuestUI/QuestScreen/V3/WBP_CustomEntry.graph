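# Quest screen tile widget, derived from WBP_QuestTile_Base, that displays a custom entry's icon over a background reflecting its discovered state.
# SetCustomEntryVM releases the previous view model, binds to discovery changes, then drives a background material parameter and lazy-loads the small icon.
# Selection toggles an outline image, while a registered UIKit outline block uses custom timing.
# Parent hover logic is skipped on gamepad, and design-time preview shows a placeholder icon.
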WidgetBlueprint /QuestUI/QuestScreen/V3/WBP_CustomEntry
animations [ShowSelection {
  binding_count=1
  length_seconds=0.1
  object=&self:ShowSelection
  sequence {bindings [SelectedStateOutline {
     id=#1
     possessed_class=Image
     tracks [
      {
       type=MovieSceneFloatTrack
       display_name=Render Opacity
       sections [{type=MovieSceneFloatSection, end_frame=2, start_frame=0, channels [None {type=MovieSceneScriptingFloatChannel, keys [[0,0],[2,1]]}]}]
      }
      {
       type=MovieSceneByteTrack
       display_name=Visibility
       sections [{type=MovieSceneByteSection, end_frame=2, start_frame=0, channels [None {type=MovieSceneScriptingByteChannel, keys [[0,1],[0.6667,4],[2,4]]}]}]
      }
     ]
    }]}
 }]
graphs [
 EventGraph {nodes [
   n1 = LoadIcon!() -> (OutputDelegate, then, Icon:Soft Object Reference) {is=title K2Node_CustomEvent}
   > n2 = Async Load Asset(Asset=@n1.Icon)
   n3 = Cast To Texture!(execute=@n2.Completed, Object=@n2) -> (then, CastFailed, AsTexture:Texture Object Reference) {is=title K2Node_DynamicCast}
   > n4 = GetDynamicMaterial(self=Get LazyImage_Icon() -> (LazyImage_Icon:Fort Lazy Image Object Reference))
   > SetTextureParameterValue(self=@n4, ParameterName=Texture, Value=@n3)
   Event On Hovered()
   n5 = Get CommonInputSubsystem()
   > n6 = Branch(Condition="Equal (Enum)"(A=GetCurrentInputType(self=@n5), B=Gamepad))
   Parent: On Hovered(execute=@n6.else) {is=title K2Node_CallParentFunction}
   n7 = Event PreConstruct()
   > SetIsDiscovered!(execute, self, IsDiscovered:bool=false) -> (then)
   > n8 = Branch(Condition=@n7.IsDesignTime)
   > SetBrushFromTexture!(
    execute
    self:Image Object Reference=Get LazyImage_Icon()
    Texture:Texture 2D Object Reference=/Game/2dAssets/Emoji/Season4/Crab
    bMatchSize:bool=false
   ) -> (then)
   RegisterBlocks(execute=@n8.else)
   > SetBlocksTiming!(
    execute
    self:UIKit Modular Button Object Reference
    Timing:"UIKit Block Timing Structure (by ref)"=Get BlockTiming() -> (BlockTiming:UIKit Block Timing Structure)
   ) -> (then)
   Event On Selected()
   > SelectionChangedVisualization!(execute, self, IsSelected:bool=true) -> (then)
   Event On Deselected()
   > Parent: On Deselected() {is=title K2Node_CallParentFunction}
   > SelectionChangedVisualization(IsSelected=false)
  ]}
 RegisterBlocks {nodes [
   RegisterBlocks() {is=title K2Node_FunctionEntry}
   > RegisterModularBlock!(
    execute
    self:UIKit Modular Button Object Reference
    Block:UIKit Block Object Reference=Get Block_Outline() -> (Block_Outline:WBP UIKit Block Outline Object Reference)
   ) -> (then)
  ]}
 ReleaseEntry {nodes [
   ReleaseEntry() {is=title K2Node_FunctionEntry}
   n1 = Get BaseEntryVM() -> (BaseEntryVM:Quest UI Custom Entry View Model Object Reference)
   > Is Valid(InputObject=@n1)
   > Unbind Event from On Is Discovered Changed(self:Quest UI Custom Entry View Model Object Reference=@n1, Delegate=Create Event()) {is=title K2Node_RemoveDelegate}
   > Set BaseEntryVM(BaseEntryVM:Quest UI Custom Entry View Model Object Reference) -> (Output_Get:Quest UI Custom Entry View Model Object Reference)
  ]}
 SelectionChangedVisualization {nodes [
   n1 = SelectionChangedVisualization!() -> (then, IsSelected:bool=~) {is=title K2Node_FunctionEntry}
   > SetVisibility(
    self=Get SelectedStateOutline() -> (SelectedStateOutline:Image Object Reference)
    InVisibility=Select(Option 0=Collapsed, Option 1=SelfHitTestInvisible, Index=@n1)
   )
  ]}
 SetCustomEntryVM {nodes [
   n1 = SetCustomEntryVM!() -> (then, CustomEntryVM:Quest UI Custom Entry View Model Object Reference) {is=title K2Node_FunctionEntry}
   > ReleaseEntry() {is=title K2Node_CallFunction}
   > n2 = Set BaseEntryVM(BaseEntryVM=@n1)
   > Is Valid(InputObject=@n2)
   > n3 = Sequence!(execute) -> (then_0, then_1, then_2:Exec)
   SetEntryIcon!(
    execute=@n3.then_2
    self
    Icon:Soft Object Reference=Get SmallIcon!(self:Quest UI Custom Entry View Model Object Reference=@n2) -> (SmallIcon:Soft Object Reference)
   ) -> (then)
   SetIsDiscovered(execute=@n3.then_1, IsDiscovered=Get bIsDiscovered!(self:Quest UI Custom Entry View Model Object Reference=@n2) -> (bIsDiscovered:bool))
   Bind Event to On Is Discovered Changed(execute=@n3.then_0, self:Quest UI Custom Entry View Model Object Reference=@n2, Delegate=Create Event()) {is=title K2Node_AddDelegate}
  ]}
 SetEntryIcon {nodes [
   n1 = SetEntryIcon!() -> (then, Icon:Soft Object Reference) {is=title K2Node_FunctionEntry}
   > SetBrushFromLazyDisplayAsset!(
    execute
    self:Common Lazy Image Object Reference=Get LazyImage_Icon()
    LazyObject:"Soft Object Reference (by ref)"=@n1
    bMatchTextureSize:bool=false
   ) -> (then)
  ]}
 SetIsDiscovered {nodes [
   n1 = SetIsDiscovered!() -> (then, IsDiscovered:bool=~) {is=title K2Node_FunctionEntry}
   > n2 = GetDynamicMaterial(self=Get Image_Background() -> (Image_Background:Image Object Reference))
   > SetScalarParameterValue(self=@n2, ParameterName=IsDiscovered, Value="To Float (Boolean)"(InBool=@n1))
  ]}
]
variables [
 BaseEntryVM {type=QuestUICustomEntryVM, default=null}
 BlockTiming {
  type=UIKitBlockTiming
  default {
   deselectingFocused=0.2
   deselectingIdle=0.2
   disabling=0
   enabling=0
   focusing=0.2
   hovering=0.2
   pressing=0.2
   releasing=0.2
   selecting=0.2
   unfocusing=0.2
   unhovering=0.2
  }
 }
]
widgets {
 info {
  inheritedWidgetCount=1
  parentClass=/QuestUI/QuestScreen/QuestTile/WBP_QuestTile_Base.WBP_QuestTile_Base_C
  rootWidgetClass=&Overlay
  widgetCount=7
 }
 widgets [
  Overlay_19 {widgetClassPath=&Overlay, widget=Overlay_19}
  Image_Background {
   widgetClassPath=&Image
   bIsVariable=true
   parent=Overlay_19
   widget=Image_Background
   properties {brush {
     bIsDynamicallyLoaded=false
     drawAs=Image
     imageType=FullColor
     mirroring=NoMirror
     resourceName=None
     resourceObject=&/M_Quest_CustomEntryBackground
     tiling=NoTile
     imageSize=<154 183>
     margin=<ltrb 0 0 0 0>
     outlineSettings {
      bUseBrushTransparency=false
      roundingType=HalfHeightRadius
      width=0
      color {colorUseRule=UseColor_Specified, specifiedColor=<rgba 0 0 0 0>}
      cornerRadii=<0 0 0 0>
     }
     tintColor {colorUseRule=UseColor_Specified, specifiedColor=<rgba 1 1 1 1>}
     uVRegion {bIsValid=false, max=<0 0>, min=<0 0>}
    }}
   slot {class=OverlaySlot, properties {horizontalAlignment=HAlign_Fill, verticalAlignment=VAlign_Fill}}
  }
  LazyImage_Icon {
   widgetClassPath=&FortLazyImage
   bIsVariable=true
   parent=Overlay_19
   widget=LazyImage_Icon
   properties {
    brush {
     bIsDynamicallyLoaded=false
     drawAs=Image
     imageType=FullColor
     mirroring=NoMirror
     resourceName=None
     resourceObject=&Crab
     tiling=NoTile
     imageSize=<154 183>
     margin=<ltrb 0 0 0 0>
     outlineSettings {
      bUseBrushTransparency=false
      roundingType=HalfHeightRadius
      width=0
      color {colorUseRule=UseColor_Specified, specifiedColor=<rgba 0 0 0 0>}
      cornerRadii=<0 0 0 0>
     }
     tintColor {colorUseRule=UseColor_Specified, specifiedColor=<rgba 1 1 1 1>}
     uVRegion {bIsValid=false, max=<0 0>, min=<0 0>}
    }
    loadingBackgroundBrush {
     bIsDynamicallyLoaded=false
     drawAs=NoDrawType
     imageType=NoImage
     mirroring=NoMirror
     resourceName=None
     resourceObject=null
     tiling=NoTile
     imageSize=<31 32>
     margin=<ltrb 0 0 0 0>
     outlineSettings {
      bUseBrushTransparency=false
      roundingType=HalfHeightRadius
      width=0
      color {colorUseRule=UseColor_Specified, specifiedColor=<rgba 0 0 0 0>}
      cornerRadii=<0 0 0 0>
     }
     tintColor {colorUseRule=UseColor_Specified, specifiedColor=<rgba 1 1 1 1>}
     uVRegion {bIsValid=false, max=<0 0>, min=<0 0>}
    }
    renderTransform {angle=0, scale=<0.85 0.85>, shear=<0 0>, translation=<0 0>}
   }
   slot {class=OverlaySlot, properties {horizontalAlignment=HAlign_Center, verticalAlignment=VAlign_Center}}
  }
  Slot_Child {
   widgetClassPath=&NamedSlot
   bIsVariable=true
   parent=Overlay_19
   widget=Slot_Child
   slot {class=OverlaySlot, properties {horizontalAlignment=HAlign_Fill, verticalAlignment=VAlign_Fill}}
  }
  SelectedStateOutline {
   widgetClassPath=&Image
   bIsVariable=true
   parent=Overlay_19
   widget=SelectedStateOutline
   properties {
    visibility=Collapsed
    brush {
     bIsDynamicallyLoaded=false
     drawAs=Image
     imageType=FullColor
     mirroring=NoMirror
     resourceName=None
     resourceObject=/QuestUI/QuestScreen/QuestTile/Materials/MI_UI_Outline_Selected
     tiling=NoTile
     imageSize=<32 32>
     margin=<ltrb 0 0 0 0>
     outlineSettings {
      bUseBrushTransparency=false
      roundingType=HalfHeightRadius
      width=0
      color {colorUseRule=UseColor_Specified, specifiedColor=<rgba 0 0 0 0>}
      cornerRadii=<0 0 0 0>
     }
     tintColor {colorUseRule=UseColor_Specified, specifiedColor=<rgba 1 1 1 1>}
     uVRegion {bIsValid=false, max=<0 0>, min=<0 0>}
    }
   }
   slot {class=OverlaySlot, properties {horizontalAlignment=HAlign_Fill, verticalAlignment=VAlign_Fill}}
  }
  Block_Outline {
   widgetClassPath=/FortUILibrary/UIKit/Buttons/Blocks/WBP_UIKit_Block_Outline
   bIsVariable=true
   parent=Overlay_19
   widget=Block_Outline
   properties {bottomLeftCornerRadius=24, bottomRightCornerRadius=24, radiusInPixels=true, renderOpacity=0, topLeftCornerRadius=24, topRightCornerRadius=24}
   slot {class=OverlaySlot, properties {horizontalAlignment=HAlign_Fill, verticalAlignment=VAlign_Fill}}
  }
  InputActionWidget {widgetClassPath=&CommonActionWidget, bInherited=true, widget=null}
 ]
}
