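# Quest screen row widget that shows a row view model's ordered content as selectable entries in a Dynamic Entry Box.
# On item set it rebuilds entries using WBP_CustomEntry or the row's override class, assigning view models directly or by class.
# Selecting an entry clears the previous one and reports its column and view model through SetColumnSelected.
# SetActiveColumn selects by clamped index, and gamepad focus targets the selected entry.
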WidgetBlueprint /QuestUI/QuestScreen/V3/WBP_CustomRow
graphs [
 BP_GetDesiredFocusTarget {nodes [
   Get Desired Focus Target()
   n1 = GetCurrentInputType(self=Get CommonInputSubsystem())
   > Return Node!(
    execute
    ReturnValue:Widget Object Reference=Select(
     Option 0:Widget Object Reference
     Option 1:Widget Object Reference=Get SelectedEntry() -> (SelectedEntry:Common Button Base Object Reference)
     Index="Equal (Enum)"(A=@n1, B=Gamepad)
    ) -> (ReturnValue:Widget Object Reference)
   )
  ]}
 ClearActiveSelection {nodes [
   ClearActiveSelection() {is=title K2Node_FunctionEntry}
   n1 = Get SelectedEntry()
   > Is Valid(InputObject=@n1)
   > ClearSelection(self:Common Button Base Object Reference=@n1)
   > Set SelectedEntry(SelectedEntry:Common Button Base Object Reference) -> (Output_Get:Common Button Base Object Reference)
  ]}
 EventGraph {nodes [
   n1 = Event On Item Selection Changed!() -> (OutputDelegate, then, bIsSelected:bool) {is=title K2Node_Event}
   > n2 = Branch(Condition=@n1.bIsSelected)
   ClearActiveSelection(execute=@n2.else) {is=title K2Node_CallFunction}
   Event On Entry Released() {is=title K2Node_Event}
   > Set QuestUICustomRowVM(QuestUICustomRowVM:Quest UI Custom Row View Model Object Reference) -> (Output_Get:Quest UI Custom Row View Model Object Reference)
   n3 = Event SetActiveColumn!() -> (OutputDelegate, then, ActiveColumn:int=0) {is=title K2Node_Event}
   n4 = Get EntryBox_Entries() -> (EntryBox_Entries:Dynamic Entry Box Object Reference)
   n5 = int - int(A=GetNumEntries!(self:Dynamic Entry Box Base Object Reference=@n4) -> (ReturnValue:int=0))
   > n6 = Cast To CommonButtonBase!(
    execute
    Object="Get (a copy)"(
     Array:Array of User Widget Object References=GetAllEntries!(self:Dynamic Entry Box Base Object Reference=@n4) -> (ReturnValue:Array of User Widget Object References)
     Dimension 1="Clamp (Integer)"(Value=@n3.ActiveColumn, Min=0, Max=@n5) -> (ReturnValue=0)
    ) -> (Output:User Widget Object Reference)
   ) -> (then, CastFailed, AsCommon Button Base:Common Button Base Object Reference) {is=title K2Node_DynamicCast}
   > SetIsSelected!(execute, self:Common Button Base Object Reference=@n6, InSelected:bool=true, bGiveClickFeedback:bool=true) -> (then)
   n7 = Event On Poly View Item Set!() -> (OutputDelegate, then, Item:Object Reference, SourceIndex:int=0) {is=title K2Node_Event}
   > ClearActiveSelection()
   > Reset!(execute, self:Dynamic Entry Box Object Reference=Get EntryBox_Entries(), bDeleteWidgets:bool=false) -> (then)
   > n8 = Cast To QuestUICustomRowVM!(execute, Object=@n7.Item) -> (then, CastFailed, AsQuest UI Custom Row View Model:Quest UI Custom Row View Model Object Reference) {is=title K2Node_DynamicCast}
   > Set QuestUICustomRowVM(QuestUICustomRowVM=@n8)
   > n9 = For Each Loop(Array:Array of Quest UI Base Content View Model Object References=Get OrderedContentVMs!(self:Quest UI Custom Row View Model Object Reference=@n8) -> (OrderedContentVMs:Array of Quest UI Base Content View Model Object References)) -> (Array Element:Quest UI Base Content View Model Object Reference)
   n10 = Get CustomEntryWidgetOverride!(self:Quest UI Custom Row View Model Object Reference=@n8) -> (CustomEntryWidgetOverride:User Widget Class Reference)
   > n11 = Create Entry of Class!(
    execute
    self:Dynamic Entry Box Object Reference=Get EntryBox_Entries()
    EntryClass:User Widget Class Reference=Select(
     Option 0:User Widget Class Reference=&/WBP_CustomEntry.WBP_CustomEntry_C
     Option 1:User Widget Class Reference=@n10
     Index=IsValidClass!(Class:Object Class Reference=@n10) -> (ReturnValue:bool)
    ) -> (ReturnValue:User Widget Class Reference)
   ) -> (then, ReturnValue:User Widget Object Reference)
   > n12 = Cast To WBP_CustomEntry!(execute, Object=@n11) -> (then, CastFailed, AsWBP Custom Entry:WBP Custom Entry Object Reference) {is=title K2Node_DynamicCast}
   > n13 = Cast To QuestUICustomEntryVM!(execute, Object=@n9) -> (then, CastFailed, AsQuest UI Custom Entry View Model:Quest UI Custom Entry View Model Object Reference) {is=title K2Node_DynamicCast}
   > n14 = SetCustomEntryVM!(execute, self:WBP Custom Entry Object Reference=@n12, CustomEntryVM:Quest UI Custom Entry View Model Object Reference=@n13) -> (then)
   n15 = Cast To CommonButtonBase(execute=[@n14.then, @n16.then], Object=@n11)
   > Bind Event to On Selected(self:Common Button Base Object Reference=@n15, Delegate=Create Event()) {is=title K2Node_AddDelegate}
   n16 = SetViewModelByClass!(execute=[@n13.CastFailed, @n12.CastFailed], Widget:User Widget Object Reference=@n11, ViewModel:Notify Field Value Changed Interface=@n9) -> (then)
  ]}
 GetRowVM {nodes [
   GetRowVM() {is=title K2Node_FunctionEntry}
   > Return Node(ReturnValue:Quest UI Base Content View Model Object Reference=Get QuestUICustomRowVM() -> (QuestUICustomRowVM:Quest UI Custom Row View Model Object Reference))
  ]}
 OnEntrySelected {nodes [
   n1 = OnEntrySelected!() -> (then, Button:Common Button Base Object Reference) {is=title K2Node_FunctionEntry}
   > ClearActiveSelection()
   > n2 = Set SelectedEntry(SelectedEntry=@n1)
   > SetIsSelected(self=@n2, InSelected=true, bGiveClickFeedback=false)
   n3 = Find Item!(
    TargetArray:Array of User Widget Object References=GetAllEntries(self=Get EntryBox_Entries())
    ItemToFind:"User Widget Object Reference (by ref)"=@n2
   ) -> (ReturnValue)
   > Branch(Condition="integer >= integer"(A=@n3))
   n4 = Get QuestUICustomRowVM()
   > SetColumnSelected!(
    execute
    self:Quest UIRow Widget Object Reference
    SelectedVM:Quest UI Base Content View Model Object Reference="Get (a copy)"(Array:Array of Quest UI Base Content View Model Object References=Get OrderedContentVMs(self=@n4), Dimension 1=@n3) -> (Output:Quest UI Base Content View Model Object Reference)
    SelectedColumn:int=@n3
   ) -> (then)
  ]}
]
variables [SelectedEntry {type=CommonButtonBase, default=null}]
widgets {
 info {inheritedWidgetCount=0, parentClass=/Script/QuestUI.QuestUIRowWidget, rootWidgetClass=&DynamicEntryBox, widgetCount=1}
 widgets [EntryBox_Entries {
   widgetClassPath=&DynamicEntryBox
   bIsVariable=true
   widget=EntryBox_Entries
   properties {entryWidgetClass=&WBP_CustomEntry_C, numDesignerPreviewEntries=5, entrySpacing=<25 0>}
  }]
}
