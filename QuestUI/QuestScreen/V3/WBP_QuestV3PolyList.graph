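# Polymorphic quest list for the V3 quest screen, built on FortMultiSourcePolymorphicListView with WBP_QuestTile as the default entry.
# Entry initialization binds quest tile and horizontal goals-progress events, re-broadcast with scroll events through its own dispatchers; release unbinds them.
# CheckItemSelectability casts by view model type, allowing quests, punch cards and messages but rejecting headers, separators, locked, compatibility and category items.
# SetIsMoveMode updates displayed tiles; new tiles inherit it, with draggability set by source index.
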WidgetBlueprint /QuestUI/QuestScreen/V3/WBP_QuestV3PolyList
graphs [
 CheckItemSelectability {nodes [
   n1 = CheckItemSelectability!() -> (then, InputObject:Object Reference) {is=title K2Node_FunctionEntry}
   > n2 = Set Item(Item:Object Reference=@n1) -> (Output_Get:Object Reference)
   > Is Valid(InputObject=@n2)
   > n3 = Cast To QuestUIHeaderVM!(execute, Object=Get Item() -> (Item:Object Reference)) -> (then, CastFailed, AsQuest UI Header View Model:Quest UI Header View Model Object Reference) {is=title K2Node_DynamicCast}
   > Return Node!(execute, ReturnValue:bool=false)
   n4 = Cast To MVVM_PunchCardQuest!(execute=@n3.CastFailed, Object=Get Item()) -> (then, CastFailed, AsMVVM Punch Card Quest:MVVM Punch Card Quest Object Reference) {is=title K2Node_DynamicCast}
   > Return Node(ReturnValue=true)
   n5 = Cast To QuestUIQuestVM!(execute=@n4.CastFailed, Object=Get Item()) -> (then, CastFailed, AsQuest UI Quest View Model:Quest UI Quest View Model Object Reference) {is=title K2Node_DynamicCast}
   > Return Node(ReturnValue=true)
   n6 = Cast To QuestUILockedQuestVM!(execute=@n5.CastFailed, Object=Get Item()) -> (then, CastFailed, AsQuest UI Locked Quest View Model:Quest UI Locked Quest View Model Object Reference) {is=title K2Node_DynamicCast}
   > Return Node(ReturnValue=false)
   n7 = Cast To QuestUICompatibilityVM!(execute=@n6.CastFailed, Object=Get Item()) -> (then, CastFailed, AsQuest UI Compatibility View Model:Quest UI Compatibility View Model Object Reference) {is=title K2Node_DynamicCast}
   > Return Node(ReturnValue=false)
   n8 = Cast To QuestUIMessageVM!(execute=@n7.CastFailed, Object=Get Item()) -> (then, CastFailed, AsQuest UI Message View Model:Quest UI Message View Model Object Reference) {is=title K2Node_DynamicCast}
   > Return Node(ReturnValue=true)
   n9 = Cast To QuestUIHorizontalPunchCardVM!(execute=@n8.CastFailed, Object=Get Item()) -> (then, CastFailed, AsQuest UI Horizontal Punch Card View Model:Quest UI Horizontal Punch Card View Model Object Reference) {is=title K2Node_DynamicCast}
   > Return Node(ReturnValue=true)
   n10 = Cast To QuestUIGoalHeaderVM!(execute=@n9.CastFailed, Object=Get Item()) -> (then, CastFailed, AsQuest UIGoal Header VM:Quest UIGoal Header VM Object Reference) {is=title K2Node_DynamicCast}
   > Return Node(ReturnValue=false)
   n11 = Cast To VM_QuestUIPinnedQuestSeparator!(execute=@n10.CastFailed, Object=Get Item()) -> (then, CastFailed, AsVM Quest UIPinned Quest Separator:VM Quest UIPinned Quest Separator Object Reference) {is=title K2Node_DynamicCast}
   > Return Node(ReturnValue=false)
   n12 = Cast To QuestUICategoryVM!(execute=@n11.CastFailed, Object=Get Item()) -> (then, CastFailed, AsQuest UI Category View Model:Quest UI Category View Model Object Reference) {is=title K2Node_DynamicCast}
   Return Node(execute=[@n12.CastFailed, @n12.then], ReturnValue=false)
  ]}
 EventGraph {nodes [
   n1 = "On Entry Initialized (ListView)"!() -> (OutputDelegate, then, Item:Object Reference, Widget:User Widget Object Reference) {is=title K2Node_ComponentBoundEvent}
   > n2 = Cast To WBP_QuestTile!(execute, Object=@n1.Widget) -> (then, CastFailed, AsWBP Quest Tile:WBP Quest Tile Object Reference) {is=title K2Node_DynamicCast}
   > SetIsMoveMode!(execute, self:WBP Quest Tile Object Reference=@n2, IsMoveMode:bool=Get IsMoveMode() -> (IsMoveMode:bool), NeedsToAnimate:bool=false) -> (then)
   > SetIsDraggable!(
    execute
    self:WBP Quest Tile Object Reference=@n2
    IsDraggable:bool="Equal (Integer)"(A=GetSourceIndex!(Item:Object Reference=@n1.Item) -> (ReturnValue:int=0))
   ) -> (then)
   > Bind Event to On Quest Clicked(self:WBP Quest Tile Object Reference=@n2, Delegate=Create Event()) {is=title K2Node_AddDelegate}
   > Bind Event to On Quest Double Clicked(self:WBP Quest Tile Object Reference=@n2, Delegate=Create Event()) {is=title K2Node_AddDelegate}
   > Bind Event to On Quest Navigate(self:WBP Quest Tile Object Reference=@n2, Delegate=Create Event()) {is=title K2Node_AddDelegate}
   > Bind Event to On Quest Hover(self:WBP Quest Tile Object Reference=@n2, Delegate=Create Event()) {is=title K2Node_AddDelegate}
   n3 = Cast To WBP_GoalsProgress_Horizontal!(execute=@n2.CastFailed, Object=@n1.Widget) -> (then, CastFailed, AsWBP Goals Progress Horizontal:WBP Goals Progress Horizontal Object Reference) {is=title K2Node_DynamicCast}
   > Bind Event to On Goals Navigate(self:WBP Goals Progress Horizontal Object Reference=@n3, Delegate=Create Event()) {is=title K2Node_AddDelegate}
   > Assign On Goals Double Click!(execute:Exec, self:WBP Goals Progress Horizontal Object Reference=@n3, Delegate:"Delegate (by ref)"=Create Event()) -> (then:Exec) {is=title K2Node_AssignDelegate}
   > Bind Event to On Selected Index Changed(self:WBP Goals Progress Horizontal Object Reference=@n3, Delegate=Create Event()) {is=title K2Node_AddDelegate}
   > Bind Event to On Reward View Changed(self:WBP Goals Progress Horizontal Object Reference=@n3, Delegate=Create Event()) {is=title K2Node_AddDelegate}
   n4 = "On Entry Released (ListView)"!() -> (OutputDelegate, then, Widget:User Widget Object Reference) {is=title K2Node_ComponentBoundEvent}
   > n5 = Cast To WBP_QuestTile(Object=@n4.Widget)
   > Unbind all Events from On Quest Clicked!(execute:Exec, self:WBP Quest Tile Object Reference=@n5) -> (then:Exec) {is=title K2Node_ClearDelegate}
   > Unbind all Events from On Quest Double Clicked!(execute:Exec, self:WBP Quest Tile Object Reference=@n5) -> (then:Exec) {is=title K2Node_ClearDelegate}
   > Unbind all Events from On Quest Navigate!(execute:Exec, self:WBP Quest Tile Object Reference=@n5) -> (then:Exec) {is=title K2Node_ClearDelegate}
   > Unbind all Events from On Quest Hover!(execute:Exec, self:WBP Quest Tile Object Reference=@n5) -> (then:Exec) {is=title K2Node_ClearDelegate}
   > n6 = Cast To WBP_GoalsProgress_Horizontal(Object=@n4.Widget)
   > Unbind all Events from On Goals Double Click!(execute:Exec, self:WBP Goals Progress Horizontal Object Reference=@n6) -> (then:Exec) {is=title K2Node_ClearDelegate}
   > Unbind all Events from On Goals Navigate!(execute:Exec, self:WBP Goals Progress Horizontal Object Reference=@n6) -> (then:Exec) {is=title K2Node_ClearDelegate}
   > Unbind all Events from On Selected Index Changed!(execute:Exec, self:WBP Goals Progress Horizontal Object Reference=@n6) -> (then:Exec) {is=title K2Node_ClearDelegate}
   > Unbind all Events from On Reward View Changed!(execute:Exec, self:WBP Goals Progress Horizontal Object Reference=@n6) -> (then:Exec) {is=title K2Node_ClearDelegate}
   n7 = "On List View Scrolled (ListView)"!() -> (OutputDelegate, then, ItemOffset:float, DistanceRemaining:float) {is=title K2Node_ComponentBoundEvent}
   > Call On List View Scrolled!(execute:Exec, self:WBP Quest V3Poly List Object Reference, ItemOffset:double=@n7.ItemOffset, DistanceRemaining:double=@n7.DistanceRemaining) -> (then:Exec) {is=title K2Node_CallDelegate}
   "On List View Finished Scrolling (ListView)"() {is=title K2Node_ComponentBoundEvent}
   > Call On List View Finished Scrolling!(execute:Exec, self:WBP Quest V3Poly List Object Reference) -> (then:Exec) {is=title K2Node_CallDelegate}
   Event PreConstruct()
   > SetScrollbarVisibility!(
    execute
    self:List View Base Object Reference=Get ListView() -> (ListView:Common List View Object Reference)
    InVisibility:ESlateVisibility Enum=Collapsed
   ) -> (then)
  ]}
 HandleGoalsDoubleClick {nodes [
   n1 = HandleGoalsDoubleClick!() -> (then, GoalsQuest:WBP Goals Progress Horizontal Object Reference) {is=title K2Node_FunctionEntry}
   > Call On Goals Double Clicked!(execute:Exec, self:WBP Quest V3Poly List Object Reference, GoalsQuestsHorizontal:WBP Goals Progress Horizontal Object Reference=@n1) -> (then:Exec) {is=title K2Node_CallDelegate}
  ]}
 HandlePunchCardRewardViewChanged {nodes [
   n1 = HandlePunchCardRewardViewChanged!() -> (then, InItemView:bool=~) {is=title K2Node_FunctionEntry}
   > Call On Punch Card Reward View Changed!(execute:Exec, self:WBP Quest V3Poly List Object Reference, InItemView:bool=@n1) -> (then:Exec) {is=title K2Node_CallDelegate}
  ]}
 HandlePunchCardSelectedIndexChanged {nodes [
   n1 = HandlePunchCardSelectedIndexChanged!() -> (then, SelectedIndex:int) {is=title K2Node_FunctionEntry}
   > Call On Goals Selected Index Changed!(execute:Exec, self:WBP Quest V3Poly List Object Reference, SelectedIndex:int=@n1) -> (then:Exec) {is=title K2Node_CallDelegate}
  ]}
 HandleQuestClicked {nodes [
   n1 = HandleQuestClicked!() -> (then, DroppedQuest:WBP Quest Tile Object Reference, WasDropped:bool=~) {is=title K2Node_FunctionEntry}
   > Call On Quest Clicked!(execute:Exec, self:WBP Quest V3Poly List Object Reference, ClickedQuest:WBP Quest Tile Object Reference=@n1, WasDropped:bool=@n1.WasDropped) -> (then:Exec) {is=title K2Node_CallDelegate}
  ]}
 HandleQuestDoubleClicked {nodes [
   n1 = HandleQuestDoubleClicked!() -> (then, QuestTile:WBP Quest Tile Object Reference) {is=title K2Node_FunctionEntry}
   > Call On Quest Double Clicked!(execute:Exec, self:WBP Quest V3Poly List Object Reference, QuestTile:WBP Quest Tile Object Reference=@n1) -> (then:Exec) {is=title K2Node_CallDelegate}
  ]}
 HandleQuestHover {nodes [
   n1 = HandleQuestHover!() -> (then, Quest Tile:WBP Quest Tile Object Reference) {is=title K2Node_FunctionEntry}
   > Call On Quest Hover!(execute:Exec, self:WBP Quest V3Poly List Object Reference, QuestTile:WBP Quest Tile Object Reference=@n1) -> (then:Exec) {is=title K2Node_CallDelegate}
  ]}
 HandleQuestNavigate {nodes [
   n1 = HandleQuestNavigate!() -> (then, QuestTile:User Widget Object Reference, Direction:EUINavigation Enum) {is=title K2Node_FunctionEntry}
   > Call On Quest Navigate!(
    execute:Exec
    self:WBP Quest V3Poly List Object Reference
    QuestTile:User Widget Object Reference=@n1
    Direction:EUINavigation Enum=@n1.Direction
   ) -> (then:Exec) {is=title K2Node_CallDelegate}
  ]}
 OnGoalsDoubleClicked {nodes [OnGoalsDoubleClicked!() -> (then, GoalsQuestsHorizontal:WBP Goals Progress Horizontal Object Reference) {is=title K2Node_FunctionEntry}]}
 OnGoalsSelectedIndexChanged {nodes [OnGoalsSelectedIndexChanged!() -> (then, SelectedIndex:int) {is=title K2Node_FunctionEntry}]}
 OnListViewFinishedScrolling {nodes [OnListViewFinishedScrolling() {is=title K2Node_FunctionEntry}]}
 OnListViewScrolled {nodes [OnListViewScrolled!() -> (then, ItemOffset:double=~, DistanceRemaining:double=~) {is=title K2Node_FunctionEntry}]}
 OnPunchCardRewardViewChanged {nodes [OnPunchCardRewardViewChanged!() -> (then, InItemView:bool=~) {is=title K2Node_FunctionEntry}]}
 OnQuestClicked {nodes [OnQuestClicked!() -> (then, ClickedQuest:WBP Quest Tile Object Reference, WasDropped:bool=~) {is=title K2Node_FunctionEntry}]}
 OnQuestDoubleClicked {nodes [OnQuestDoubleClicked!() -> (then, QuestTile:WBP Quest Tile Object Reference) {is=title K2Node_FunctionEntry}]}
 OnQuestHover {nodes [OnQuestHover!() -> (then, QuestTile:WBP Quest Tile Object Reference) {is=title K2Node_FunctionEntry}]}
 OnQuestNavigate {nodes [OnQuestNavigate!() -> (then, QuestTile:User Widget Object Reference, Direction:EUINavigation Enum) {is=title K2Node_FunctionEntry}]}
 SetIsMoveMode {nodes [
   n1 = SetIsMoveMode!() -> (then, IsMoveMode:bool=~) {is=title K2Node_FunctionEntry}
   > Set IsMoveMode(IsMoveMode:bool=@n1) -> (Output_Get:bool)
   > n2 = For Each Loop(Array:Array of Object References=GetDisplayedEntryWidgets!(self:List View Base Object Reference=Get ListView()) -> (ReturnValue:Array of User Widget Object References)) -> (Array Element:Object Reference)
   > n3 = Cast To WBP_QuestTile(Object=@n2)
   > SetIsMoveMode(self=@n3, IsMoveMode=Get IsMoveMode(), NeedsToAnimate=false) {is=title K2Node_CallFunction}
  ]}
]
variables [
 IsMoveMode {type=boolean}
 OnGoalsDoubleClicked {}
 OnGoalsSelectedIndexChanged {}
 OnListViewFinishedScrolling {}
 OnListViewScrolled {}
 OnPunchCardRewardViewChanged {}
 OnQuestClicked {}
 OnQuestDoubleClicked {}
 OnQuestHover {}
 OnQuestNavigate {}
]
widgets {
 info {
  inheritedWidgetCount=1
  parentClass=/Script/FortniteUI.FortMultiSourcePolymorphicListView
  rootWidgetClass=&CommonListView
  widgetCount=1
 }
 widgets [ListView {
   widgetClassPath=&CommonListView
   bInherited=true
   bIsVariable=true
   widget=ListView
   properties {
    bInEnableTouchAnimatedScrolling=true
    bReturnFocusToSelection=true
    clipping=Inherit
    entryWidgetClass=/QuestUI/QuestScreen/QuestTile/WBP_QuestTile.WBP_QuestTile_C
    navigation=&self:WidgetTree.ListView.WidgetNavigation_0
    numDesignerPreviewEntries=9
    scrollIntoViewAlignment=TopOrLeft
    verticalEntrySpacing=24
    scrollBarStyle {
     thickness=12
     draggedThumbImage {
      bIsDynamicallyLoaded=false
      drawAs=Box
      imageType=FullColor
      mirroring=NoMirror
      resourceName=../../../Engine/Content/Slate/Common/Scrollbar_Thumb.png
      resourceObject=null
      tiling=NoTile
      imageSize=<0 0>
      margin=<ltrb 0.25 0.25 0.25 0.25>
      outlineSettings {
       bUseBrushTransparency=false
       roundingType=HalfHeightRadius
       width=0
       color {colorUseRule=UseColor_Specified, specifiedColor=<rgba 0 0 0 0>}
       cornerRadii=<0 0 0 0>
      }
      tintColor {colorUseRule=UseColor_Specified, specifiedColor=<rgba 1 1 1 1>}
      uVRegion {bIsValid=false, max=<0 0>, min=<0 0>}
     }
     horizontalBackgroundImage {
      bIsDynamicallyLoaded=false
      drawAs=NoDrawType
      imageType=NoImage
      mirroring=NoMirror
      resourceName=None
      resourceObject=null
      tiling=NoTile
      imageSize=<0 0>
      margin=<ltrb 0 0 0 0>
      outlineSettings {
       bUseBrushTransparency=false
       roundingType=HalfHeightRadius
       width=0
       color {colorUseRule=UseColor_Specified, specifiedColor=<rgba 0 0 0 0>}
       cornerRadii=<0 0 0 0>
      }
      tintColor {colorUseRule=UseColor_Specified, specifiedColor=<rgba 1 1 1 1>}
      uVRegion {bIsValid=false, max=<0 0>, min=<0 0>}
     }
     horizontalBottomSlotImage {
      bIsDynamicallyLoaded=false
      drawAs=Image
      imageType=FullColor
      mirroring=NoMirror
      resourceName=../../../Engine/Content/Slate/Common/Scrollbar_Background_Horizontal.png
      resourceObject=null
      tiling=NoTile
      imageSize=<8 8>
      margin=<ltrb 0 0 0 0>
      outlineSettings {
       bUseBrushTransparency=false
       roundingType=HalfHeightRadius
       width=0
       color {colorUseRule=UseColor_Specified, specifiedColor=<rgba 0 0 0 0>}
       cornerRadii=<0 0 0 0>
      }
      tintColor {colorUseRule=UseColor_Specified, specifiedColor=<rgba 1 1 1 1>}
      uVRegion {bIsValid=false, max=<0 0>, min=<0 0>}
     }
     horizontalTopSlotImage {
      bIsDynamicallyLoaded=false
      drawAs=Image
      imageType=FullColor
      mirroring=NoMirror
      resourceName=../../../Engine/Content/Slate/Common/Scrollbar_Background_Horizontal.png
      resourceObject=null
      tiling=NoTile
      imageSize=<8 8>
      margin=<ltrb 0 0 0 0>
      outlineSettings {
       bUseBrushTransparency=false
       roundingType=HalfHeightRadius
       width=0
       color {colorUseRule=UseColor_Specified, specifiedColor=<rgba 0 0 0 0>}
       cornerRadii=<0 0 0 0>
      }
      tintColor {colorUseRule=UseColor_Specified, specifiedColor=<rgba 1 1 1 1>}
      uVRegion {bIsValid=false, max=<0 0>, min=<0 0>}
     }
     hoveredThumbImage {
      bIsDynamicallyLoaded=false
      drawAs=Box
      imageType=FullColor
      mirroring=NoMirror
      resourceName=../../../Engine/Content/Slate/Common/Scrollbar_Thumb.png
      resourceObject=null
      tiling=NoTile
      imageSize=<0 0>
      margin=<ltrb 0.25 0.25 0.25 0.25>
      outlineSettings {
       bUseBrushTransparency=false
       roundingType=HalfHeightRadius
       width=0
       color {colorUseRule=UseColor_Specified, specifiedColor=<rgba 0 0 0 0>}
       cornerRadii=<0 0 0 0>
      }
      tintColor {colorUseRule=UseColor_Specified, specifiedColor=<rgba 1 1 1 1>}
      uVRegion {bIsValid=false, max=<0 0>, min=<0 0>}
     }
     normalThumbImage {
      bIsDynamicallyLoaded=false
      drawAs=Box
      imageType=FullColor
      mirroring=NoMirror
      resourceName=../../../Engine/Content/Slate/Common/Scrollbar_Thumb.png
      resourceObject=null
      tiling=NoTile
      imageSize=<0 0>
      margin=<ltrb 0.25 0.25 0.25 0.25>
      outlineSettings {
       bUseBrushTransparency=false
       roundingType=HalfHeightRadius
       width=0
       color {colorUseRule=UseColor_Specified, specifiedColor=<rgba 0 0 0 0>}
       cornerRadii=<0 0 0 0>
      }
      tintColor {colorUseRule=UseColor_Specified, specifiedColor=<rgba 1 1 1 1>}
      uVRegion {bIsValid=false, max=<0 0>, min=<0 0>}
     }
     verticalBackgroundImage {
      bIsDynamicallyLoaded=false
      drawAs=NoDrawType
      imageType=NoImage
      mirroring=NoMirror
      resourceName=None
      resourceObject=null
      tiling=NoTile
      imageSize=<0 0>
      margin=<ltrb 0 0 0 0>
      outlineSettings {
       bUseBrushTransparency=false
       roundingType=HalfHeightRadius
       width=0
       color {colorUseRule=UseColor_Specified, specifiedColor=<rgba 0 0 0 0>}
       cornerRadii=<0 0 0 0>
      }
      tintColor {colorUseRule=UseColor_Specified, specifiedColor=<rgba 1 1 1 1>}
      uVRegion {bIsValid=false, max=<0 0>, min=<0 0>}
     }
     verticalBottomSlotImage {
      bIsDynamicallyLoaded=false
      drawAs=Image
      imageType=FullColor
      mirroring=NoMirror
      resourceName=../../../Engine/Content/Slate/Common/Scrollbar_Background_Vertical.png
      resourceObject=null
      tiling=NoTile
      imageSize=<8 8>
      margin=<ltrb 0 0 0 0>
      outlineSettings {
       bUseBrushTransparency=false
       roundingType=HalfHeightRadius
       width=0
       color {colorUseRule=UseColor_Specified, specifiedColor=<rgba 0 0 0 0>}
       cornerRadii=<0 0 0 0>
      }
      tintColor {colorUseRule=UseColor_Specified, specifiedColor=<rgba 1 1 1 1>}
      uVRegion {bIsValid=false, max=<0 0>, min=<0 0>}
     }
     verticalTopSlotImage {
      bIsDynamicallyLoaded=false
      drawAs=Image
      imageType=FullColor
      mirroring=NoMirror
      resourceName=../../../Engine/Content/Slate/Common/Scrollbar_Background_Vertical.png
      resourceObject=null
      tiling=NoTile
      imageSize=<8 8>
      margin=<ltrb 0 0 0 0>
      outlineSettings {
       bUseBrushTransparency=false
       roundingType=HalfHeightRadius
       width=0
       color {colorUseRule=UseColor_Specified, specifiedColor=<rgba 0 0 0 0>}
       cornerRadii=<0 0 0 0>
      }
      tintColor {colorUseRule=UseColor_Specified, specifiedColor=<rgba 1 1 1 1>}
      uVRegion {bIsValid=false, max=<0 0>, min=<0 0>}
     }
    }
   }
  }]
}
