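# A quest screen list entry widget that shows only a vertical gap, built on CommonUserWidget with a single Spacer as root.
# When assigned a list item, it unwraps the item, casts it to the quest UI spacer view model and stores it.
# A SetSpacerSize function maps a Small, Medium or Large enum to a fixed spacer height with zero width; the event graph does not call it directly.
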WidgetBlueprint /QuestUI/QuestScreen/V3/WBP_QuestV3_Spacer
graphs [
 EventGraph {nodes [
   n1 = Event OnListItemObjectSet()
   > n2 = Cast To QuestUISpacerVM!(execute, Object=UnwrapItem!(Item:Object Reference=@n1.ListItemObject) -> (ReturnValue:Object Reference)) -> (then, CastFailed, AsQuest UI Spacer View Model:Quest UI Spacer View Model Object Reference) {is=title K2Node_DynamicCast}
   > Set QuestUISpacerVM(QuestUISpacerVM:Quest UI Spacer View Model Object Reference=@n2) -> (Output_Get:Quest UI Spacer View Model Object Reference)
  ]}
 SetSpacerSize {nodes [
   n1 = SetSpacerSize!() -> (then, SpacerSize:EQuestUISpacerSize Enum=Small) {is=title K2Node_FunctionEntry}
   > SetSize!(
    execute
    self:Spacer Object Reference=Get Spacer() -> (Spacer:Spacer Object Reference)
    InSize_X:double=0
    InSize_Y:double=Select!(Small:double=12, Medium:double=24, Large:double=72, Index:EQuestUISpacerSize Enum=@n1) -> (ReturnValue:double)
   ) -> (then)
  ]}
]
variables []
widgets {
 info {inheritedWidgetCount=0, parentClass=&CommonUserWidget, rootWidgetClass=&Spacer, widgetCount=1}
 widgets [Spacer {widgetClassPath=&Spacer, bIsVariable=true, widget=Spacer, properties {size=<0 0>}, slot=~}]
}
