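# Quest screen list entry, derived from WBP_QuestTile_Base, that shows a centered rich-text message on a backplate.
# Item assignment stores either a QuestUIMessageVM or VM_QuestUISpecialMessage; SetMessageType displays that view model's text for Custom, or fixed text for party-up, all-complete and no-tracked-quests.
# Selection plays the outline animation forward or reversed.
# Entry release, or deselection without a valid item, instantly clears the outline and animations to reset pooled entries.
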WidgetBlueprint /QuestUI/QuestScreen/V3/WBP_QuestsMessage_Entry
animations [ShowSelection {
  binding_count=1
  length_seconds=0.1
  object=&self:ShowSelection
  sequence {bindings [SelectedStateOutline {
     id=#1
     possessed_class=Image
     tracks [
      {
       type=MovieSceneFloatTrack
       display_name=Render Opacity
       sections [{type=MovieSceneFloatSection, end_frame=2, start_frame=0, channels [None {type=MovieSceneScriptingFloatChannel, keys [[0,0],[2,1]]}]}]
      }
      {
       type=MovieSceneByteTrack
       display_name=Visibility
       sections [{type=MovieSceneByteSection, end_frame=2, start_frame=0, channels [None {type=MovieSceneScriptingByteChannel, keys [[0,1],[0.6667,4],[2,4]]}]}]
      }
     ]
    }]}
 }]
graphs [
 EventGraph {nodes [
   Event PreConstruct()
   > RegisterBlocks()
   > SetBlocksTiming!(
    execute
    self:UIKit Modular Button Object Reference
    Timing:"UIKit Block Timing Structure (by ref)"=Get BlockTiming() -> (BlockTiming:UIKit Block Timing Structure)
   ) -> (then)
   n1 = Event On Poly View Item Set!() -> (OutputDelegate, then, Item:Object Reference, SourceIndex:int=0) {is=title K2Node_Event}
   > n2 = Cast To QuestUIMessageVM!(execute, Object=@n1.Item) -> (then, CastFailed, AsQuest UI Message View Model:Quest UI Message View Model Object Reference) {is=title K2Node_DynamicCast}
   > Set VM_QuestUISpecialMessage(VM_QuestUISpecialMessage:VM Quest UISpecial Message Object Reference) -> (Output_Get:VM Quest UISpecial Message Object Reference)
   > Set QuestUIMessageVM(QuestUIMessageVM:Quest UI Message View Model Object Reference=@n2) -> (Output_Get:Quest UI Message View Model Object Reference)
   n3 = Cast To VM_QuestUISpecialMessage!(execute=@n2.CastFailed, Object=@n1.Item) -> (then, CastFailed, AsVM Quest UISpecial Message:VM Quest UISpecial Message Object Reference) {is=title K2Node_DynamicCast}
   > Set QuestUIMessageVM()
   > Set VM_QuestUISpecialMessage(VM_QuestUISpecialMessage=@n3)
   Event On Entry Released() {is=title K2Node_Event}
   > InstantDeselect()
   n4 = Event On Item Selection Changed!() -> (OutputDelegate, then, bIsSelected:bool) {is=title K2Node_Event}
   > PlayAnimation(
    InAnimation=Get ShowSelection() -> (ShowSelection:Widget Animation Object Reference)
    PlayMode=Select(Option 0:EUMGSequencePlayMode Enum=Reverse, Option 1:EUMGSequencePlayMode Enum, Index=@n4.bIsSelected) -> (ReturnValue:EUMGSequencePlayMode Enum)
   )
   Event On Deselected()
   > n5 = Is Valid(InputObject=GetListItemObject!(UserObjectListEntry:User Object List Entry Interface) -> (ReturnValue:Object Reference))
   InstantDeselect(execute=@n5.Is Not Valid)
  ]}
 InstantDeselect {nodes [
   InstantDeselect() {is=title K2Node_FunctionEntry}
   > SetVisibility(self=Get SelectedStateOutline() -> (SelectedStateOutline:Image Object Reference), InVisibility=Collapsed)
   > StopAllAnimations(self:User Widget Object Reference)
   > FlushAnimations(self:User Widget Object Reference)
   > n1 = For Each Loop(Array:Array of UIKit Block Object References=Get ModularBlocks() -> (ModularBlocks:Array of UIKit Block Object References)) -> (Array Element:UIKit Block Object Reference)
   > n2 = Set CurrentBlock(CurrentBlock:UIKit Block Object Reference=@n1) -> (Output_Get:UIKit Block Object Reference)
   > StopAllAnimations(self=@n2)
   > n3 = Cast To WBP_UIKit_Block_Outline!(execute, Object=Get CurrentBlock() -> (CurrentBlock:UIKit Block Object Reference)) -> (then, CastFailed, AsWBP UIKit Block Outline:WBP UIKit Block Outline Object Reference) {is=title K2Node_DynamicCast}
   > RequestInstantTransition!(execute, self:UIKit Block Object Reference=@n3, State:EUIKitBlockInstantTransitionState Enum=InstantUnfocused) -> (then)
  ]}
 RegisterBlocks {nodes [
   RegisterBlocks() {is=title K2Node_FunctionEntry}
   > RegisterModularBlock!(
    execute
    self:UIKit Modular Button Object Reference
    Block:UIKit Block Object Reference=Get Block_Outline() -> (Block_Outline:WBP UIKit Block Outline Object Reference)
   ) -> (then)
  ]}
 SetMessage {nodes [
   n1 = SetMessage!() -> (then, Message:Text) {is=title K2Node_FunctionEntry}
   > SetText(self=Get RichMessage() -> (RichMessage:Common Rich Text Block Object Reference), InText=@n1)
  ]}
 SetMessageType {nodes [
   n1 = SetMessageType!() -> (then, MessageType:EQuestUIMessageType Enum=Custom) {is=title K2Node_FunctionEntry}
   > n2 = Switch on EQuestUIMessageType!(execute, Selection:EQuestUIMessageType Enum=@n1) -> (Custom:Exec, PartyUp:Exec, AllComplete:Exec, NoTrackedQuests:Exec) {is=title K2Node_SwitchEnum}
   n3 = Get QuestUIMessageVM() -> (QuestUIMessageVM:Quest UI Message View Model Object Reference)
   n4 = Get VM_QuestUISpecialMessage() -> (VM_QuestUISpecialMessage:VM Quest UISpecial Message Object Reference)
   > SetMessage!(
    execute
    self
    Message:Text=Select(
     Option 0:Text=Get Message!(self:VM Quest UISpecial Message Object Reference=@n4) -> (Message:Text)
     Option 1:Text=Get CustomMessage!(self:Quest UI Message View Model Object Reference=@n3) -> (CustomMessage:Text)
     Index=IsValid(Object=@n3)
    ) -> (ReturnValue:Text)
   ) -> (then) {is=title K2Node_CallFunction}
   SetMessage(execute=@n2.PartyUp, Message="LOCTEXT(\"Party up with friends to share Daily Quests! \")")
   SetMessage(execute=@n2.AllComplete, Message="LOCTEXT(\"All quests completed!\")")
   SetMessage(execute=@n2.NoTrackedQuests, Message="LOCTEXT(\"<text size=\\\"15.75\\\">NO TRACKED QUESTS</>\\r\\nStart tracking a quest to see it here\")")
  ]}
]
variables [BlockTiming {
  type=UIKitBlockTiming
  default {
   deselectingFocused=0.2
   deselectingIdle=0.2
   disabling=0
   enabling=0
   focusing=0.2
   hovering=0.2
   pressing=0.2
   releasing=0.2
   selecting=0.2
   unfocusing=0.2
   unhovering=0.2
  }
 }]
widgets {
 info {
  inheritedWidgetCount=1
  parentClass=/QuestUI/QuestScreen/QuestTile/WBP_QuestTile_Base.WBP_QuestTile_Base_C
  rootWidgetClass=&GridPanel
  widgetCount=6
 }
 widgets [
  Root {widgetClassPath=&GridPanel, widget=Root, properties {columnFill [1], rowFill [1]}}
  Background {
   widgetClassPath=&Image
   bIsVariable=true
   parent=Root
   widget=Background
   properties {brush {
     bIsDynamicallyLoaded=false
     drawAs=Image
     imageType=FullColor
     mirroring=NoMirror
     resourceName=None
     resourceObject=/FortniteCoreUI/UI/Quests/MaterialInstances/MI_UI_QuestBackplate
     tiling=NoTile
     imageSize=<0 102>
     margin=<ltrb 0 0 0 0>
     outlineSettings {
      bUseBrushTransparency=false
      roundingType=HalfHeightRadius
      width=0
      color {colorUseRule=UseColor_Specified, specifiedColor=<rgba 0 0 0 0>}
      cornerRadii=<0 0 0 0>
     }
     tintColor {colorUseRule=UseColor_Specified, specifiedColor=<rgba 1 1 1 1>}
     uVRegion {bIsValid=false, max=<0 0>, min=<0 0>}
    }}
   slot {class=GridSlot}
  }
  SelectedStateOutline {
   widgetClassPath=&Image
   bIsVariable=true
   parent=Root
   widget=SelectedStateOutline
   properties {
    visibility=Collapsed
    brush {
     bIsDynamicallyLoaded=false
     drawAs=Image
     imageType=FullColor
     mirroring=NoMirror
     resourceName=None
     resourceObject=/QuestUI/QuestScreen/QuestTile/Materials/MI_UI_Outline_Selected
     tiling=NoTile
     imageSize=<32 32>
     margin=<ltrb 0 0 0 0>
     outlineSettings {
      bUseBrushTransparency=false
      roundingType=HalfHeightRadius
      width=0
      color {colorUseRule=UseColor_Specified, specifiedColor=<rgba 0 0 0 0>}
      cornerRadii=<0 0 0 0>
     }
     tintColor {colorUseRule=UseColor_Specified, specifiedColor=<rgba 1 1 1 1>}
     uVRegion {bIsValid=false, max=<0 0>, min=<0 0>}
    }
   }
   slot {class=GridSlot, properties {layer=1}}
  }
  Block_Outline {
   widgetClassPath=/FortUILibrary/UIKit/Buttons/Blocks/WBP_UIKit_Block_Outline
   bIsVariable=true
   parent=Root
   widget=Block_Outline
   properties {bottomLeftCornerRadius=24, bottomRightCornerRadius=24, radiusInPixels=true, renderOpacity=0, topLeftCornerRadius=24, topRightCornerRadius=24}
   slot {class=GridSlot, properties {columnSpan=4, layer=10}}
  }
  RichMessage {
   widgetClassPath=&CommonRichTextBlock
   bIsVariable=true
   parent=Root
   widget=RichMessage
   properties {
    bOverrideDefaultStyle=true
    justification=Center
    lineHeightPercentage=1.3
    mobileTextBlockScale=1.2
    scrollStyle=&TextScrollStyle-Base_C
    text="<text size=\"15.75\"> Placeholder</> \r\nPlaceholder"
    defaultTextStyleOverride {
     overflowPolicy=Clip
     transformPolicy=None
     colorAndOpacity {colorUseRule=UseColor_Specified, specifiedColor=<rgba 1 1 1 1>}
     font {
      bForceMonospaced=false
      bMaterialIsStencil=false
      fontMaterial=null
      fontObject=&HeadingNow
      letterSpacing=10
      monospacedWidth=1
      size=12.75
      skewAmount=0
      typefaceFontName=75 Medium
      outlineSettings {
       bApplyOutlineToDropShadows=false
       bMiteredCorners=false
       bSeparateFillAlpha=false
       outlineMaterial=null
       outlineSize=0
       outlineColor=<rgba 1 1 1 1>
      }
     }
     highlightColor {colorUseRule=UseColor_Specified, specifiedColor=<rgba 0 0 0 1>}
     highlightShape {
      bIsDynamicallyLoaded=false
      drawAs=Image
      imageType=NoImage
      mirroring=NoMirror
      resourceName=None
      resourceObject=null
      tiling=NoTile
      imageSize=<32 32>
      margin=<ltrb 0 0 0 0>
      outlineSettings {
       bUseBrushTransparency=false
       roundingType=HalfHeightRadius
       width=0
       color {colorUseRule=UseColor_Specified, specifiedColor=<rgba 0 0 0 0>}
       cornerRadii=<0 0 0 0>
      }
      tintColor {colorUseRule=UseColor_Specified, specifiedColor=<rgba 1 1 1 1>}
      uVRegion {bIsValid=false, max=<0 0>, min=<0 0>}
     }
     selectedBackgroundColor {colorUseRule=UseColor_Foreground, specifiedColor=<rgba 1 0 1 1>}
     shadowColorAndOpacity=<rgba 0 0 0 1>
     shadowOffset=<0 0>
     strikeBrush {
      bIsDynamicallyLoaded=false
      drawAs=Image
      imageType=NoImage
      mirroring=NoMirror
      resourceName=None
      resourceObject=null
      tiling=NoTile
      imageSize=<32 32>
      margin=<ltrb 0 0 0 0>
      outlineSettings {
       bUseBrushTransparency=false
       roundingType=HalfHeightRadius
       width=0
       color {colorUseRule=UseColor_Specified, specifiedColor=<rgba 0 0 0 0>}
       cornerRadii=<0 0 0 0>
      }
      tintColor {colorUseRule=UseColor_Specified, specifiedColor=<rgba 1 1 1 1>}
      uVRegion {bIsValid=false, max=<0 0>, min=<0 0>}
     }
     underlineBrush {
      bIsDynamicallyLoaded=false
      drawAs=Image
      imageType=NoImage
      mirroring=NoMirror
      resourceName=None
      resourceObject=null
      tiling=NoTile
      imageSize=<32 32>
      margin=<ltrb 0 0 0 0>
      outlineSettings {
       bUseBrushTransparency=false
       roundingType=HalfHeightRadius
       width=0
       color {colorUseRule=UseColor_Specified, specifiedColor=<rgba 0 0 0 0>}
       cornerRadii=<0 0 0 0>
      }
      tintColor {colorUseRule=UseColor_Specified, specifiedColor=<rgba 1 1 1 1>}
      uVRegion {bIsValid=false, max=<0 0>, min=<0 0>}
     }
    }
   }
   slot {class=GridSlot, properties {layer=10, verticalAlignment=VAlign_Center, padding=<ltrb 20 8 20 8>}}
  }
  InputActionWidget {widgetClassPath=&CommonActionWidget, bInherited=true, widget=null}
 ]
}
