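# Quest screen V3 list widget wrapping a polymorphic list view that maps each quest view model type to an entry widget and selectability.
# Binds generated and released entries to track row column selection, carrying the active column across multi-column rows.
# Broadcasts OnActiveColumnVMChanged when the active column's view model changes.
# Provides helpers to find selectable items, select and scroll to content, focus the list on gamepad, and forward move mode.
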WidgetBlueprint /QuestUI/QuestScreen/V3/WBP_QuestsV3ListView
graphs [
 CheckIfItemIsSelectableByIndex {nodes [
   n1 = CheckIfItemIsSelectableByIndex!() -> (then, Index:int) {is=title K2Node_FunctionEntry}
   n2 = Get QuestTilesList() -> (QuestTilesList:WBP Quest V3Poly List Object Reference)
   n3 = "Get (a copy)"(
    Array:Array of Object References=GetListItems!(self:List View Object Reference=Get ListView!(self:Fort Polymorphic List View Object Reference=@n2) -> (ListView:Common List View Object Reference)) -> (ReturnValue:Array of Object References)
    Dimension 1=@n1
   ) -> (Output:Object Reference)
   > Return Node!(
    execute
    IsSelectable:bool=CheckItemSelectability!(
     self:WBP Quest V3Poly List Object Reference=Get QuestTilesList()
     InputObject:Object Reference=UnwrapItem!(Item:Object Reference=@n3) -> (ReturnValue:Object Reference)
    ) -> (ReturnValue:bool)
   )
  ]}
 EventGraph {nodes [
   Event Construct()
   n1 = Create Event()
   n2 = Get ListView(self=Get QuestTilesList())
   > Bind Event to On Entry Generated(self:List View Base Object Reference=@n2, Delegate=@n1) {is=title K2Node_AddDelegate}
   n3 = Create Event()
   > Bind Event to On Entry Released(self:List View Base Object Reference=@n2, Delegate=@n3) {is=title K2Node_AddDelegate}
   n4 = Create Event()
   > Bind Event to On Item Selection Changed(self:List View Object Reference=@n2, Delegate=@n4) {is=title K2Node_AddDelegate}
   > SetReturnFocusToSelection!(execute, self:List View Object Reference=@n2, bEnabled:bool=true) -> (then)
   Event Destruct()
   > Unbind Event from On Entry Generated(self:List View Base Object Reference=@n2, Delegate=@n1) {is=title K2Node_RemoveDelegate}
   > Bind Event to On Entry Released(self=@n2, Delegate=@n3)
   > Bind Event to On Item Selection Changed(self=@n2, Delegate=@n4)
  ]}
 FindFirstSelectableElement {nodes [
   FindFirstSelectableElement() {is=title K2Node_FunctionEntry}
   > Set FirstSelectableItem(FirstSelectableItem:int=-1) -> (Output_Get:int=0)
   n1 = Get QuestTilesList()
   > n2 = For Each Loop with Break!(Exec:Exec, Array:Array of Object References=GetListItems(self=Get ListView(self=@n1)), Break:Exec=@n3.then) -> (LoopBody:Exec, Array Element:Object Reference, Array Index:int, Completed:Exec)
   Return Node!(execute=@n2.Completed, FirstSelectedIndex:int=Get FirstSelectableItem() -> (FirstSelectableItem:int=0))
   > Branch(Condition=CheckItemSelectability(self=Get QuestTilesList(), InputObject=UnwrapItem(Item=@n2)))
   > n3 = Set FirstSelectableItem(FirstSelectableItem=@n2.Array Index) -> (Output_Get=0)
   Reroute Node(InputPin:Wildcard) -> (OutputPin:Wildcard)
  ]}
 FindLastSelectableElement {nodes [
   FindLastSelectableElement() {is=title K2Node_FunctionEntry}
   n1 = Get ListView(self=Get QuestTilesList())
   > Set LastSelecatableItem(LastSelecatableItem:int=Last Index!(TargetArray:Array of Object References=GetListItems(self=@n1)) -> (ReturnValue=0)) -> (Output_Get:int=0)
   n2 = GetListItems(self=Get ListView(self=Get QuestTilesList()))
   > n3 = For Each Loop with Break(Array=@n2, Break=@n6.then)
   n4 = int - int(A=Last Index(TargetArray=@n2) -> (ReturnValue=0), B=@n3.Array Index)
   n5 = Get ListView(self=Get QuestTilesList())
   > Branch(Condition=CheckItemSelectability(self=Get QuestTilesList(), InputObject=GetItemAt!(self:List View Object Reference=@n5, Index:int=@n4) -> (ReturnValue:Object Reference)))
   > n6 = Set LastSelecatableItem(LastSelecatableItem=@n4) -> (Output_Get=0)
  ]}
 FocusListview {nodes [
   FocusListview() {is=title K2Node_FunctionEntry}
   n1 = Get CommonInputSubsystem()
   > Branch(Condition="Equal (Enum)"(A=GetCurrentInputType(self=@n1), B=Gamepad))
   > SetFocus(self=Get ListView(self=Get QuestTilesList()))
  ]}
 OnActiveColumnVMChanged {nodes [OnActiveColumnVMChanged!() -> (then, ColumnVM:Quest UI Base Content View Model Object Reference) {is=title K2Node_FunctionEntry}]}
 OnColumnSelected {nodes [
   n1 = OnColumnSelected!() -> (
    then
    RowVM:Quest UI Base Content View Model Object Reference
    SelectionVM:Quest UI Base Content View Model Object Reference
    ActiveColumn:int
   ) {is=title K2Node_FunctionEntry}
   > Set ActiveColumn(ActiveColumn:int=@n1.ActiveColumn) -> (Output_Get:int=0)
   > UpdateActiveColumnVM!(execute, self, NewVM:Quest UI Base Content View Model Object Reference=@n1.SelectionVM) -> (then)
   n2 = Get QuestTilesList()
   > Branch(Condition="Not Equal (Object)"(
     A:Object Reference=GetSelectedItem!(self:Fort Multi Source Polymorphic List View Object Reference=@n2) -> (ReturnValue:Object Reference)
     B:Object Reference=@n1
    ) -> (ReturnValue:bool=~) {is=title K2Node_PromotableOperator})
   > SetSelectedItem!(execute, self:Fort Multi Source Polymorphic List View Object Reference=@n2, Item:Object Reference=@n1) -> (then, ReturnValue:bool)
  ]}
 OnEntrySelectionChanged {nodes [
   n1 = OnEntrySelectionChanged!() -> (then, Item:Object Reference, bIsSelected:bool=~) {is=title K2Node_FunctionEntry}
   > Branch(Condition=@n1.bIsSelected)
   > n2 = Cast To QuestUIRowWidget!(
    execute
    Object=GetWidgetFromItem!(self:Fort Polymorphic View Object Reference=Get QuestTilesList(), Item:Object Reference=@n1) -> (ReturnValue:User Widget Object Reference)
   ) -> (then, CastFailed, AsQuest UIRow Widget:Quest UIRow Widget Object Reference) {is=title K2Node_DynamicCast}
   > SetActiveColumn!(execute, self:Quest UIRow Widget Object Reference=@n2, ActiveColumn:int=Get ActiveColumn() -> (ActiveColumn:int=0)) -> (then)
   UpdateActiveColumnVM(execute=@n2.CastFailed)
  ]}
 OnEntryWidgetGenerated {nodes [
   n1 = OnEntryWidgetGenerated!() -> (then, Widget:User Widget Object Reference) {is=title K2Node_FunctionEntry}
   > n2 = Cast To QuestUIRowWidget(Object=@n1)
   > Bind Event to On Column Selected(self:Quest UIRow Widget Object Reference=@n2, Delegate=Create Event()) {is=title K2Node_AddDelegate}
   n3 = Get QuestTilesList()
   > Branch(Condition="Equal (Object)"(
     A=GetRowVM!(self:Quest UIRow Widget Object Reference=@n2) -> (ReturnValue:Quest UI Base Content View Model Object Reference)
     B=GetSelectedItem(self=@n3)
    ))
   > SetActiveColumn(self=@n2, ActiveColumn=Get ActiveColumn() -> (ActiveColumn=0))
  ]}
 OnEntryWidgetReleased {nodes [
   n1 = OnEntryWidgetReleased!() -> (then, Widget:User Widget Object Reference) {is=title K2Node_FunctionEntry}
   > n2 = Cast To QuestUIRowWidget(Object=@n1)
   > Unbind Event from On Column Selected(self:Quest UIRow Widget Object Reference=@n2, Delegate=Create Event()) {is=title K2Node_RemoveDelegate}
  ]}
 SelectContent {nodes [
   n1 = SelectContent!() -> (then, ContentToSelect:Quest UI Base Content View Model Object Reference) {is=title K2Node_FunctionEntry}
   n2 = Get QuestTilesList()
   > n3 = SetSelectedItem(self=@n2, Item=@n1)
   > n4 = Branch(Condition=@n3)
   > FocusListview() {is=title K2Node_CallFunction}
   > Navigate To Item!(
    execute
    self:List View Object Reference=Get ListView(self=@n2)
    Item:Object Reference=Get ContentToSelect() -> (ContentToSelect:Quest UI Base Content View Model Object Reference)
   ) -> (then)
   > Return Node!(execute, Success:bool=true)
   Return Node(execute=@n4.else, Success=false)
  ]}
 SelectIndex {nodes [
   n1 = SelectIndex!() -> (then, Index:int) {is=title K2Node_FunctionEntry}
   n2 = Get ListView(self=Get QuestTilesList())
   > SetSelectedIndex!(execute, self:List View Object Reference=@n2, Index:int=@n1) -> (then)
   > ScrollIndexIntoView!(execute, self:List View Object Reference=@n2, Index:int=@n1) -> (then)
  ]}
 SetIsMoveMode {nodes [
   n1 = SetIsMoveMode!() -> (then, IsMoveMode:bool=~) {is=title K2Node_FunctionEntry}
   > SetIsMoveMode!(execute, self:WBP Quest V3Poly List Object Reference=Get QuestTilesList(), IsMoveMode:bool=@n1) -> (then) {is=title K2Node_CallFunction}
  ]}
 UpdateActiveColumnVM {nodes [
   n1 = UpdateActiveColumnVM!() -> (then, NewVM:Quest UI Base Content View Model Object Reference) {is=title K2Node_FunctionEntry}
   > Branch(Condition="Not Equal (Object)"(A=@n1, B=Get ActiveColumnVM() -> (ActiveColumnVM:Quest UI Base Content View Model Object Reference)) -> (ReturnValue=~))
   > n2 = Set ActiveColumnVM(ActiveColumnVM:Quest UI Base Content View Model Object Reference=@n1) -> (Output_Get:Quest UI Base Content View Model Object Reference)
   > Call On Active Column VMChanged!(execute:Exec, self:WBP Quests V3List View Object Reference, ColumnVM:Quest UI Base Content View Model Object Reference=@n2) -> (then:Exec) {is=title K2Node_CallDelegate}
  ]}
]
variables [
 ActiveColumn {type=integer}
 ActiveColumnVM {type=QuestUIBaseContentVM, default=null}
 FirstPreferredSelectableItem {type=integer}
 FirstSelectableItem {type=integer}
 LastSelecatableItem {type=integer}
 List {type="array<MVVMViewModelBase>", default []}
 OnActiveColumnVMChanged {}
 SelectableQuestItems {type="array<integer>"}
]
widgets {
 info {
  inheritedWidgetCount=0
  parentClass=&CommonUserWidget
  rootWidgetClass=&/WBP_QuestV3PolyList.WBP_QuestV3PolyList_C
  widgetCount=1
 }
 widgets [QuestTilesList {
   widgetClassPath=&/WBP_QuestV3PolyList
   bIsVariable=true
   widget=QuestTilesList
   properties {
    bIsFocusable=true
    itemClassToEntryData {
     &Class'/Script/QuestUI.QuestUICategoryVM' {bIsEntrySelectable=false, desiredEntryClass=&/WBP_QuestsSeparate_Header.WBP_QuestsSeparate_Header_C}
     &Class'/Script/QuestUI.QuestUICompatibilityVM' {
      bIsEntrySelectable=false
      desiredEntryClass=/QuestUI/QuestScreen/QuestTile/WBP_QuestGroupSeparatorTitle_Entry.WBP_QuestGroupSeparatorTitle_Entry_C
     }
     &Class'/Script/QuestUI.QuestUICustomCategoryVM' {bIsEntrySelectable=false, desiredEntryClass=&/WBP_QuestsSeparate_CustomHeader.WBP_QuestsSeparate_CustomHeader_C}
     &Class'/Script/QuestUI.QuestUICustomRowVM' {bIsEntrySelectable=true, desiredEntryClass=&/WBP_CustomRow.WBP_CustomRow_C}
     &Class'/Script/QuestUI.QuestUIHeaderVM' {bIsEntrySelectable=false, desiredEntryClass=&/WBP_QuestV3Category.WBP_QuestV3Category_C}
     &Class'/Script/QuestUI.QuestUIHorizontalPunchCardVM' {bIsEntrySelectable=true, desiredEntryClass=/QuestUI/QuestScreen/Modal/Goals/WBP_GoalsProgress_Horizontal.WBP_GoalsProgress_Horizontal_C}
     &Class'/Script/QuestUI.QuestUILockedQuestVM' {
      bIsEntrySelectable=false
      desiredEntryClass=&WBP_QuestGroupSeparatorTitle_Entry_C
     }
     &Class'/Script/QuestUI.QuestUIMessageVM' {bIsEntrySelectable=true, desiredEntryClass=&/WBP_QuestsMessage_Entry.WBP_QuestsMessage_Entry_C}
     &Class'/Script/QuestUI.QuestUIQuestVM' {bIsEntrySelectable=true, desiredEntryClass=/QuestUI/QuestScreen/QuestTile/WBP_QuestTile.WBP_QuestTile_C}
     &Class'/Script/QuestUI.QuestUISpacerVM' {bIsEntrySelectable=false, desiredEntryClass=&/WBP_QuestV3_Spacer.WBP_QuestV3_Spacer_C}
     &Class'/Script/QuestUI.QuestUITrackerVM' {bIsEntrySelectable=false, desiredEntryClass=&/WBP_QuestsSeparate_Tracker.WBP_QuestsSeparate_Tracker_C}
     /Script/Engine.BlueprintGeneratedClass'/QuestUI/QuestScreen/Debug/DebugQuestUIHeaderVM.DebugQuestUIHeaderVM_C' {
      bIsEntrySelectable=false
      desiredEntryClass=&WBP_QuestGroupSeparatorTitle_Entry_C
     }
     &BlueprintGeneratedClass'/QuestUI/QuestScreen/Modal/Goals/QuestUIGoalHeaderVM.QuestUIGoalHeaderVM_C' {
      bIsEntrySelectable=false
      desiredEntryClass=/QuestUI/QuestScreen/Modal/Goals/WBP_GoalProgressHeader_Vertical.WBP_GoalProgressHeader_Vertical_C
     }
     &BlueprintGeneratedClass'/QuestUI/QuestScreen/Modal/MVVM_PunchCardQuest.MVVM_PunchCardQuest_C' {bIsEntrySelectable=true, desiredEntryClass=&WBP_QuestTile_C}
     &BlueprintGeneratedClass'/QuestUI/QuestScreen/Modal/QuestUISpacerVM.QuestUISpacerVM_C' {bIsEntrySelectable=false, desiredEntryClass=/QuestUI/QuestScreen/QuestTile/WBP_QuestSpacer_Entry.WBP_QuestSpacer_Entry_C}
     &BlueprintGeneratedClass'/QuestUI/QuestScreen/Modal/VM_QuestUIPinnedQuestSeparator.VM_QuestUIPinnedQuestSeparator_C' {bIsEntrySelectable=false, desiredEntryClass=/QuestUI/QuestScreen/Modal/WBP_PinnedSeparator.WBP_PinnedSeparator_C}
     &BlueprintGeneratedClass'&/VM_QuestUISpecialMessage.VM_QuestUISpecialMessage_C' {bIsEntrySelectable=true, desiredEntryClass=&WBP_QuestsMessage_Entry_C}
    }
   }
  }]
}
