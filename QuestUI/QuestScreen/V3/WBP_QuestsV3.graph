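# Activatable V3 quest screen for browsing, tracking and reordering quests, with a side navigation, multi-source quest tile list and selection panel.
# Category selection jumps the list to the matching source, whose first source holds tracked quests, and updates the panel.
# Move mode and drag-drop show placement with a viewport overlay tile.
# List delegates bind on activation, alongside an event-router listener opening islands in Discovery; deactivation caches selection and commits analytics impressions.
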WidgetBlueprint /QuestUI/QuestScreen/V3/WBP_QuestsV3
animations [Anim_ScreenIn {
  binding_count=1
  length_seconds=0.4966
  object=&self:Anim_ScreenIn
  sequence {bindings [WBP_UIKit_FrontendLayout {
     id=#1
     possessed_class=WBP_UIKit_FrontendLayout_C
     tracks [
      {
       type=MovieSceneFloatTrack
       display_name=Render Opacity
       sections [{type=MovieSceneFloatSection, end_frame=9, start_frame=0, channels [None {type=MovieSceneScriptingFloatChannel, keys [[0,0],[9.9313,1]]}]}]
      }
      {
       type=MovieScene2DTransformTrack
       sections [{
         type=MovieScene2DTransformSection
         end_frame=9
         start_frame=0
         channels [
          Translation.X {type=MovieSceneScriptingFloatChannel, keys [[0,50],[9.9313,0]]}
          Translation.Y {type=MovieSceneScriptingFloatChannel, const [0,1]}
         ]
        }]
      }
     ]
    }]}
 }]
graphs [
 ActivateSelectionPanel {nodes [
   n1 = ActivateSelectionPanel!() -> (then, Activate:bool=~) {is=title K2Node_FunctionEntry}
   > n2 = Branch(Condition=@n1)
   n3 = Get SelectionPanel() -> (SelectionPanel:WBP Selection Panel Object Reference)
   > ActivateWidget(self=@n3)
   DeactivateWidget(execute=@n2.else, self=@n3)
  ]}
 BP_GetDesiredFocusTarget {nodes [
   Get Desired Focus Target()
   n1 = Get ListView!(self:Fort Polymorphic List View Object Reference=Get QuestTilesList!(self:WBP Quests V3List View Object Reference=Get QuestsV3ListView() -> (QuestsV3ListView:WBP Quests V3List View Object Reference)) -> (QuestTilesList:WBP Quest V3Poly List Object Reference)) -> (ListView:Common List View Object Reference)
   n2 = NOT Boolean(A=IsValid(Object=Get Selected Item!(self:List View Object Reference=@n1) -> (ReturnValue:Object Reference) {is=title K2Node_CallFunction}))
   > n3 = Branch(Condition=@n2)
   n4 = Get SideNavigation() -> (SideNavigation:WBP Quest V3 Side Navigation Object Reference)
   Return Node!(
    execute=[@n5.then, @n3.then]
    ReturnValue:Widget Object Reference=Select(
     Option 0:Widget Object Reference=@n1
     Option 1:Widget Object Reference=GetSelectedButton!(self:WBP Quest V3 Side Navigation Object Reference=@n4) -> (ButtonInstance:Common Button Base Object Reference)
     Index=@n2
    ) -> (ReturnValue:Widget Object Reference)
   )
   n5 = SetIsCollapsed!(execute=@n3.else, self:WBP Quest V3 Side Navigation Object Reference=@n4, IsCollapsed:bool=true, SkipAnimation:bool=true) -> (then)
  ]}
 BranchInputType {nodes [
   Inputs!() -> (In:Exec)
   n1 = GetCurrentInputType(self=Get CommonInputSubsystem!() -> (ReturnValue:Common Input Object Reference) {is=title K2Node_GetBlueprintContextEx})
   > n2 = Branch(Condition=OR Boolean(A="Equal (Enum)"(A=@n1, B=Gamepad), B=IsMobileGame()))
   > Outputs!(GamepadOrTouch:Exec, MnK:Exec=@n2.else)
  ]}
 BranchMultipinningCVar {nodes [
   Inputs()
   n1 = GetNumericCVar!(CVar:String=QuestUI.EnableMultiPinning) -> (OutValue:float, ReturnValue:bool)
   > n2 = Branch(Condition=Select(Option 0:bool=false, Option 1:bool="Equal (Float)"(A=@n1, B=1), Index=@n1.ReturnValue) -> (ReturnValue:bool))
   > Outputs!(UseMultipinning:Exec, NoMultipinning:Exec=@n2.else)
  ]}
 CacheCurrentSelection {nodes [
   CacheCurrentSelection() {is=title K2Node_FunctionEntry}
   > Is Valid(InputObject=Get QuestUIPlayerVM() -> (QuestUIPlayerVM:Quest UI Player View Model Object Reference))
   > n1 = Sequence()
   n2 = GetSelectedUnwrappedItem!() -> (ReturnValue:Object Reference)
   > n3 = Cast To QuestUIBaseContentVM!(execute, Object=@n2) -> (then, CastFailed, AsQuest UI Base Content View Model:Quest UI Base Content View Model Object Reference) {is=title K2Node_DynamicCast}
   > Set ContentVM(ContentVM:Quest UI Base Content View Model Object Reference=@n3) -> (Output_Get:Quest UI Base Content View Model Object Reference)
   n4 = Cast To MVVM_PunchCardQuest!(execute=@n3.CastFailed, Object=@n2) -> (then, CastFailed, AsMVVM Punch Card Quest:MVVM Punch Card Quest Object Reference) {is=title K2Node_DynamicCast}
   > Set ContentVM(ContentVM=Get QuestData!(self:MVVM Punch Card Quest Object Reference=@n4) -> (QuestData:Quest UI Quest View Model Object Reference))
   SetSelectionData!(
    execute=@n1.then_1
    self:Quest UI Player View Model Object Reference=Get QuestUIPlayerVM()
    InSelectionData:"Quest UISelection Data Structure (by ref)"=Make Quest UISelection Data!(
     ProductVM:Quest UI Product View Model Object Reference=Get QuestUIProductVM() -> (QuestUIProductVM:Quest UI Product View Model Object Reference)
     CategoryInterfaceVM:Quest UI Category Interface View Model Object Reference=Get QuestUICategoryInterfaceVM() -> (QuestUICategoryInterfaceVM:Quest UI Category Interface View Model Object Reference)
     ContentVM:Quest UI Base Content View Model Object Reference=Get ContentVM() -> (ContentVM:Quest UI Base Content View Model Object Reference)
    ) -> (QuestUISelectionData:Quest UISelection Data Structure) {is=title K2Node_MakeStruct}
   ) -> (then)
  ]}
 ClearDialog {nodes [
   ClearDialog!() -> (then, ButtonIndex:int) {is=title K2Node_FunctionEntry}
   > Is Valid(InputObject=Get Dialog_TrackedCap() -> (Dialog_TrackedCap:WBP UIKit Dialog Object Reference))
   > Unbind all Events from Dialog Button Clicked Dispatcher!(
    execute:Exec
    self:WBP UIKit Dialog Internal Object Reference=Get Dialog_Internal!(self:WBP UIKit Dialog Object Reference=Get Dialog_TrackedCap()) -> (Dialog_Internal:WBP UIKit Dialog Internal Object Reference)
   ) -> (then:Exec) {is=title K2Node_ClearDelegate}
   > DeactivateWidget(self=Get Dialog_TrackedCap())
   > ClearChildren(self=Get Overlay_Dialog() -> (Overlay_Dialog:Overlay Object Reference))
   n1 = Get QuestsV3ListView()
   > SetFocus(self=Get ListView(self=Get QuestTilesList(self=@n1)))
   > Set Dialog_TrackedCap(Dialog_TrackedCap:WBP UIKit Dialog Object Reference) -> (Output_Get:WBP UIKit Dialog Object Reference)
  ]}
 CommitAnalyticsImpressions {nodes [
   CommitAnalyticsImpressions() {is=title K2Node_FunctionEntry}
   n1 = Get CurrentImpressions() -> (CurrentImpressions:Array of Flow Analytics Impression Section Content Structures)
   > FireQuestAnalytics_CommitImpressions!(
    execute
    CategoryInterfaceVM:Quest UI Category Interface View Model Object Reference=Get QuestUICategoryInterfaceVM()
    Impressions:Array of Flow Analytics Impression Section Content Structures=@n1
   ) -> (then)
   > Clear(TargetArray:Array of Flow Analytics Impression Section Content Structures=@n1)
  ]}
 DoCustomNavigation {nodes [
   n1 = DoCustomNavigation!() -> (then, Navigation:EUINavigation Enum) {is=title K2Node_FunctionEntry}
   > n2 = Switch on EUINavigation!(execute, Selection:EUINavigation Enum=@n1) -> (Left:Exec, Right:Exec, Up:Exec, Down:Exec, Next:Exec, Previous:Exec, Invalid:Exec) {is=title K2Node_SwitchEnum}
   > Return Node(ReturnValue=GetSelectedButton(self=Get SideNavigation()))
   n3 = Get QuestsV3ListView()
   Return Node(execute=@n2.Right, ReturnValue=Get ListView(self=Get QuestTilesList(self=@n3)))
  ]}
 EventGraph {nodes [
   n1 = HandleTrackQuestClicked!(execute=@n7.then, self) -> (then, SelectedItemIndex1:int=0)
   n2 = CheckIfItemIsSelectableByIndex!(execute, self:WBP Quests V3List View Object Reference=Get QuestsV3ListView(), Index:int=@n1) -> (then, IsSelectable:bool)
   n3 = Get QuestsV3ListView()
   > SetSelectedIndex!(
    execute
    self:List View Object Reference=Get ListView(self=Get QuestTilesList(self=@n3))
    Index:int=Select(Option 0:int=int - int(A=@n1), Option 1:int=@n1, Index=@n2) -> (ReturnValue:int)
   ) -> (then)
   "On Clicked (DecoyButtonDebugQuest)"!() -> (OutputDelegate, then, Button:Common Button Base Object Reference) {is=title K2Node_ComponentBoundEvent}
   > n4 = Branch(Condition="Equal Exactly (String)"(A=GetBuildConfiguration!() -> (ReturnValue:String), B=Get ShippingConfig() -> (ShippingConfig:String)))
   QueueModalWidgetSoftClassAsync!(execute:Exec=@n4.else, SoftWidgetClass:Common Activatable Widget Soft Class Reference=/QuestUI/DebugTool/QuestDebugTool.QuestDebugTool_C) -> (then:Exec, OnModalWidgetQueued:Exec, ActivatableWidget:Common Activatable Widget Object Reference) {is=title K2Node_AsyncAction}
   > DeactivateWidget()
   "On Clicked (DecoyButtonGoToTop)"!() -> (OutputDelegate, then, Button:Common Button Base Object Reference) {is=title K2Node_ComponentBoundEvent}
   > NavigateToFirstSelectableInSource!(
    execute
    self:Fort Multi Source Polymorphic List View Object Reference=Get QuestTilesList(self=Get QuestsV3ListView())
    SourceIndex:int=0
    ScrollOverrideIndexInSource:int=-1
   ) -> (then, ReturnValue:bool)
   "On Clicked (Button_CancelMove)"!() -> (OutputDelegate, then, Button:Common Button Base Object Reference) {is=title K2Node_ComponentBoundEvent}
   > Shutdown Modes!(In:Exec) -> (Shutdown:Exec, NoShutdownRequired:Exec) {is=title K2Node_MacroInstance}
   n5 = "On Background Changed (WBP_QuestsModal_Header)"!() -> (OutputDelegate, then, Color:Linear Color Structure) {is=title K2Node_ComponentBoundEvent}
   > SetVectorParameterValue(
    self=Get MID_Background!(self:WBP UIKit Frontend Layout Object Reference=Get WBP_UIKit_FrontendLayout() -> (WBP_UIKit_FrontendLayout:WBP UIKit Frontend Layout Object Reference)) -> (MID_Background:Material Instance Dynamic Object Reference)
    ParameterName=FadeColor
    Value=Select(
     Option 0:Linear Color Structure=@n5.Color
     Option 1:Linear Color Structure=(R=0,G=0,B=0,A=0)
     Index="Equal (LinearColor)"(A:Linear Color Structure=@n5.Color, B:Linear Color Structure) -> (ReturnValue:bool=~) {is=title K2Node_PromotableOperator}
    ) -> (ReturnValue:Linear Color Structure)
   )
   "On Clicked (DecoyButtonHideDisabled)"!() -> (OutputDelegate, then, Button:Common Button Base Object Reference) {is=title K2Node_ComponentBoundEvent}
   > HandleHideUnavailableClicked()
   n6 = Get QuestTilesList(self=Get QuestsV3ListView())
   > SetSourceListItems!(
    execute
    self:Fort Multi Source Polymorphic List View Object Reference=@n6
    SourceIndex:int=0
    Items:Array of Object References=GetSortedTrackedQuests!() -> (FilteredVMs:Array of MVVM Base Viewmodel Object References)
   ) -> (then)
   > NavigateToFirstSelectableInSource(self=@n6, SourceIndex=0, ScrollOverrideIndexInSource=-1)
   n7 = "On Clicked (DecoyButtonTrackQuest)"!() -> (OutputDelegate, then, Button:Common Button Base Object Reference) {is=title K2Node_ComponentBoundEvent}
   "On Clicked (DecoyButtonTrackAll)"!() -> (OutputDelegate, then, Button:Common Button Base Object Reference) {is=title K2Node_ComponentBoundEvent}
   > HandleTrackAllQuestClicked()
   "On Clicked (DecoyButtonMoveQuest)"!() -> (OutputDelegate, then, Button:Common Button Base Object Reference) {is=title K2Node_ComponentBoundEvent}
   > Branch(Condition=IsInTrackedCategory!() -> (ReturnValue:bool))
   > SetMoveMode!(execute, self, IsMoveMode:bool=NOT Boolean(A=Get IsMoving() -> (IsMoving:bool)), IsShutdown:bool=false) -> (then)
   "On Clicked (DecoyButtonExternalLink)"!() -> (OutputDelegate, then, Button:Common Button Base Object Reference) {is=title K2Node_ComponentBoundEvent}
   > n8 = Get!(execute:Exec) -> (then:Exec, else:Exec, Quest UI Horizontal Punch Card View Model:Quest UI Horizontal Punch Card View Model Object Reference)
   > n9 = Get!(execute, self:Quest UI Punch Card View Model Object Reference=@n8) -> (then, else, ExternalLinkVM:Quest UIExternal Link VM Object Reference)
   > n10 = Branch(Condition=Get bOwnerAllowsText!(self:Quest UIExternal Link VM Object Reference=@n9) -> (bOwnerAllowsText:bool))
   > n11 = Cast To FortPlayerController(Object=GetOwningPlayer())
   > n12 = "Fire Event Element Interacted (Non-Localized)"!(
    execute
    FortPC:Fort Player Controller Object Reference=@n11
    ElementName:String=GetAnalyticsID!(self:Quest UI Base Content View Model Object Reference=@n9) -> (ReturnValue:String)
    AdditionalInfo:Flow Analytics Additional Info Structure
   ) -> (then)
   ExecuteLink(execute=[@n12.then, @n10.else], self:Quest UIExternal Link VM Object Reference=@n9)
   n13 = "On Category Selected (SideNavigation)"!() -> (OutputDelegate, then, TabID:Name=None, InnerTabID:Name=None, Category:MVVM Base Viewmodel Object Reference) {is=title K2Node_ComponentBoundEvent}
   > Set PreviousSideNavCategoryVM(PreviousSideNavCategoryVM:MVVM Base Viewmodel Object Reference=@n13.Category) -> (Output_Get:MVVM Base Viewmodel Object Reference)
   > AlternativeNavigation!(execute, self:WBP Quest V3 Side Navigation Object Reference=Get SideNavigation(), TabID:Name=@n13.TabID) -> (then)
   n14 = Get SideNavigation()
   > n15 = Branch(Condition=NOT Boolean(A=Get IsCollapsed!(self:WBP Quest V3 Side Navigation Object Reference=@n14) -> (IsCollapsed:bool)))
   n16 = ExtractCategoryIfNeeded!(VM:MVVM Base Viewmodel Object Reference=@n13.Category) -> (ExtractedVM:MVVM Base Viewmodel Object Reference)
   n17 = Find Item!(
    TargetArray:Array of MVVM Base Viewmodel Object References=Get CategoriesID() -> (CategoriesID:Array of MVVM Base Viewmodel Object References)
    ItemToFind:"MVVM Base Viewmodel Object Reference (by ref)"=@n16
   ) -> (ReturnValue)
   > n18 = NavigateToFirstSelectableInSource(
    self=Get QuestTilesList(self=Get QuestsV3ListView())
    SourceIndex=@n17
    ScrollOverrideIndexInSource=Select(Option 0:int=-1, Option 1:int=1, Index="Not Equal (Integer)"(A=@n17)) -> (ReturnValue:int)
   )
   UpdateTrackAllButtonState!(execute=[@n18.then, @n15.else], self, ForceUntrackCategory:bool=false) -> (then)
   > n19 = Cast To QuestUICategoryInterfaceVM!(execute, Object=@n16) -> (then, CastFailed, AsQuest UI Category Interface View Model:Quest UI Category Interface View Model Object Reference) {is=title K2Node_DynamicCast}
   > MarkActiveCategorySeen!(execute, self, CommitSeenOnly:bool=true, Category:Quest UI Base View Model Object Reference=@n19) -> (then)
   > CommitAnalyticsImpressions() {is=title K2Node_CallFunction}
   > SetActiveCategory!(
    execute
    self:WBP Selection Panel Object Reference=Get SelectionPanel()
    CategoryInterfaceVM:Quest UI Category Interface View Model Object Reference=@n19
   ) -> (then)
   > Set QuestUICategoryInterfaceVM(QuestUICategoryInterfaceVM:Quest UI Category Interface View Model Object Reference=@n19) -> (Output_Get:Quest UI Category Interface View Model Object Reference)
   > UpdateMoveQuestButtonVisibility()
   n20 = "On Collapsed State Changed (SideNavigation)"!() -> (OutputDelegate, then, IsCollapsed:bool) {is=title K2Node_ComponentBoundEvent}
   n21 = Get SelectionPanel()
   > HandleSidebarCollapsedStateChanged!(execute, self:WBP Selection Panel Object Reference=@n21, Collapsed:bool=@n20.IsCollapsed) -> (then)
   > Refresh(self:WBP Selection Panel Object Reference=@n21)
   "On Swipe Right (SwipePanel_SideNavigation)"() {is=title K2Node_ComponentBoundEvent}
   > SetIsCollapsed(self:BPI UIKit Collapsible Interface=Get SideNavigation(), IsCollapsed=true, SkipAnimation=false)
   > Refresh(self=Get SelectionPanel())
   n22 = DragDropTargetChanged_Visualization!() -> (
    OutputDelegate
    then
    OldDropTarget:WBP Quest Tile Object Reference
    DropPlacementEnter:EDragDropPlacementType Enum=NewEnumerator0
    DropPlacementLeave:EDragDropPlacementType Enum=NewEnumerator0
   ) {is=title K2Node_CustomEvent}
   > Branch(Condition=Get IsMoving())
   > n23 = Is Valid(InputObject=@n22.OldDropTarget)
   > n24 = DraggedOverChangedVisualization!(
    execute
    self:WBP Quest Tile Object Reference=@n22.OldDropTarget
    IsEntering:bool=false
    DropPlacement:"EDragDropPlacementType Enum (by ref)"=@n22.DropPlacementLeave
   ) -> (then)
   n25 = Get TargetQuestTile() -> (TargetQuestTile:WBP Quest Tile Object Reference)
   DraggedOverChangedVisualization(execute=[@n24.then, @n23.Is Not Valid], self=@n25, IsEntering=true, DropPlacement=@n22.DropPlacementEnter)
   > SetPositionInViewport!(
    execute
    self:User Widget Object Reference=Get WBP_QuestTile_Overlay() -> (WBP_QuestTile_Overlay:WBP Quest Tile Object Reference)
    Position:Vector 2D Structure=GetDragDropVisPosition!(Widget:Widget Object Reference=@n25, Placement:EDragDropPlacementType Enum=@n22.DropPlacementEnter) -> (ReturnValue:Vector 2D Structure)
    bRemoveDPIScale:bool=true
   ) -> (then)
   n26 = OnSelectIsland!() -> (OutputDelegate, then, LinkID:Online Link Id Structure, LinkType:EDiscoveryLinkTreatmentType Enum=Default) {is=title K2Node_CustomEvent}
   > DeactivateWidget()
   > n27 = GetCurrentStateWidget!(execute, PlayerController:Player Controller Object Reference=GetOwningPlayer()) -> (then, ReturnValue:Common Activatable Widget Object Reference)
   > n28 = Cast To AthenaUIStateWidget_Frontend!(execute, Object=@n27) -> (then, CastFailed, AsAthena UIState Widget Frontend:Athena UIState Widget Frontend Object Reference) {is=title K2Node_DynamicCast}
   > NavigateToDiscovery!(
    execute
    self:Athena UIState Widget Frontend Object Reference=@n28
    LinkId:"Online Link Id Structure (by ref)"=@n26.LinkID
    Category:Name=None
    bShouldOpenDescription:bool=true
   ) -> (then)
   Event On Activated()
   > n29 = Sequence()
   > Parent: On Activated() {is=title K2Node_CallParentFunction}
   n30 = Create Event()
   n31 = Get QuestsV3ListView()
   > Bind Event to On Active Column VMChanged(self:WBP Quests V3List View Object Reference=@n31, Delegate=@n30) {is=title K2Node_AddDelegate}
   n32 = Create Event()
   n33 = Get QuestTilesList(self=@n31)
   > Bind Event to On List View Scrolled(self:WBP Quest V3Poly List Object Reference=@n33, Delegate=@n32) {is=title K2Node_AddDelegate}
   n34 = Get QuestsV3ListView()
   > Bind Event to On Item Selection Changed(self:List View Object Reference=Get ListView(self=Get QuestTilesList(self=@n34)), Delegate=Create Event()) {is=title K2Node_AddDelegate}
   > InitSideNavigation!(execute, self, IsCollapsed:bool=true) -> (then)
   > ActivateSelectionPanel!(execute, self, Activate:bool=true) -> (then) {is=title K2Node_CallFunction}
   > Delay(Duration=0)
   > RequestRefreshFocus(self:Common Activatable Widget Object Reference)
   n35 = Get CommonInputSubsystem() {is=title K2Node_GetSubsystem}
   > Bind Event to On Input Method Changed(self=@n35, Delegate=Create Event())
   > OnInputTypeChanged!(execute, self, bNewInputType:ECommonInputType Enum=GetCurrentInputType(self=@n35)) -> (then)
   n36 = Get QuestTilesList(self=Get QuestsV3ListView())
   > Assign On Quest Clicked!(execute:Exec, self:WBP Quest V3Poly List Object Reference=@n36, Delegate:"Delegate (by ref)"=Create Event()) -> (then:Exec) {is=title K2Node_AssignDelegate}
   > Bind Event to On Quest Hover(self:WBP Quest V3Poly List Object Reference=@n36, Delegate=Create Event()) {is=title K2Node_AddDelegate}
   > Assign On Quest Double Clicked!(execute:Exec, self:WBP Quest V3Poly List Object Reference=@n36, Delegate:"Delegate (by ref)"=Create Event()) -> (then:Exec) {is=title K2Node_AssignDelegate}
   > Bind Event to On Quest Navigate(self:WBP Quest V3Poly List Object Reference=@n36, Delegate=Create Event()) {is=title K2Node_AddDelegate}
   > Bind Event to On Goals Double Clicked(self:WBP Quest V3Poly List Object Reference=@n36, Delegate=Create Event()) {is=title K2Node_AddDelegate}
   > Bind Event to On Goals Selected Index Changed(self:WBP Quest V3Poly List Object Reference=@n36, Delegate=Create Event()) {is=title K2Node_AddDelegate}
   > Bind Event to On Punch Card Reward View Changed(self:WBP Quest V3Poly List Object Reference=@n36, Delegate=Create Event()) {is=title K2Node_AddDelegate}
   > n37 = GetEventRouter!(
    execute
    Scope:Object Class Reference=/Script/FortniteGame.FortPlayspace
    ContextActor:Actor Object Reference=GetOwningPlayer()
    None:%=Get GameplayEventRouterSubsystem() -> (ReturnValue:Gameplay Event Router Subsystem Object Reference) {is=title K2Node_GetSubsystem}
   ) -> (then, ReturnValue:Gameplay Event Router Component Object Reference)
   > n38 = StartListeningToEvent!(
    execute:Exec
    Target:Gameplay Event Router Component Object Reference=@n37
    EventType:Script Struct Object Reference=/QuestUI/QuestScreen/FSelectIslandEventData
    Context:Object Reference
    EventBubblingRule:EEventBubblingRule Enum
    None:%=Self-Reference()
   ) -> (
    then:Exec
    AsyncTaskProxy:Async Action Start Listening To Event Object Reference
    OnEventReceived:Exec
    Payload:FSelect Island Event Data Structure
   )
   > Set SelectIslandListener(SelectIslandListener:Async Action Start Listening To Event Object Reference=@n38) -> (Output_Get:Async Action Start Listening To Event Object Reference)
   n39 = Break FSelect Island Event Data!(FSelectIslandEventData:"FSelect Island Event Data Structure (by ref)"=@n38.Payload) -> (
    LinkID_2_#2:Online Link Id Structure
    LinkType_5_#3:EDiscoveryLinkTreatmentType Enum
   ) {is=title K2Node_BreakStruct}
   OnSelectIsland!(
    execute=@n38.OnEventReceived
    self
    LinkID:Online Link Id Structure=@n39
    LinkType:EDiscoveryLinkTreatmentType Enum=@n39.LinkType_5_#3
   ) -> (then) {is=title K2Node_CallFunction}
   PopulateCategories!(
    execute=@n29.then_1
    self
    ProductList:Array of Quest UI Product View Model Object References=Get OrderedProductVMs!(self:Quest UI Player View Model Object Reference=Get QuestUIPlayerVM()) -> (OrderedProductVMs:Array of Quest UI Product View Model Object References)
   ) -> (then)
   n40 = Event PreConstruct()
   > Parent: PreConstruct(IsDesignTime=@n40.IsDesignTime)
   > SetMobileStyle() {is=title K2Node_CallFunction}
   > UpdateTabsScrollbar()
   > SetClipping!(
    execute
    self:Widget Object Reference=Get SizeBox_ContentPanel_Primary!(self:WBP UIKit Frontend Layout Object Reference=Get WBP_UIKit_FrontendLayout()) -> (SizeBox_ContentPanel_Primary:Size Box Object Reference)
    InClipping:EWidgetClipping Enum=Inherit
   ) -> (then)
   Event On Deactivated()
   > Parent: On Deactivated() {is=title K2Node_CallParentFunction}
   > Bind Event to On Active Column VMChanged(self=@n31, Delegate=@n30)
   > Unbind Event from On List View Scrolled(self:WBP Quest V3Poly List Object Reference=@n33, Delegate=@n32) {is=title K2Node_RemoveDelegate}
   > Set PreviousSideNavCategoryVM()
   > SetMoveMode(IsMoveMode=false, IsShutdown=false)
   > CancelDragDrop!(execute) -> (then)
   > CacheCurrentSelection() {is=title K2Node_CallFunction}
   > ActivateSelectionPanel(Activate=false)
   > CommitAnalyticsImpressions()
   n41 = Get QuestsV3ListView()
   > Unbind Event from On Item Selection Changed(self:List View Object Reference=Get ListView(self=Get QuestTilesList(self=@n41)), Delegate=Create Event()) {is=title K2Node_RemoveDelegate}
   > Unbind Event from On Input Method Changed(self=Get CommonInputSubsystem(), Delegate=Create Event())
   n42 = Get QuestTilesList(self=Get QuestsV3ListView())
   > Unbind all Events from On Item Clicked!(execute:Exec, self:List View Object Reference=Get ListView(self=@n42)) -> (then:Exec) {is=title K2Node_ClearDelegate}
   > Unbind all Events from On Quest Clicked!(execute:Exec, self:WBP Quest V3Poly List Object Reference=@n42) -> (then:Exec) {is=title K2Node_ClearDelegate}
   > Unbind all Events from On Quest Double Clicked!(execute:Exec, self:WBP Quest V3Poly List Object Reference=@n42) -> (then:Exec) {is=title K2Node_ClearDelegate}
   > Unbind all Events from On Quest Navigate!(execute:Exec, self:WBP Quest V3Poly List Object Reference=@n42) -> (then:Exec) {is=title K2Node_ClearDelegate}
   > Unbind all Events from On Quest Hover!(execute:Exec, self:WBP Quest V3Poly List Object Reference=@n42) -> (then:Exec) {is=title K2Node_ClearDelegate}
   > Unbind all Events from On Goals Double Clicked!(execute:Exec, self:WBP Quest V3Poly List Object Reference=@n42) -> (then:Exec) {is=title K2Node_ClearDelegate}
   n43 = Get SelectIslandListener() -> (SelectIslandListener:Async Action Start Listening To Event Object Reference)
   > Is Valid(InputObject=@n43)
   > StopListeningToEvent(self:Async Action Start Listening To Event Object Reference=@n43)
   Event Construct()
   > Parent: Construct()
   > SetBottomButtonVisibility!(
    execute
    self:WBP Bottom Bar Decoy Button Object Reference=Get DecoyButtonExternalLink() -> (DecoyButtonExternalLink:WBP Bottom Bar Decoy Button Object Reference)
    ShouldButtonBeVisible:bool=false
   ) -> (then)
   > SetBottomButtonVisibility(
    self=Get DecoyButtonDebugQuest() -> (DecoyButtonDebugQuest:WBP Bottom Bar Decoy Button Object Reference)
    ShouldButtonBeVisible="Not Equal Exactly (String)"(A=GetBuildConfiguration(), B=Get ShippingConfig())
   )
   > UpdateInGamePadding()
   > n44 = Is Valid(InputObject=Get WBP_QuestTile_Overlay())
   n45 = Create Widget!(
    execute:Exec=@n44.Is Not Valid
    Class:User Widget Class Reference=/QuestUI/QuestScreen/QuestTile/WBP_QuestTile.WBP_QuestTile_C
    OwningPlayer:Player Controller Object Reference=GetOwningPlayer()
    QuestUIQuestVM:Quest UI Quest View Model Object Reference
    IsDraggable:bool=False
    Style:Common Button Style Class Reference=/FortUILibrary/Foundation/Buttons/ButtonStyles/ButtonStyle-Empty.ButtonStyle-Empty_C
    bApplyAlphaOnDisable:bool=True
    bLocked:bool=False
    bSelectable:bool=True
    bShouldSelectUponReceivingFocus:bool=False
    bInteractableWhenSelected:bool=false
    bToggleable:bool=False
    bTriggerClickedAfterSelection:bool=False
    bDisplayInputActionWhenNotInteractable:bool=True
    bHideInputActionWithKeyboard:bool=False
    bShouldUseFallbackDefaultInputAction:bool=True
    bRequiresHold:bool=False
    HoldData:Common UIHold Data Class Reference
    ClickMethod:EButtonClickMethod Enum=DownAndUp
    InputPriority:int=0
    TriggeringInputAction:Data Table Row Handle Structure
   ) -> (then:Exec, ReturnValue:WBP Quest Tile Object Reference)
   > n46 = Set WBP_QuestTile_Overlay(WBP_QuestTile_Overlay:WBP Quest Tile Object Reference=@n45) -> (Output_Get:WBP Quest Tile Object Reference)
   > AddToViewport!(execute, self:User Widget Object Reference=@n46, ZOrder:int=100) -> (then)
   > SetVisibility(self=Get WBP_QuestTile_Overlay(), InVisibility=Collapsed)
   n47 = Get SideNavigation()
   > SetConsumePointerInput!(
    execute
    self:Scroll Box Object Reference=Get ScrollBox_Tabs!(self:WBP UIKit Tab Group Base Object Reference=Get SideNavigation!(self:WBP Quest V3 Side Navigation Object Reference=@n47) -> (SideNavigation:WBP UIKit Side Navigation Object Reference)) -> (ScrollBox_Tabs:Scroll Box Object Reference)
    bInConsumePointerInput:bool=false
   ) -> (then)
   Event OnDragCancelled!() -> (OutputDelegate, then, PointerEvent:"Pointer Event Structure (by ref)", Operation:Drag Drop Operation Object Reference) {is=title K2Node_Event}
   > SetMoveMode(IsMoveMode=false, IsShutdown=false)
   Event OnDragEnter!() -> (OutputDelegate, then, MyGeometry:Geometry Structure, PointerEvent:Pointer Event Structure, Operation:Drag Drop Operation Object Reference) {is=title K2Node_Event}
   > SetMoveMode(IsMoveMode=true, IsShutdown=false)
  ]}
 ExtractCategoryIfNeeded {nodes [
   n1 = ExtractCategoryIfNeeded!() -> (then, VM:MVVM Base Viewmodel Object Reference) {is=title K2Node_FunctionEntry}
   > n2 = Cast To QuestUIBucketVM!(execute, Object=@n1) -> (then, CastFailed, AsQuest UI Bucket View Model:Quest UI Bucket View Model Object Reference) {is=title K2Node_DynamicCast}
   n3 = Get OrderedContentVMs!(self:Quest UI Bucket View Model Object Reference=@n2) -> (OrderedContentVMs:Array of Quest UI Base Content View Model Object References)
   > n4 = Branch(Condition="integer > integer"(A=Length(TargetArray:Array of Quest UI Base Content View Model Object References=@n3)))
   > Return Node!(
    execute
    ExtractedVM:MVVM Base Viewmodel Object Reference="Get (a copy)"(Array:Array of Quest UI Base Content View Model Object References=@n3) -> (Output:Quest UI Base Content View Model Object Reference)
   )
   Return Node(execute=@n4.else)
   Return Node(execute=@n2.CastFailed, ExtractedVM=@n1)
  ]}
 FilterTrackableQuests {nodes [
   n1 = FilterTrackableQuests!() -> (then, Array:Array of Quest UI Base Content View Model Object References) {is=title K2Node_FunctionEntry}
   > n2 = For Each Loop(Array:Array of Quest UI Base Content View Model Object References=@n1) -> (Array Element:Quest UI Base Content View Model Object Reference)
   Return Node!(
    execute=@n2.Completed
    TrackableQuests:Array of Quest UI Quest View Model Object References=Get Quests() -> (Quests:Array of Quest UI Quest View Model Object References)
   )
   > n3 = Cast To QuestUIQuestVM!(execute, Object=@n2) -> (then, CastFailed, AsQuest UI Quest View Model:Quest UI Quest View Model Object Reference) {is=title K2Node_DynamicCast}
   > Branch(Condition="Not Equal (Enum)"!(
     A:EQuestTrackState Enum=GetQuestTrackState!(self:Quest UI Quest View Model Object Reference=@n3) -> (ReturnValue:EQuestTrackState Enum=AvailableToTrack)
     B:EQuestTrackState Enum=NotTrackable
    ) -> (ReturnValue:bool))
   > Add(
    TargetArray:Array of Quest UI Quest View Model Object References=Get Quests()
    NewItem:"Quest UI Quest View Model Object Reference (by ref)"=@n3
   ) {is=title K2Node_CallArrayFunction}
  ]}
 GatherAnalyticsImpression {nodes [
   n1 = GatherAnalyticsImpression!() -> (then, VM:Quest UI Base Content View Model Object Reference) {is=title K2Node_FunctionEntry}
   > Is Valid(InputObject=@n1)
   n2 = Get ListView(self=Get QuestTilesList(self=Get QuestsV3ListView()))
   n3 = GetAnalyticsID(self=@n1)
   > Add(
    TargetArray:Array of Flow Analytics Impression Section Content Structures=Get CurrentImpressions()
    NewItem:"Flow Analytics Impression Section Content Structure (by ref)"=Make Flow Analytics Impression Section Content!(
     ContentID:Text="To Text (String)"(InString=@n3)
     Index:int=GetIndexForItem!(self:List View Object Reference=@n2, Item:Object Reference=@n1) -> (ReturnValue:int=0)
    ) -> (FlowAnalyticsImpressionSectionContent:Flow Analytics Impression Section Content Structure) {is=title K2Node_MakeStruct}
   )
  ]}
 GetCurrentContentVMs {nodes [
   GetCurrentContentVMs() {is=title K2Node_FunctionEntry}
   > n1 = Branch(Condition=IsInTrackedCategory())
   n2 = Get QuestUIPlayerVM()
   > Return Node!(
    execute
    ReturnValue:Array of Quest UI Base Content View Model Object References=Get OrderedContentVMs(self:Quest UI Tracker View Model Object Reference=Get TrackerVM!(self:Quest UI Player View Model Object Reference=@n2) -> (TrackerVM:Quest UI Tracker View Model Object Reference))
   )
   n3 = GetSelectedCategoryVM!() -> (AsQuest UI Category View Model:Quest UI Category View Model Object Reference)
   n4 = Is Valid(exec=@n1.else, InputObject=@n3)
   Return Node(execute=@n4.Is Not Valid)
   > Return Node(ReturnValue=Get OrderedContentVMs(self:Quest UI Category View Model Object Reference=@n3))
  ]}
 GetCurrentSelectedWidget {nodes [
   Inputs()
   n1 = Get QuestTilesList(self=Get QuestsV3ListView())
   > n2 = For Each Loop(Array:Array of User Widget Object References=GetDisplayedEntryWidgets!(self:List View Base Object Reference=Get ListView(self=@n1)) -> (ReturnValue:Array of User Widget Object References)) -> (Array Element:User Widget Object Reference)
   > n3 = Cast To WBP_QuestTile!(execute, Object=@n2) -> (then, CastFailed, AsWBP Quest Tile:WBP Quest Tile Object Reference) {is=title K2Node_DynamicCast}
   n4 = Get QuestsV3ListView()
   > Branch(Condition=IsSelectedWidget!(self:Fort Multi Source Polymorphic List View Object Reference=Get QuestTilesList(self=@n4), EntryWidget:User Widget Object Reference=@n3) -> (ReturnValue:bool))
   > Outputs!(FoundWidget:Exec, WBP_QuestTile:WBP Quest Tile Object Reference=@n3, No Widget:Exec=@n2.Completed)
  ]}
 GetDragDropVisPosition {nodes [
   n1 = GetDragDropVisPosition!() -> (then, Widget:Widget Object Reference, Placement:EDragDropPlacementType Enum=NewEnumerator0) {is=title K2Node_FunctionEntry}
   > n2 = Cast To WBP_QuestTile(Object=@n1)
   n3 = GetCachedGeometry!(self:Widget Object Reference=@n2) -> (ReturnValue:"Geometry Structure (by ref)")
   n4 = GetLocalSize!(Geometry:"Geometry Structure (by ref)"=@n3) -> (ReturnValue_X:double, ReturnValue_Y:double)
   n5 = vector2d + vector2d!(
    A:Vector 2D Structure=LocalToViewport!(Geometry:"Geometry Structure (by ref)"=@n3, LocalCoordinate:Vector 2D Structure) -> (PixelPosition:Vector 2D Structure, ViewportPosition:Vector 2D Structure)
    B_X:double=float / float(A=@n4, B=3)
    B_Y:double=float / float(A=@n4.ReturnValue_Y, B=-2)
   ) -> (ReturnValue_X:double, ReturnValue_Y:double) {is=title K2Node_PromotableOperator}
   n6 = Get ListView(self=Get QuestTilesList(self=Get QuestsV3ListView()))
   n7 = GetCachedGeometry(self=@n6)
   n8 = LocalToViewport(Geometry=@n7)
   n9 = vector2d + vector2d(A=@n8.ViewportPosition, B_X=0, B_Y=GetLocalSize(Geometry=@n7).ReturnValue_Y)
   n10 = Get VerticalEntrySpacing!(self:List View Object Reference=@n6) -> (VerticalEntrySpacing:float)
   > Return Node!(
    execute
    ReturnValue_X:double=@n5
    ReturnValue_Y:double="Clamp (Float)"(
     Value=@n5.ReturnValue_Y
     Min=vector2d + vector2d(A=@n8.ViewportPosition, B_X=0, B_Y=@n10).ReturnValue_Y
     Max=float - float(A=@n9.ReturnValue_Y, B=@n10)
    )
   )
   n11 = GetLocalSize(Geometry=GetCachedGeometry())
   Set TileWidth(TileWidth:double=@n11) -> (Output_Get:double)
   > Set TileHeight(TileHeight:double=@n11.ReturnValue_Y) -> (Output_Get:double)
   n12 = LocalToViewport(Geometry=GetCachedGeometry(self=Get LeftArrow_Img!(self:WBP Quest Tile Object Reference) -> (LeftArrow_Img:Image Object Reference)))
   > Set TilePosition(TilePosition:Vector 2D Structure=@n12) -> (Output_Get:Vector 2D Structure)
   n13 = Get TileWidth() -> (TileWidth:double)
   n14 = float / float(A=Get TileHeight() -> (TileHeight:double), B=2)
   n15 = float * float(A=Get TileHeight(), B=-1) {is=title K2Node_PromotableOperator}
   n16 = Get Placement() -> (Placement:EDragDropPlacementType Enum=NewEnumerator0)
   n17 = Get ListView(self=Get QuestTilesList(self=Get QuestsV3ListView()))
   n18 = GetCachedGeometry(self=@n17)
   n19 = LocalToViewport(Geometry=@n18)
   "Clamp (Float)"(
    Value=vector2d + vector2d(
     A=@n12
     B_X=float / float(A=@n13, B=3)
     B_Y=Select!(NewEnumerator0:double=@n14, NewEnumerator1:double=@n15, NewEnumerator2:double=0, Index:EDragDropPlacementType Enum=@n16) -> (ReturnValue:double)
    ).ReturnValue_Y
    Min=vector2d + vector2d(A=@n19, B_X=0, B_Y=Get VerticalEntrySpacing(self=@n17)).ReturnValue_Y
    Max=float + float(A=BreakVector2D(InVec=@n19).Y, B=GetLocalSize(Geometry=@n18).ReturnValue_Y)
   )
   float * float(A=float / float(A=Get TileHeight(), B=4), B=-1)
  ]}
 GetFilteredQuestList {nodes [
   n1 = GetFilteredQuestList!() -> (then, Quests:Array of Quest UI Base Content View Model Object References) {is=title K2Node_FunctionEntry}
   > n2 = Branch(Condition=AND Boolean(A=Get IsTrackedQuests() -> (IsTrackedQuests:bool), B=Get IsHidingDisabled() -> (IsHidingDisabled:bool)))
   > Clear(TargetArray:Array of Quest UI Base Content View Model Object References=Get AvailableQuests() -> (AvailableQuests:Array of Quest UI Base Content View Model Object References))
   > n3 = For Each Loop(Array:Array of Quest UI Base Content View Model Object References=@n1) -> (Array Element:Quest UI Base Content View Model Object Reference)
   n4 = Get Quests() -> (Quests:Array of Quest UI Base Content View Model Object References)
   n5 = Get AvailableQuests()
   Return Node!(
    execute=@n3.Completed
    FilteredQuests:Array of Quest UI Base Content View Model Object References=Get AvailableQuests()
    QuestsRemoved:bool="Not Equal (Integer)"(A=Length(TargetArray=@n5), B=Length(TargetArray=@n4))
   )
   > Branch(Condition=Get bContentEnabled!(self:Quest UI Base Content View Model Object Reference=@n3) -> (bContentEnabled:bool))
   > Add(
    TargetArray:Array of Quest UI Base Content View Model Object References=Get AvailableQuests()
    NewItem:"Quest UI Base Content View Model Object Reference (by ref)"=@n3
   )
   Return Node(execute=@n2.else, FilteredQuests=@n1, QuestsRemoved=false)
  ]}
 GetNextQuest {nodes [
   GetNextQuest!() -> (then, ItemToFind:"Quest UI Quest View Model Object Reference (by ref)") {is=title K2Node_FunctionEntry}
   n1 = Get OrderedContentVMs(self:Quest UI Tracker View Model Object Reference=Get TrackerVM(self=Get QuestUIPlayerVM()))
   n2 = int + int(A=Find Item(
     TargetArray:Array of Quest UI Base Content View Model Object References=@n1
     ItemToFind:"Quest UI Base Content View Model Object Reference (by ref)"=Get ItemToFind() -> (ItemToFind:"Quest UI Quest View Model Object Reference (by ref)")
    ))
   > n3 = Branch(Condition=Is Valid Index!(TargetArray:Array of Quest UI Base Content View Model Object References=@n1, IndexToTest:int=@n2) -> (ReturnValue:bool))
   > n4 = Cast To QuestUIQuestVM(Object="Get (a copy)"(Array=@n1, Dimension 1=@n2))
   > Return Node!(execute, NextQuest:Quest UI Quest View Model Object Reference=@n4)
   Return Node(execute=[@n7.CastFailed, @n6.then, @n4.CastFailed])
   n5 = Get OrderedContentVMs(self=Get TrackerVM(self=Get QuestUIPlayerVM()))
   n6 = Branch(execute=@n3.else, Condition=Is Empty!(TargetArray:Array of Quest UI Base Content View Model Object References=@n5) -> (ReturnValue:bool))
   n7 = Cast To QuestUIQuestVM(execute=@n6.else, Object="Get (a copy)"(Array=@n5))
   > Return Node(NextQuest=@n7)
  ]}
 GetSelectedCategoryVM {nodes [
   GetSelectedCategoryVM() {is=title K2Node_FunctionEntry}
   n1 = Get QuestsV3ListView()
   n2 = Get Selected Item(self=Get ListView(self=Get QuestTilesList(self=@n1)))
   > n3 = Cast To QuestUICategoryVM!(
    execute
    Object="Get (a copy)"(
     Array:Array of MVVM Base Viewmodel Object References=Get CategoriesID()
     Dimension 1=GetSourceIndex!(Item:Object Reference=@n2) -> (ReturnValue:int=0)
    ) -> (Output:MVVM Base Viewmodel Object Reference)
   ) -> (then, CastFailed, AsQuest UI Category View Model:Quest UI Category View Model Object Reference) {is=title K2Node_DynamicCast}
   > Return Node!(execute, AsQuest UI Category View Model:Quest UI Category View Model Object Reference=@n3)
  ]}
 GetSelectedUnwrappedItem {nodes [
   GetSelectedUnwrappedItem() {is=title K2Node_FunctionEntry}
   n1 = Get ListView(self=Get QuestTilesList(self=Get QuestsV3ListView()))
   > Return Node!(execute, ReturnValue:Object Reference=UnwrapItem!(Item:Object Reference=Get Selected Item(self=@n1)) -> (ReturnValue:Object Reference))
  ]}
 GetSortedTrackedQuests {nodes [
   GetSortedTrackedQuests() {is=title K2Node_FunctionEntry}
   > Set HasDisabledTrackedQuest(HasDisabledTrackedQuest:bool=false) -> (Output_Get:bool)
   > n1 = For Each Loop(Array=Get OrderedContentVMs(self=Get TrackedVM_Ref() -> (TrackedVM_Ref:Quest UI Tracker View Model Object Reference)))
   n2 = Get TrackedVMs() -> (TrackedVMs:Array of MVVM Base Viewmodel Object References)
   n3 = Branch(execute=@n1.Completed, Condition="integer > integer"(A=Length(TargetArray:Array of MVVM Base Viewmodel Object References=@n2)))
   > Return Node!(execute, FilteredVMs:Array of MVVM Base Viewmodel Object References=Get TrackedVMs())
   n4 = Get!(execute=@n3.else) -> (then, else, NoTrackedQuestsMessageVM:VM Quest UISpecial Message Object Reference)
   > n5 = Add(
    TargetArray:Array of MVVM Base Viewmodel Object References=Get TrackedVMs()
    NewItem:"MVVM Base Viewmodel Object Reference (by ref)"=Get NoTrackedQuestsMessageVM() -> (NoTrackedQuestsMessageVM:VM Quest UISpecial Message Object Reference)
   )
   Return Node(execute=[@n5.then, @n8.then], FilteredVMs=Get TrackedVMs())
   n6 = Construct Object from Class!(
    execute=@n4.else
    Class=&/VM_QuestUISpecialMessage.VM_QuestUISpecialMessage_C
    self
    Message:Text
    MessageType:EQuestUIMessageType Enum=NoTrackedQuests
   ) -> (then, ReturnValue:VM Quest UISpecial Message Object Reference)
   > n7 = Set NoTrackedQuestsMessageVM(NoTrackedQuestsMessageVM:VM Quest UISpecial Message Object Reference=@n6) -> (Output_Get:VM Quest UISpecial Message Object Reference)
   > n8 = Add(TargetArray=Get TrackedVMs(), NewItem=@n7)
   n9 = Branch(execute=@n1.LoopBody, Condition=Get bContentEnabled(self=@n1))
   > Add(TargetArray=Get TrackedVMs(), NewItem=@n1)
   Set HasDisabledTrackedQuest(execute=@n9.else, HasDisabledTrackedQuest=true)
   > n10 = Branch(Condition=Get IsHidingDisabled())
   Add(execute=@n10.else, TargetArray=Get TrackedVMs(), NewItem=@n1)
  ]}
 GetTabID {nodes [
   n1 = GetTabID!() -> (then, CategoryInterfaceVM:Quest UI Category Interface View Model Object Reference) {is=title K2Node_FunctionEntry}
   > n2 = Is Valid(InputObject=@n1)
   Return Node!(execute=@n2.Is Not Valid, TabID:Name=InvalidTab)
   > Return Node(TabID=String To Name(InString=GetAnalyticsID(self=@n1)))
  ]}
 GetTrackableQuests {nodes [
   GetTrackableQuests() {is=title K2Node_FunctionEntry}
   > n1 = FilterTrackableQuests!(
    execute
    self
    Array:Array of Quest UI Base Content View Model Object References=GetFilteredQuestList!(Quests:Array of Quest UI Base Content View Model Object References=GetCurrentContentVMs!() -> (ReturnValue:Array of Quest UI Base Content View Model Object References) {is=title K2Node_CallFunction}) -> (FilteredQuests:Array of Quest UI Base Content View Model Object References, QuestsRemoved:bool) {is=title K2Node_CallFunction}
   ) -> (then, TrackableQuests:Array of Quest UI Quest View Model Object References) {is=title K2Node_CallFunction}
   > Return Node!(execute, Ordered Quest VMs:Array of Quest UI Quest View Model Object References=@n1)
  ]}
 HandleActiveColumnVMChanged {nodes [
   n1 = HandleActiveColumnVMChanged!() -> (then, ColumnVM:Quest UI Base Content View Model Object Reference) {is=title K2Node_FunctionEntry}
   > SetRowSelection!(
    execute
    self:WBP Selection Panel Object Reference=Get SelectionPanel()
    RowSelectionVM:Quest UI Base Content View Model Object Reference=@n1
   ) -> (then)
  ]}
 HandleBonusGoalsDoubleClick {nodes [
   HandleBonusGoalsDoubleClick!() -> (then, GoalsQuestsHorizontal:WBP Goals Progress Horizontal Object Reference) {is=title K2Node_FunctionEntry}
   n1 = GetTrackableQuests!() -> (Ordered Quest VMs:Array of Quest UI Quest View Model Object References) {is=title K2Node_CallFunction}
   > Branch(Condition="integer > integer"(A=Length(TargetArray:Array of Quest UI Quest View Model Object References=@n1), B=1))
   > HandleTrackAllQuestClicked()
  ]}
 HandleHideUnavailableClicked {nodes [
   HandleHideUnavailableClicked() {is=title K2Node_FunctionEntry}
   > Set IsHidingDisabled(IsHidingDisabled:bool=NOT Boolean(A=Get IsHidingDisabled())) -> (Output_Get:bool)
   n1 = Get DecoyButtonHideDisabled() -> (DecoyButtonHideDisabled:WBP Bottom Bar Decoy Button Object Reference)
   > SetTriggeringInputAction!(
    execute
    self:Common Button Base Object Reference=@n1
    InputActionRow:"Data Table Row Handle Structure (by ref)"=Select!(
     Option 0:"Data Table Row Handle Structure (by ref)"=Make Data Table Row Handle!(DataTable:Data Table Object Reference=/QuestUI/QuestScreen/RewardsPanel/QuestInputActionDataTable, RowName:Name=HideDisabled) -> (DataTableRowHandle:Data Table Row Handle Structure) {is=title K2Node_MakeStruct}
     Option 1:"Data Table Row Handle Structure (by ref)"=Make Data Table Row Handle(DataTable=&QuestInputActionDataTable, RowName=ShowDisabled)
     Index:bool=Get IsHidingDisabled()
    ) -> (ReturnValue:"Data Table Row Handle Structure (by ref)")
   ) -> (then)
   > CacheInputAction(self:WBP Bottom Bar Decoy Button Object Reference=@n1)
   > n2 = Cast To QuestUIQuestVM(Object=GetSelectedUnwrappedItem() {is=title K2Node_CallFunction})
   > n3 = Set InitialSelection(InitialSelection:Quest UI Quest View Model Object Reference=@n2) -> (Output_Get:Quest UI Quest View Model Object Reference)
   UpdateMoveQuestButtonVisibility(execute=[@n3.then, @n2.CastFailed])
   > UpdateTrackAllButtonState(ForceUntrackCategory=false)
  ]}
 HandleItemDoubleClicked {nodes [
   HandleItemDoubleClicked!() -> (then, QuestTile:WBP Quest Tile Object Reference) {is=title K2Node_FunctionEntry}
   > HandleTrackQuestClicked() -> (SelectedItemIndex1=0)
  ]}
 HandleItemHover {nodes [
   HandleItemHover!() -> (then, QuestTile:WBP Quest Tile Object Reference) {is=title K2Node_FunctionEntry}
   > Branch(Condition=GetConfigDataBool!(EntryName:String=quest_category_tab_ui_optimization) -> (bValueOut:bool, ReturnValue:bool))
   n1 = Get SideNavigation!() -> (SideNavigation:WBP Quest V3 Side Navigation Object Reference)
   > n2 = Branch(Condition=Get SideNavigationIsCollapsed!(self:WBP UIKit Side Navigation Base Object Reference=Get SideNavigation!(self:WBP Quest V3 Side Navigation Object Reference=@n1) -> (SideNavigation:WBP UIKit Side Navigation Object Reference)) -> (SideNavigationIsCollapsed:bool))
   SetIsCollapsed(
    execute=@n2.else
    self=@n1
    IsCollapsed=true
    SkipAnimation=IsInZone!(self:Fort Global UIContext Object Reference=Get FortGlobalUIContext() -> (ReturnValue:Fort Global UIContext Object Reference) {is=title K2Node_GetSubsystem}) -> (ReturnValue:bool)
   )
  ]}
 HandleQuestDroppedFromTile {nodes [
   HandleQuestDroppedFromTile!() -> (then, ClickedQuest:WBP Quest Tile Object Reference, WasDropped:bool=~) {is=title K2Node_FunctionEntry}
   > n1 = Branch Input Type!(In:Exec) -> (GamepadOrTouch:Exec, MnK:Exec) {is=title K2Node_MacroInstance}
   Branch(execute=@n1.MnK, Condition=Get WasDropped() -> (WasDropped:bool))
   n2 = Get ClickedQuest() -> (ClickedQuest:WBP Quest Tile Object Reference)
   > n3 = Is Valid(InputObject=@n2)
   SetMoveMode(execute=@n3.Is Not Valid, IsMoveMode=false, IsShutdown=false)
   ReorganizeQuestList!(execute=@n3.Is Valid, self, DroppedQuest:WBP Quest Tile Object Reference=@n2) -> (then)
   Branch(execute=@n1.GamepadOrTouch, Condition=Get WasDropped())
   > Set TargetQuestTile(TargetQuestTile:WBP Quest Tile Object Reference=Get ClickedQuest()) -> (Output_Get:WBP Quest Tile Object Reference)
   > HandleQuestDroppedGamepad()
  ]}
 HandleQuestDroppedGamepad {nodes [
   HandleQuestDroppedGamepad() {is=title K2Node_FunctionEntry}
   n1 = Get SourceQuestTile() -> (SourceQuestTile:WBP Quest Tile Object Reference)
   n2 = Get TargetQuestTile()
   > n3 = Branch(Condition=AND Boolean(A=IsValid(Object=@n1), B=IsValid(Object=@n2)))
   n4 = Get TrackerVM(self=Get QuestUIPlayerVM())
   n5 = Find Item(
    TargetArray=Get OrderedContentVMs(self=@n4)
    ItemToFind=Get QuestUIQuestVM!(self:WBP Quest Tile Object Reference=Get TargetQuestTile()) -> (QuestUIQuestVM:Quest UI Quest View Model Object Reference)
   )
   > SetIsTracked!(
    execute
    self:Quest UI Quest View Model Object Reference=Get QuestUIQuestVM(self=Get SourceQuestTile())
    TrackedIndex:int=Select(
     Option 0:int=@n5
     Option 1:int=GetCorrectedTrackedIndex!(self:Quest UI Tracker View Model Object Reference=@n4, QuestVMIndex:int=@n5) -> (ReturnValue:int=0)
     Index=UseCorrectedTrackedIndex!() -> (UseCorrected:bool)
    ) -> (ReturnValue:int)
   ) -> (then)
   > ReorganizeQuestList(DroppedQuest=Get SourceQuestTile())
   SetMoveMode(execute=@n3.else, IsMoveMode=false, IsShutdown=false)
  ]}
 HandleQuestNavigate {nodes [
   n1 = HandleQuestNavigate!() -> (then, QuestToReturnFocus:User Widget Object Reference, Direction:EUINavigation Enum) {is=title K2Node_FunctionEntry}
   > n2 = Shutdown Modes()
   Set ReturnFocusTile(execute=@n2.NoShutdownRequired, ReturnFocusTile:User Widget Object Reference=@n1) -> (Output_Get:User Widget Object Reference)
   > n3 = Branch(Condition="Equal (Enum)"(A:EUINavigation Enum=Get Direction() -> (Direction:EUINavigation Enum), B:EUINavigation Enum=Right))
   > HandleFocusRequest(self:WBP Selection Panel Object Reference=Get SelectionPanel())
   SetFocus(execute=@n3.else, self=GetSelectedButton(self=Get SideNavigation!() -> (SideNavigation:WBP Quest V3 Side Navigation Object Reference)))
  ]}
 HandleSelectionChanged {nodes [
   n1 = HandleSelectionChanged!() -> (then, Item:Object Reference, bIsSelected:bool=~) {is=title K2Node_FunctionEntry}
   > Branch(Condition=@n1.bIsSelected)
   > SetBottomButtonVisibility(self=Get DecoyButtonExternalLink(), ShouldButtonBeVisible=false)
   > SetBackToSideNavigationVisibiliity!(execute, self:WBP Quest V3 Side Navigation Object Reference=Get SideNavigation(), IsVisible:bool=NOT Boolean(A=Get IsMoving())) -> (then)
   > n2 = Sequence()
   n3 = UnwrapItem(Item=@n1)
   > SetActiveSelection!(execute, self:WBP Selection Panel Object Reference=Get SelectionPanel(), SelectionVM:Object Reference=@n3) -> (then)
   > n4 = Cast To QuestUIHorizontalPunchCardVM!(execute, Object=@n3) -> (then, CastFailed, AsQuest UI Horizontal Punch Card View Model:Quest UI Horizontal Punch Card View Model Object Reference) {is=title K2Node_DynamicCast}
   > n5 = Set Quest UI Horizontal Punch Card View Model(Quest UI Horizontal Punch Card View Model:Quest UI Horizontal Punch Card View Model Object Reference=@n4) -> (Output_Get:Quest UI Horizontal Punch Card View Model Object Reference)
   UpdateGamepadDropTarget!(execute=[@n9.then, @n5.then], self, NewSelectedItem:Object Reference=@n3) -> (then)
   > UpdateTrackedQuestButton!(execute, self, ListItem:Object Reference=@n3) -> (then)
   > UpdateHideDisabledVisibility()
   > n6 = Cast To QuestUIBaseContentVM(Object=@n3)
   > n7 = GatherAnalyticsImpression!(execute, self, VM:Quest UI Base Content View Model Object Reference=@n6) -> (then) {is=title K2Node_CallFunction}
   n8 = Get QuestTilesList(self=Get QuestsV3ListView())
   OnListScrolled!(
    execute=[@n7.then, @n6.CastFailed]
    self
    ItemOffset:double=GetScrollOffset!(self:List View Base Object Reference=Get ListView(self=@n8)) -> (ReturnValue:float)
    DistanceRemaining:double=0
   ) -> (then)
   n9 = Set Quest UI Horizontal Punch Card View Model(execute=@n4.CastFailed)
   n10 = Get Class(Object="Get (a copy)"(Array=Get CategoriesID(), Dimension 1=GetSourceIndex(Item=@n1) -> (ReturnValue=0))) -> (ReturnValue:MVVM Base Viewmodel Class Reference)
   Branch(
    execute=@n2.then_1
    Condition=AND Boolean(
     A=Get IsMoving()
     B="Not Equal (Class)"(A:Object Class Reference=@n10, B:Object Class Reference=/Script/QuestUI.QuestUITrackerVM) -> (ReturnValue:bool=~) {is=title K2Node_PromotableOperator}
    )
   )
   > Shutdown Modes()
  ]}
 HandleTrackAllQuestClicked {nodes [
   HandleTrackAllQuestClicked() {is=title K2Node_FunctionEntry}
   > n1 = Set QuestArray(QuestArray:Array of Quest UI Quest View Model Object References=GetTrackableQuests()) -> (Output_Get:Array of Quest UI Quest View Model Object References)
   > Set NumQuests(NumQuests:int=Length(TargetArray=@n1)) -> (Output_Get:int=0)
   > n2 = For Each Loop with Break!(
    Exec:Exec
    Array:Array of Quest UI Base Content View Model Object References=Get QuestArray() -> (QuestArray:Array of Quest UI Quest View Model Object References)
    Break:Exec=@n7.then
   ) -> (LoopBody:Exec, Array Element:Quest UI Base Content View Model Object Reference, Array Index:int, Completed:Exec)
   UpdateCTAButtons(execute=@n2.Completed)
   > RefreshTrackedQuests()
   > n3 = Branch(Condition="integer >= integer"(A=Get CannotTrackIndex() -> (CannotTrackIndex:int=0)))
   n4 = Get CannotTrackIndex() -> (CannotTrackIndex=0)
   > PushTrackedQuestCapReachedDialog!(execute, self, NumberExceeded:int=int - int(A=Get NumQuests() -> (NumQuests:int=0), B=int + int(A=@n4))) -> (then)
   > UpdateTrackAllButtonState(ForceUntrackCategory=true)
   PlaySound2D(
    execute=@n3.else
    Sound=Select(
     Option 0:Sound Base Object Reference=/Game/Sounds/Fort_UI_Sounds/FrontEnd/New_FrontEnd/MSS_Fort_UI_QuestScreen_QuestSelected
     Option 1:Sound Base Object Reference=/Game/Sounds/Fort_UI_Sounds/FrontEnd/New_FrontEnd/MSS_Fort_UI_QuestScreen_QuestDeSelected
     Index=Get IsTrackAll() -> (IsTrackAll:bool)
    ) -> (ReturnValue:Sound Base Object Reference)
   )
   n5 = Cast To QuestUIQuestVM(execute=@n2.LoopBody, Object=@n2)
   > n6 = SetQuestTracking!(execute, self, Quest:Quest UI Quest View Model Object Reference=@n5, ShouldTrack:bool=Get IsTrackAll()) -> (then, AtQuestLimit:bool, InitialSelection:Quest UI Quest View Model Object Reference)
   > Branch(Condition=@n6)
   > n7 = Set CannotTrackIndex(CannotTrackIndex:int=@n2.Array Index) -> (Output_Get:int=0)
  ]}
 HandleTrackQuestClicked {nodes [
   HandleTrackQuestClicked() {is=title K2Node_FunctionEntry}
   > Set SelectedObject(SelectedObject:Object Reference=GetSelectedUnwrappedItem()) -> (Output_Get:Object Reference)
   n1 = Get ListView(self=Get QuestTilesList(self=Get QuestsV3ListView()))
   > Set SelectedItemIndex(SelectedItemIndex:int=GetIndexForItem(self=@n1, Item=Get Selected Item(self=@n1)) -> (ReturnValue=0)) -> (Output_Get:int=0)
   > n2 = Sequence()
   > n3 = Cast To QuestUIQuestVM(Object=Get SelectedObject() -> (SelectedObject:Object Reference))
   > n4 = Set ShouldTrack(ShouldTrack:bool="Equal (Enum)"(A:EQuestTrackState Enum=GetQuestTrackState(self=@n3) -> (ReturnValue=AvailableToTrack), B:EQuestTrackState Enum=AvailableToTrack)) -> (Output_Get:bool)
   > n5 = SetQuestTracking(Quest=@n3, ShouldTrack=@n4)
   > Set HitLimit(HitLimit:bool=@n5) -> (Output_Get:bool)
   > RefreshTrackedQuests()
   > n6 = Branch(Condition=Get HitLimit() -> (HitLimit:bool))
   > n7 = PushTrackedQuestCapReachedDialog(NumberExceeded=0)
   PlaySound2D(
    execute=[@n7.then, @n6.else]
    Sound=Select(
     Option 0=&MSS_Fort_UI_QuestScreen_QuestDeSelected
     Option 1=&MSS_Fort_UI_QuestScreen_QuestSelected
     Index=Get ShouldTrack() -> (ShouldTrack:bool)
    )
   )
   SetActiveSelection(execute=@n2.then_1, self=Get SelectionPanel(), SelectionVM=Get SelectedObject())
   > UpdateCTAButtons()
   > Return Node!(execute, SelectedItemIndex1:int=Get SelectedItemIndex() -> (SelectedItemIndex:int=0))
  ]}
 InitSideNavigation {nodes [
   n1 = InitSideNavigation!() -> (then, IsCollapsed:bool=true) {is=title K2Node_FunctionEntry}
   > SetIsCollapsed(self:WBP Quest V3 Side Navigation Object Reference=Get SideNavigation(), IsCollapsed=@n1, SkipAnimation=true)
   > RequestRefreshFocus()
  ]}
 IsDescriptionNeeded {nodes [
   n1 = IsDescriptionNeeded!() -> (then, InputPin:Quest UI Category View Model Object Reference) {is=title K2Node_FunctionEntry}
   > n2 = GetMcpRelativeUtcNow!(
    execute
    FortPC:Fort Player Controller Object Reference=GetFirstLocalFortPlayerController!(bIncludeNonSpectators:bool=true, bIncludeSpectators:bool=false) -> (ReturnValue:Fort Player Controller Object Reference)
   ) -> (then, ReturnValue:Date Time Structure)
   n3 = "Timespan > Timespan"(
    A:Timespan Structure=DateTime - DateTime!(
     A:Date Time Structure=Get ExpirationTime!(self:Quest UI Category View Model Object Reference=@n1) -> (ExpirationTime:Date Time Structure)
     B:Date Time Structure=@n2
    ) -> (ReturnValue:Timespan Structure)
    B:Timespan Structure=MakeTimespan!(Days:int=0, Hours:int=0, Minutes:int=0, Seconds:int=1, Milliseconds:int=0) -> (ReturnValue:Timespan Structure)
   ) -> (ReturnValue:bool=~) {is=title K2Node_PromotableOperator}
   n4 = TextIsEmpty(InText=GetCategoryDescription!(self:Quest UI Category View Model Object Reference=@n1) -> (ReturnValue:Text))
   n5 = NOT Boolean(A=Get bHasExpirationTime!(self:Quest UI Category View Model Object Reference=@n1) -> (bHasExpirationTime:bool))
   > Return Node!(execute, IsValid:bool=NOR Boolean!(A:bool=AND Boolean(A=@n4, B=@n5), B:bool=AND Boolean(A=@n4, B=@n3)) -> (ReturnValue:bool))
  ]}
 IsInFrontend {nodes [
   IsInFrontend() {is=title K2Node_FunctionEntry}
   > n1 = Cast To FortPlayerControllerFrontEnd!(execute, Object=GetOwningPlayer()) -> (then, CastFailed, AsFort Player Controller Front End:Fort Player Controller Front End Object Reference) {is=title K2Node_DynamicCast}
   > Return Node!(execute, InFrontend:bool=true)
   Return Node(execute=@n1.CastFailed, InFrontend=false)
  ]}
 IsInTrackedCategory {nodes [
   IsInTrackedCategory() {is=title K2Node_FunctionEntry}
   n1 = Get ListView(self=Get QuestTilesList(self=Get QuestsV3ListView()))
   n2 = "Get (a copy)"(Array=Get CategoriesID(), Dimension 1=GetSourceIndex(Item=Get Selected Item(self=@n1)) -> (ReturnValue=0))
   > Is Valid(InputObject=@n2)
   > Return Node!(execute, ReturnValue:bool="Equal (Class)"(A=Get Class(Object=@n2), B=&QuestUITrackerVM))
  ]}
 MarkActiveCategorySeen {nodes [
   n1 = MarkActiveCategorySeen!() -> (then, CommitSeenOnly:bool=~, Category:Quest UI Base View Model Object Reference) {is=title K2Node_FunctionEntry}
   > n2 = Branch(Condition=@n1)
   > CommitSeen(self:Quest UI Base View Model Object Reference=@n1.Category)
   MarkSeen!(execute=@n2.else, self:Quest UI Base View Model Object Reference=@n1.Category, bCommitToMcp:bool=true) -> (then)
  ]}
 OnDrop {nodes [
   OnDrop!() -> (then, MyGeometry:Geometry Structure, PointerEvent:Pointer Event Structure, Operation:Drag Drop Operation Object Reference) {is=title K2Node_FunctionEntry}
   > SetMoveMode(IsMoveMode=false, IsShutdown=false)
   > Return Node(ReturnValue=false)
  ]}
 OnInputTypeChanged {nodes [
   OnInputTypeChanged!() -> (then, bNewInputType:ECommonInputType Enum) {is=title K2Node_FunctionEntry}
   n1 = Get bNewInputType() -> (bNewInputType:ECommonInputType Enum)
   > SetVisibility(
    self=Get WBox_MoveModeInputs() -> (WBox_MoveModeInputs:Wrap Box Object Reference)
    InVisibility=Select(
     Option 0:ESlateVisibility Enum=SelfHitTestInvisible
     Option 1:ESlateVisibility Enum=Collapsed
     Index="Equal (Enum)"(A:ECommonInputType Enum=@n1, B:ECommonInputType Enum=MouseAndKeyboard)
    ) -> (ReturnValue:ESlateVisibility Enum)
   )
   > UpdateMoveQuestButtonVisibility()
   > n2 = Shutdown Modes()
   n3 = Switch on ECommonInputType!(execute=[@n2.NoShutdownRequired, @n2.Shutdown], Selection:ECommonInputType Enum=Get bNewInputType()) -> (MouseAndKeyboard:Exec, Gamepad:Exec, Touch:Exec, Count:Exec) {is=title K2Node_SwitchEnum}
   n4 = Get ListView(self=Get QuestTilesList(self=Get QuestsV3ListView()))
   n5 = Get Selected Item(self=@n4)
   Is Valid(exec=@n3.Gamepad, InputObject=@n5)
   > Navigate To Item!(execute, self:List View Object Reference=@n4, Item:Object Reference=@n5) -> (then)
   > SetFocus(self=@n4)
  ]}
 OnKeyDown {nodes [
   n1 = Cast To FortPlayerController(Object=GetOwningPlayer())
   n2 = GetKey!(Input:"Key Event Structure (by ref)"=Get InKeyEvent() -> (InKeyEvent:Key Event Structure)) -> (ReturnValue:Key Structure)
   n3 = GetKeysBoundToAction!(
    self:Fort Player Controller Object Reference=@n1
    Action:Name=Get GamepadToggleFullScreenMap() -> (GamepadToggleFullScreenMap:Name=None)
    OverrideInputType:ECommonInputType Enum=Gamepad
    PresetNameOverride:Name=None
   ) -> (OutPrimaryKey:Key Structure, OutSecondaryKey:Key Structure, ReturnValue:bool)
   n4 = GetKeysBoundToAction(
    self=@n1
    Action=Get ToggleFullScreenMapKey() -> (ToggleFullScreenMapKey:Name=None)
    OverrideInputType=MouseAndKeyboard
    PresetNameOverride=None
   )
   > n5 = Branch(Condition=OR Boolean!(
     A="Equal (Key)"(A:Key Structure=@n2, B:Key Structure=@n4) -> (ReturnValue:bool=~) {is=title K2Node_PromotableOperator}
     B="Equal (Key)"(A=@n2, B=@n4.OutSecondaryKey) -> (ReturnValue=~)
     C:bool="Equal (Key)"(A=@n2, B=@n3) -> (ReturnValue=~)
     D:bool="Equal (Key)"(A=@n2, B=@n3.OutSecondaryKey) -> (ReturnValue=~)
    ) -> (ReturnValue))
   > n6 = Set IsExiting(IsExiting:bool=true) -> (Output_Get:bool)
   Return Node(
    execute=[@n6.then, @n5.else]
    ReturnValue:Event Reply Structure=Make Event Reply!() -> (EventReply:Event Reply Structure) {is=title K2Node_MakeStruct}
   )
   OnKeyDown!() -> (then, MyGeometry:Geometry Structure, InKeyEvent:Key Event Structure) {is=title K2Node_FunctionEntry}
  ]}
 OnKeyUp {nodes [
   Branch(Condition=Get IsExiting() -> (IsExiting:bool))
   > Set IsExiting(IsExiting=false)
   > On Exit()
   > Return Node(ReturnValue=Make Event Reply())
   OnKeyUp!() -> (then, MyGeometry:Geometry Structure, InKeyEvent:Key Event Structure) {is=title K2Node_FunctionEntry}
  ]}
 OnListScrolled {nodes [
   n1 = OnListScrolled!() -> (then, ItemOffset:double=~, DistanceRemaining:double=~) {is=title K2Node_FunctionEntry}
   n2 = Truncate(A=@n1)
   > SetBottomButtonVisibility(
    self=Get DecoyButtonGoToTop() -> (DecoyButtonGoToTop:WBP Bottom Bar Decoy Button Object Reference)
    ShouldButtonBeVisible="integer > integer"(A=@n2)
   )
   > n3 = Sequence()
   n4 = int - int(A=@n2)
   n5 = Get QuestTilesList(self=Get QuestsV3ListView())
   n6 = GetSourceIndex(Item=GetItemAt!(self:List View Object Reference=Get ListView(self=@n5), Index:int="Max (Integer)"(A=@n4, B=0)) -> (ReturnValue:Object Reference)) -> (ReturnValue=0)
   > Branch(Condition="Not Equal (Integer)"(A=Get PreviousSource() -> (PreviousSource:int=0), B=@n6))
   > Set PreviousSource(PreviousSource:int=@n6) -> (Output_Get:int=0)
   n7 = "Get (a copy)"(Array=Get CategoriesID(), Dimension 1=@n6)
   > n8 = Cast To QuestUICategoryVM(Object=@n7)
   n9 = Get HeaderSwitcher() -> (HeaderSwitcher:Common Activatable Widget Switcher Object Reference)
   n10 = Get WBP_QuestsSeparate_Header() -> (WBP_QuestsSeparate_Header:WBP Quests Separate Header Object Reference)
   > SetActiveWidget(self:Widget Switcher Object Reference=@n9, Widget=@n10)
   > SetQuestUICategoryVM!(execute, self:WBP Quests Separate Header Object Reference=@n10, Viewmodel:Quest UI Category View Model Object Reference=@n8) -> (then)
   n11 = Cast To QuestUITrackerVM!(execute=@n8.CastFailed, Object=@n7) -> (then, CastFailed, AsQuest UI Tracker View Model:Quest UI Tracker View Model Object Reference) {is=title K2Node_DynamicCast}
   n12 = Get WBP_QuestsSeparate_Tracker() -> (WBP_QuestsSeparate_Tracker:WBP Quests Separate Tracker Object Reference)
   > SetActiveWidget(self=@n9, Widget=@n12)
   > SetQuestUITrackerVM!(execute, self:WBP Quests Separate Tracker Object Reference=@n12, Viewmodel:Quest UI Tracker View Model Object Reference=@n11) -> (then)
   n13 = Cast To QuestUICustomCategoryVM!(execute=@n11.CastFailed, Object=@n7) -> (then, CastFailed, AsQuest UI Custom Category View Model:Quest UI Custom Category View Model Object Reference) {is=title K2Node_DynamicCast}
   n14 = Get WBP_QuestsSeparate_CustomHeader() -> (WBP_QuestsSeparate_CustomHeader:WBP Quests Separate Custom Header Object Reference)
   > SetActiveWidget(self=@n9, Widget=@n14)
   > SetQuestUICustomCategoryVM!(execute, self:WBP Quests Separate Custom Header Object Reference=@n14, Viewmodel:Quest UI Custom Category View Model Object Reference=@n13) -> (then)
   n15 = Get PreviousSideNavCategoryVM() -> (PreviousSideNavCategoryVM:MVVM Base Viewmodel Object Reference)
   n16 = "Get (a copy)"(
    Array:Array of Quest UI Base Content View Model Object References=Get SideNavCategoriesVM() -> (SideNavCategoriesVM:Array of Quest UI Base Content View Model Object References)
    Dimension 1=@n6
   ) -> (Output:Quest UI Base Content View Model Object Reference)
   n17 = Get SideNavigation()
   Branch(
    execute=@n3.then_1
    Condition=AND Boolean(
     A="Not Equal (Object)"(A:Object Reference=@n16, B:Object Reference=@n15) -> (ReturnValue:bool=~) {is=title K2Node_PromotableOperator}
     B=Get IsCollapsed(self=@n17)
    )
   )
   > SelectSideNavigationViewModel!(
    execute
    self:WBP UIKit Side Navigation Base Object Reference=Get SideNavigation!(self:WBP Quest V3 Side Navigation Object Reference=Get SideNavigation!() -> (SideNavigation:WBP Quest V3 Side Navigation Object Reference)) -> (SideNavigation:WBP UIKit Side Navigation Object Reference)
    TabViewModel:MVVM Base Viewmodel Object Reference=@n16
    bSuppressClickFeedback:bool=false
   ) -> (then, Success:bool)
   > Set PreviousSideNavCategoryVM(PreviousSideNavCategoryVM=@n16)
  ]}
 OnPunchCardRewardViewChanged {nodes [
   n1 = OnPunchCardRewardViewChanged!() -> (then, InItemView:bool=~) {is=title K2Node_FunctionEntry}
   > UpdateIsShowingRewards!(execute, self:WBP Selection Panel Object Reference=Get SelectionPanel(), InShowingRewards:bool=@n1) -> (then)
   > n2 = Get!(execute) -> (then, else, Quest UI Horizontal Punch Card View Model:Quest UI Horizontal Punch Card View Model Object Reference)
   n3 = IsValidSoftClassReference!(SoftClassReference:"Object Soft Class Reference (by ref)"=Get CustomMapWidget!(self:Quest UI Punch Card View Model Object Reference=@n2) -> (CustomMapWidget:User Widget Soft Class Reference)) -> (ReturnValue:bool)
   n4 = Get InItemView() -> (InItemView:bool)
   > SetBackToSideNavigationVisibiliity(self=Get SideNavigation(), IsVisible=NOT Boolean(A=AND Boolean(A=@n3, B=@n4)))
  ]}
 OnPunchCardSelectedIndexChanged {nodes [
   n1 = OnPunchCardSelectedIndexChanged!() -> (then, SelectedIndex:int) {is=title K2Node_FunctionEntry}
   > SetRewardPage!(execute, self:WBP Selection Panel Object Reference=Get SelectionPanel(), Index:int=@n1) -> (then)
  ]}
 OnShouldDisplayBackActionInActionBar {nodes [
   OnShouldDisplayBackActionInActionBar() {is=title K2Node_FunctionEntry}
   > Return Node!(execute, bOutShouldDisplayBackActionInActionBar:bool=true, ReturnValue:bool=true)
  ]}
 OnTouchEnded {nodes [
   OnTouchEnded!() -> (then, MyGeometry:Geometry Structure, InTouchEvent:"Pointer Event Structure (by ref)") {is=title K2Node_FunctionEntry}
   > Return Node!(execute, ReturnValue:Event Reply Structure=Handled())
  ]}
 OnTouchStarted {nodes [
   OnTouchStarted!() -> (then, MyGeometry:Geometry Structure, InTouchEvent:"Pointer Event Structure (by ref)") {is=title K2Node_FunctionEntry}
   > Return Node(ReturnValue=Handled())
  ]}
 PopulateCategories {nodes [
   n1 = PopulateCategories!() -> (then, ProductList:Array of Quest UI Product View Model Object References) {is=title K2Node_FunctionEntry}
   > Branch(Condition="integer > integer"(A=Length(TargetArray:Array of Quest UI Product View Model Object References=@n1)))
   > n2 = Sequence!(execute) -> (then_0, then_1, then_2:Exec)
   > Set QuestUIProductVM(QuestUIProductVM:Quest UI Product View Model Object Reference="Get (a copy)"(Array:Array of Quest UI Product View Model Object References=@n1) -> (Output:Quest UI Product View Model Object Reference)) -> (Output_Get:Quest UI Product View Model Object Reference)
   ClearSideNavigationTabs(
    execute=@n2.then_1
    self:WBP UIKit Side Navigation Base Object Reference=Get SideNavigation!(self:WBP Quest V3 Side Navigation Object Reference=Get SideNavigation!() -> (SideNavigation:WBP Quest V3 Side Navigation Object Reference)) -> (SideNavigation:WBP UIKit Side Navigation Object Reference)
   )
   > Clear(TargetArray:Array of S UI Tab Container Data Structures=Get TabsData() -> (TabsData:Array of S UI Tab Container Data Structures))
   n3 = Get OrderedBucketVMs!(self:Quest UI Product View Model Object Reference=Get QuestUIProductVM()) -> (OrderedBucketVMs:Array of Quest UI Bucket View Model Object References)
   > n4 = For Each Loop(Array:Array of Quest UI Bucket View Model Object References=@n3) -> (Array Element:Quest UI Bucket View Model Object Reference)
   n5 = Get SideNavigation!(self:WBP Quest V3 Side Navigation Object Reference=Get SideNavigation!() -> (SideNavigation:WBP Quest V3 Side Navigation Object Reference)) -> (SideNavigation:WBP UIKit Side Navigation Object Reference)
   n6 = Get TabsData()
   PopulateSideNavigationTabs!(execute=@n4.Completed, self:WBP UIKit Side Navigation Base Object Reference=@n5, Data:Array of S UI Tab Container Data Structures=@n6) -> (then)
   n7 = Length(TargetArray:Array of S UI Tab Container Data Structures=@n6)
   > SetVisibility(self=Get SideNavigation(), InVisibility=Select(Option 0=Collapsed, Option 1=SelfHitTestInvisible, Index="integer > integer"(A=@n7, B=1)))
   > SelectTabByIndex!(execute, self:UIKit Tab Group Object Reference=@n5, Index:int=0, bSuppressClickFeedback:bool=false) -> (then, ReturnValue:bool)
   > UpdateTrackAllButtonState(ForceUntrackCategory=false)
   Clear(execute=@n4.LoopBody, TargetArray:Array of UIKit Tab Data Structures=Get Categories() -> (Categories:Array of UIKit Tab Data Structures))
   > n8 = For Each Loop(Array:Array of Quest UI Base Content View Model Object References=Get OrderedContentVMs(self:Quest UI Bucket View Model Object Reference=@n4)) -> (Array Element:Quest UI Base Content View Model Object Reference)
   Add!(
    execute=@n8.Completed
    TargetArray:Array of S UI Tab Container Data Structures=Get TabsData()
    NewItem_Tab_6_#4_TabID:Name=None
    NewItem_Tab_6_#4_TabViewModel:MVVM Base Viewmodel Object Reference=@n4
    NewItem_InnerTabs_7_#5:Array of UIKit Tab Data Structures=Get Categories()
   ) -> (then, ReturnValue)
   n9 = Cast To QuestUITrackerVM(execute=@n8.LoopBody, Object=@n8)
   Add!(
    execute=@n9.CastFailed
    TargetArray:Array of UIKit Tab Data Structures=Get Categories()
    NewItem_TabID:Name=None
    NewItem_TabViewModel:MVVM Base Viewmodel Object Reference=@n8
   ) -> (then, ReturnValue)
   PopulateQuestsList!(execute=@n2.then_2, self, Buckets:Array of Quest UI Bucket View Model Object References=@n3) -> (then)
   Set DebugTabs(DebugTabs:Array of S UI Tab Container Data Structures=@n6) -> (Output_Get:Array of S UI Tab Container Data Structures)
  ]}
 PopulateQuestsList {nodes [
   n1 = PopulateQuestsList!() -> (then, Buckets:Array of Quest UI Bucket View Model Object References) {is=title K2Node_FunctionEntry}
   > Clear(TargetArray:Array of Fort Poly View Source Data Structures=Get MultiSourceList() -> (MultiSourceList:Array of Fort Poly View Source Data Structures))
   > Clear(TargetArray:Array of MVVM Base Viewmodel Object References=Get CategoriesID())
   > Clear(TargetArray:Array of Quest UI Base Content View Model Object References=Get SideNavCategoriesVM())
   > n2 = For Each Loop(Array:Array of Quest UI Bucket View Model Object References=@n1) -> (Array Element:Quest UI Bucket View Model Object Reference)
   n3 = Get EndListSpaceVM() -> (EndListSpaceVM:Array of Object References)
   n4 = Branch(execute=@n2.Completed, Condition="integer > integer"(A=Length(TargetArray:Array of Object References=@n3)))
   n5 = Get MultiSourceList()
   Append Array!(
    execute=[@n4.then, @n7.Completed]
    TargetArray:Array of Object References="Get (a ref)"!(
     Array:Array of Fort Poly View Source Data Structures=@n5
     Dimension 1=Last Index!(TargetArray:Array of Fort Poly View Source Data Structures=@n5) -> (ReturnValue=0)
    ) -> (Output_SourceIndex:int, Output_Items:Array of Object References).Output_Items
    SourceArray:Array of Object References=@n3
   ) -> (then)
   > SetAllSourceListItems!(
    execute
    self:Fort Multi Source Polymorphic List View Object Reference=Get QuestTilesList(self=Get QuestsV3ListView())
    AllSources:Array of Fort Poly View Source Data Structures=Get MultiSourceList()
   ) -> (then)
   n6 = Get QuestTilesList(self=Get QuestsV3ListView())
   > OnListScrolled(ItemOffset=GetScrollOffset(self=Get ListView(self=@n6)), DistanceRemaining=0) {is=title K2Node_CallFunction}
   > SelectInitialContent()
   > RequestRefreshFocus()
   n7 = For Loop(execute=@n4.else, LastIndex=5)
   > n8 = CreateSpacerVM!(
    execute
    OwningPlayer:Local Player Object Reference=GetOwningLocalPlayer!(self:Widget Object Reference) -> (ReturnValue:Local Player Object Reference)
    SpacerSize:EQuestUISpacerSize Enum=Large
   ) -> (then, ReturnValue:Quest UI Spacer View Model Object Reference)
   > Add!(execute, TargetArray:Array of Object References=Get EndListSpaceVM(), NewItem:"Object Reference (by ref)"=@n8) -> (then, ReturnValue)
   Set CurrentBucket(execute=@n2.LoopBody, CurrentBucket:Quest UI Bucket View Model Object Reference=@n2) -> (Output_Get:Quest UI Bucket View Model Object Reference)
   > n9 = For Each Loop(Array:Array of Quest UI Base Content View Model Object References=Get OrderedContentVMs(self=@n2)) -> (Array Element:Quest UI Base Content View Model Object Reference)
   > Set CurrentArrayElement(CurrentArrayElement:Quest UI Base Content View Model Object Reference=@n9) -> (Output_Get:Quest UI Base Content View Model Object Reference)
   > Clear(TargetArray:Array of Object References=Get Category() -> (Category:Array of Object References))
   > n10 = Add(TargetArray:Array of MVVM Base Viewmodel Object References=Get CategoriesID(), NewItem:"MVVM Base Viewmodel Object Reference (by ref)"=@n9)
   > Add(TargetArray:Array of Object References=Get Category(), NewItem:"Object Reference (by ref)"=@n9)
   n11 = Get CurrentArrayElement() -> (CurrentArrayElement:Quest UI Base Content View Model Object Reference)
   > n12 = Cast To QuestUICategoryVM(Object=@n11)
   > n13 = Append Array(TargetArray=Get Category(), SourceArray=Get OrderedContentVMs(self:Quest UI Category View Model Object Reference=@n12))
   Add(
    execute=[@n16.then, @n13.then]
    TargetArray:Array of Fort Poly View Source Data Structures=Get MultiSourceList()
    NewItem:"Fort Poly View Source Data Structure (by ref)"=Make Fort Poly View Source Data!(SourceIndex:int=@n10, Items:Array of Object References=Get Category()) -> (FortPolyViewSourceData:Fort Poly View Source Data Structure) {is=title K2Node_MakeStruct}
   )
   > Add(
    TargetArray:Array of Quest UI Base Content View Model Object References=Get SideNavCategoriesVM()
    NewItem:"Quest UI Base Content View Model Object Reference (by ref)"=Get CurrentArrayElement()
   )
   n14 = Cast To QuestUITrackerVM(execute=@n12.CastFailed, Object=@n11)
   > Set TrackedVM_Ref(TrackedVM_Ref:Quest UI Tracker View Model Object Reference=@n14) -> (Output_Get:Quest UI Tracker View Model Object Reference)
   > Add(
    TargetArray:Array of Fort Poly View Source Data Structures=Get MultiSourceList()
    NewItem:"Fort Poly View Source Data Structure (by ref)"=Make Fort Poly View Source Data(SourceIndex=@n10, Items=GetSortedTrackedQuests() {is=title K2Node_CallFunction})
   )
   > Add(
    TargetArray:Array of Quest UI Base Content View Model Object References=Get SideNavCategoriesVM()
    NewItem:"Quest UI Base Content View Model Object Reference (by ref)"=Get CurrentBucket() -> (CurrentBucket:Quest UI Bucket View Model Object Reference)
   )
   n15 = Cast To QuestUICustomCategoryVM(execute=@n14.CastFailed, Object=@n11)
   > n16 = Append Array(
    TargetArray=Get Category()
    SourceArray=Get OrderedRowVMs!(self:Quest UI Custom Category View Model Object Reference=@n15) -> (OrderedRowVMs:Array of Quest UI Custom Row View Model Object References)
   )
  ]}
 PushTrackedQuestCapReachedDialog {nodes [
   PushTrackedQuestCapReachedDialog!() -> (then, NumberExceeded:int) {is=title K2Node_FunctionEntry}
   > n1 = Is Valid(InputObject=Get Dialog_TrackedCap())
   n2 = Construct Object from Class!(execute=@n1.Is Not Valid, Class=/QuestUI/QuestScreen/Modal/VM_Dialog_TrackedQuestCapReached.VM_Dialog_TrackedQuestCapReached_C, self) -> (then, ReturnValue:VM Dialog Tracked Quest Cap Reached Object Reference)
   n3 = Get NumberExceeded() -> (NumberExceeded:int=0)
   n4 = Get NumberExceeded() -> (NumberExceeded=0)
   > Set DialogInfoText!(
    execute
    DialogInfoText:Text=Select(
     Option 0:Text=Get DefaultText() -> (DefaultText:Text)
     Option 1:Text=Format Text!(
      Format="LOCTEXT(\"You tried to track {1} more quests than the Quest Tracker can handle.\\r\\nUntrack a few quests and give it another shot!\")"
      1:int=@n3
     ) -> (Result)
     Index="integer > integer"(A=@n4)
    ) -> (ReturnValue:Text)
    self:UI Kit Dialog Viewmodel Object Reference=@n2
   ) -> (then, Output_Get:Text)
   > n5 = Create Widget!(
    execute
    Class=/FortUILibrary/UIKit/Dialogs/WBP_UIKit_Dialog.WBP_UIKit_Dialog_C
    OwningPlayer
    UIKitDialogViewModel:UI Kit Dialog Viewmodel Object Reference=@n2
   ) -> (then, ReturnValue:WBP UIKit Dialog Object Reference)
   > Set Dialog_TrackedCap(Dialog_TrackedCap=@n5)
   > Bind Event to Dialog Button Clicked Dispatcher(self:WBP UIKit Dialog Internal Object Reference=Get Dialog_Internal(self=@n5), Delegate=Create Event()) {is=title K2Node_AddDelegate}
   > TryPushWidgetAsModalPanel!(execute, PushWidget:User Widget Object Reference=@n5) -> (then)
   > ActivateWidget(self=Get Dialog_TrackedCap())
  ]}
 RefreshTrackedQuests {nodes [
   RefreshTrackedQuests() {is=title K2Node_FunctionEntry}
   > SetSourceListItems(self=Get QuestTilesList(self=Get QuestsV3ListView()), SourceIndex=0, Items=GetSortedTrackedQuests())
   > UpdateHideDisabledVisibility()
  ]}
 ReorganizeQuestList {nodes [
   n1 = ReorganizeQuestList!() -> (then, DroppedQuest:WBP Quest Tile Object Reference) {is=title K2Node_FunctionEntry}
   n2 = Get QuestTilesList(self=Get QuestsV3ListView())
   > SetSourceListItems(self=@n2, SourceIndex=0, Items=GetSortedTrackedQuests())
   > SetSelectedItem!(execute, self:Fort Multi Source Polymorphic List View Object Reference=@n2, Item:Object Reference=Get QuestUIQuestVM(self=@n1)) -> (then, ReturnValue:bool)
   > SetMoveMode(IsMoveMode=false, IsShutdown=false)
  ]}
 SelectInitialContent {nodes [
   SelectInitialContent() {is=title K2Node_FunctionEntry}
   > n1 = Sequence!(execute) -> (then_0, then_1)
   n2 = Get QuestUIPlayerVM()
   > Branch(Condition=HasSelectionData!(self:Quest UI Player View Model Object Reference=@n2) -> (ReturnValue:bool))
   > n3 = ConsumeSelectionData!(execute, self:Quest UI Player View Model Object Reference=@n2) -> (then, ReturnValue:Quest UISelection Data Structure)
   n4 = Break Quest UISelection Data!(QuestUISelectionData:"Quest UISelection Data Structure (by ref)"=@n3) -> (
    ProductVM:Quest UI Product View Model Object Reference
    CategoryInterfaceVM:Quest UI Category Interface View Model Object Reference
    ContentVM:Quest UI Base Content View Model Object Reference
   ) {is=title K2Node_BreakStruct}
   > Branch(Condition="Equal (Object)"(A=@n4, B=Get QuestUIProductVM()))
   > n5 = SelectContent!(
    execute
    self:WBP Quests V3List View Object Reference=Get QuestsV3ListView()
    ContentToSelect:Quest UI Base Content View Model Object Reference=@n4.ContentVM
   ) -> (then, Success:bool)
   > Branch(Condition=@n5)
   > Return Node!(execute)
   n6 = FindFirstSelectableElement!(execute=@n1.then_1, self:WBP Quests V3List View Object Reference=Get QuestsV3ListView()) -> (then, FirstSelectedIndex:int=0)
   > Branch(Condition="integer > integer"(A=@n6, B=-1))
   n7 = Get QuestsV3ListView()
   > SetSelectedIndex(self=Get ListView(self=Get QuestTilesList(self=@n7)), Index=@n6)
  ]}
 SetDropSource {nodes [
   n1 = SetDropSource!() -> (then, SourceQuestTile:WBP Quest Tile Object Reference) {is=title K2Node_FunctionEntry}
   > Set SourceQuestTile(SourceQuestTile:WBP Quest Tile Object Reference=@n1) -> (Output_Get:WBP Quest Tile Object Reference)
   n2 = Get SourceQuestTile()
   > InitializeFromQuestTile!(execute, self:WBP Quest Tile Object Reference=Get WBP_QuestTile_Overlay(), WBP_QuestTile:WBP Quest Tile Object Reference=@n2) -> (then)
   > Set TargetQuestTile(TargetQuestTile=@n2)
   > Set SourceQuestTile(SourceQuestTile=Get WBP_QuestTile_Overlay())
   > DragDropTargetChanged_Visualization!(
    execute
    self
    OldDropTarget:WBP Quest Tile Object Reference
    DropPlacementEnter:EDragDropPlacementType Enum=NewEnumerator2
    DropPlacementLeave:EDragDropPlacementType Enum=NewEnumerator2
   ) -> (then) {is=title K2Node_CallFunction}
   > OverlayGamepadDragStartedVisualization(self:WBP Quest Tile Object Reference=Get WBP_QuestTile_Overlay())
   > SetVisibility(self=Get WBP_QuestTile_Overlay(), InVisibility=HitTestInvisible)
  ]}
 SetExternalLink {nodes [
   n1 = SetExternalLink!() -> (then, InExternalLinkVM:Quest UIExternal Link VM Object Reference) {is=title K2Node_FunctionEntry}
   > Is Valid(InputObject=@n1)
   > SetBottomButtonVisibility(
    self=Get DecoyButtonExternalLink()
    ShouldButtonBeVisible=OR Boolean!(
     A=IsInFrontend!(self) -> (InFrontend:bool) {is=title K2Node_CallFunction}
     B=IsLinkAllowedInGameplay!(self:Quest UIExternal Link VM Object Reference=@n1) -> (ReturnValue:bool)
    ) -> (ReturnValue)
   )
  ]}
 SetInputsLocked {nodes [
   n1 = SetInputsLocked!() -> (then, WBox_Input:Panel Widget Object Reference, IsLocked:bool=~) {is=title K2Node_FunctionEntry}
   > n2 = For Each Loop(Array:Array of Widget Object References=GetAllChildren(self=@n1)) -> (Array Element:Widget Object Reference)
   > n3 = Cast To WBP_UIKit_Button_Quiet!(execute, Object=@n2) -> (then, CastFailed, AsWBP UIKit Button Quiet:WBP UIKit Button Quiet Object Reference) {is=title K2Node_DynamicCast}
   > SetIsLocked!(execute, self:Common Button Base Object Reference=@n3, bInIsLocked:bool=Get IsLocked() -> (IsLocked:bool)) -> (then)
  ]}
 SetIsRestoringSelection {nodes [
   n1 = SetIsRestoringSelection!() -> (then, IsRestoringSelection:bool=~) {is=title K2Node_FunctionEntry}
   > Set IsRestoringSelection(IsRestoringSelection:bool=@n1) -> (Output_Get:bool)
  ]}
 SetMobileStyle {nodes [
   SetMobileStyle() {is=title K2Node_FunctionEntry}
   > Branch(Condition=IsUsingCompactUI())
   > SetCompactUIStyle!(
    execute
    self:WBP UIKit Frontend Layout Object Reference=Get WBP_UIKit_FrontendLayout()
    ContentPanelSizeX:double=997
    TopBarSpacer:double=160
    SideNavMaxSizeY:double=0
    SafeZoneMarginY:double=24
    SafeZoneMarginX:double=48
   ) -> (then) {is=title K2Node_CallFunction}
   n1 = SlotAsGridSlot(Widget=Get HeaderSwitcher())
   n2 = Get!(self:Grid Slot Object Reference=@n1) -> (Padding_Left:float=~, Padding_Top:float=~, Padding_Right:float=~, Padding_Bottom:float=~)
   > SetPadding!(
    execute
    self:Grid Slot Object Reference=@n1
    InPadding_Left:float=@n2
    InPadding_Top:float=@n2.Padding_Top
    InPadding_Right:float=48
    InPadding_Bottom:float=@n2.Padding_Bottom
   ) -> (then)
   n3 = SlotAsGridSlot(Widget=Get SizeBox_Quests() -> (SizeBox_Quests:Size Box Object Reference))
   > SetHorizontalAlignment(self:Grid Slot Object Reference=@n3, InHorizontalAlignment=HAlign_Fill)
   n4 = Get(self=@n3) -> (Padding_Left=~, Padding_Top=~, Padding_Right=~, Padding_Bottom=~)
   > SetPadding(self=@n3, InPadding_Left=0, InPadding_Top=@n4.Padding_Top, InPadding_Right=@n4.Padding_Right, InPadding_Bottom=@n4.Padding_Bottom)
   > DisplaySideNav(self:WBP UIKit Frontend Layout Object Reference=Get WBP_UIKit_FrontendLayout())
  ]}
 SetMoveMode {nodes [
   SetMoveMode!() -> (then, IsMoveMode:bool=~, IsShutdown:bool=~) {is=title K2Node_FunctionEntry}
   > StopTrackAllProgress()
   > Branch(Condition=XOR Boolean!(A:bool=Get IsMoving(), B:bool=Get IsMoveMode() -> (IsMoveMode:bool)) -> (ReturnValue:bool))
   > n1 = Sequence!(execute) -> (then_0, then_1, then_2:Exec)
   > n2 = Set IsMoving(IsMoving:bool=Get IsMoveMode()) -> (Output_Get:bool)
   > SetIsMoveMode!(execute, self:WBP Quests V3List View Object Reference=Get QuestsV3ListView(), IsMoveMode:bool=@n2) -> (then)
   > SetIsMoveMode!(execute, self:WBP Selection Panel Object Reference=Get SelectionPanel(), IsMoving:bool=@n2) -> (then)
   > SetBackToSideNavigationVisibiliity(self=Get SideNavigation(), IsVisible=NOT Boolean(A=@n2))
   > n3 = Branch(Condition=@n2)
   > Branch Input Type()
   > n4 = Get Current Selected Widget!(In:Exec) -> (FoundWidget:Exec, WBP_QuestTile:WBP Quest Tile Object Reference, No Widget:Exec) {is=title K2Node_MacroInstance}
   > Set InitialMoveSelection(InitialMoveSelection:WBP Quest Tile Object Reference=@n4) -> (Output_Get:WBP Quest Tile Object Reference)
   > SetDropSource!(execute, self, SourceQuestTile:WBP Quest Tile Object Reference=@n4) -> (then) {is=title K2Node_CallFunction}
   > RequestRefreshFocus()
   n5 = NOT Boolean(A=IsInTrackedCategory() {is=title K2Node_CallFunction})
   n6 = Get IsShutdown() -> (IsShutdown:bool)
   n7 = Get InitialMoveSelection() -> (InitialMoveSelection:WBP Quest Tile Object Reference)
   n8 = Get TargetQuestTile()
   n9 = Branch(execute=@n3.else, Condition=AND Boolean(A=OR Boolean(A=@n5, B=@n6), B="Not Equal (Object)"(A=@n7, B=@n8) -> (ReturnValue=~)))
   > n10 = Get!(execute:Exec) -> (then:Exec, else:Exec, InitialMoveSelection:WBP Quest Tile Object Reference)
   > n11 = ClearSelection(self:Common Button Base Object Reference=@n10)
   SetVisibility(execute=[@n9.else, @n11.then], self=Get WBP_QuestTile_Overlay(), InVisibility=Collapsed)
   > Set InitialMoveSelection()
   > n12 = Is Valid(InputObject=Get TargetQuestTile())
   > DraggedOverChangedVisualization(self=Get TargetQuestTile(), IsEntering=false, DropPlacement=NewEnumerator0)
   > n13 = Set TargetQuestTile()
   n14 = Get SourceQuestTile()
   Branch(execute=[@n13.then, @n12.Is Not Valid], Condition=AND Boolean(A=Get IsShutdown(), B=IsValid(Object=@n14)))
   > RequestRefreshFocus()
   n15 = Get IsMoving()
   SetActiveWidgetIndex(
    execute=@n1.then_1
    self=Get AnimSwitcher_Inputs() -> (AnimSwitcher_Inputs:Common Animated Switcher Object Reference)
    Index=Select(Option 0:int=0, Option 1:int=1, Index=@n15) -> (ReturnValue:int)
   )
   > n16 = GetCurrentStateWidget(PlayerController=GetOwningPlayer())
   > n17 = Cast To AthenaUIStateWidget_Frontend(Object=@n16)
   > n18 = SetFrontendVisibilityMode!(
    execute
    self:Athena UIState Widget Frontend Object Reference=@n17
    InFrontendVisibilityMode:EFrontendVisibilityMode Enum=Select(Option 0:EFrontendVisibilityMode Enum=Normal, Option 1:EFrontendVisibilityMode Enum=OnlyTop, Index=@n15) -> (ReturnValue:EFrontendVisibilityMode Enum)
   ) -> (then)
   SetInputsLocked!(
    execute=[@n24.then, @n18.then]
    self
    WBox_Input:Panel Widget Object Reference=Get WBox_MoveModeInputs()
    IsLocked:bool=NOT Boolean(A=Get IsMoving())
   ) -> (then) {is=title K2Node_CallFunction}
   > SetInputsLocked(WBox_Input=Get WBox_Inputs() -> (WBox_Inputs:Wrap Box Object Reference), IsLocked=Get IsMoving())
   > n19 = Branch(Condition=@n15)
   n20 = MakeLiteralGameplayTag!(Value:Gameplay Tag Structure=(TagName="UI.TextChat.Disabled")) -> (ReturnValue:Gameplay Tag Structure)
   n21 = Get EnhancedInputLocalPlayerSubsystem() -> (ReturnValue:Enhanced Input Local Player Subsystem Object Reference) {is=title K2Node_GetSubsystem}
   > AddTagToInputMode!(
    execute
    self:Enhanced Input Subsystem Interface Interface=@n21
    TagToAdd:"Gameplay Tag Structure (by ref)"=@n20
    Options:"Modify Context Options Structure (by ref)"=(bIgnoreAllPressedKeysUntilRelease=True,bForceImmediately=False,bNotifyUserSettings=False)
   ) -> (then)
   RemoveTagFromInputMode!(
    execute=@n19.else
    self:Enhanced Input Subsystem Interface Interface=@n21
    TagToRemove:"Gameplay Tag Structure (by ref)"=@n20
    Options:"Modify Context Options Structure (by ref)"=(bIgnoreAllPressedKeysUntilRelease=True,bForceImmediately=False,bNotifyUserSettings=False)
   ) -> (then)
   n22 = FindParentWidgetOfType!(
    execute=@n17.CastFailed
    StartingWidget:Widget Object Reference=Self-Reference()
    Type:Widget Class Reference=/FortniteCoreUI/HUD/BP_AthenaMapScreenContainer.BP_AthenaMapScreenContainer_C
   ) -> (then, ReturnValue:BP Athena Map Screen Container Object Reference)
   n23 = ShouldShowActionBarForCurrentTab!(self:Athena Map Screen Container Object Reference=@n22) -> (ReturnValue:bool)
   > n24 = SetVisibility(
    self=Get AthenaHUDBottomBar!(self:BP Athena Map Screen Container Object Reference=@n22) -> (AthenaHUDBottomBar:WBP Athena Bottom Bar Widget Object Reference)
    InVisibility=Select(
     Option 0:ESlateVisibility Enum=Select(Option 0=Collapsed, Option 1=SelfHitTestInvisible, Index=@n23)
     Option 1:ESlateVisibility Enum=Collapsed
     Index=Get IsMoving()
    ) -> (ReturnValue:ESlateVisibility Enum)
   )
   n25 = Branch(execute=@n1.then_2, Condition=Get IsMoveMode())
   RequestRefreshFocus(execute=@n25.else)
  ]}
 SetProductIcon {nodes [
   n1 = SetProductIcon!() -> (then, Icon:Material Instance Soft Object Reference) {is=title K2Node_FunctionEntry}
   > SetProductIcon!(
    execute
    self:WBP Quests Modal Header Object Reference=Get WBP_QuestsModal_Header() -> (WBP_QuestsModal_Header:WBP Quests Modal Header Object Reference)
    Icon:Material Instance Soft Object Reference=@n1
   ) -> (then) {is=title K2Node_CallFunction}
  ]}
 SetQuestTracking {nodes [
   n1 = SetQuestTracking!() -> (then, Quest:Quest UI Quest View Model Object Reference, ShouldTrack:bool=~) {is=title K2Node_FunctionEntry}
   > Set IsSelected(IsSelected:bool="Equal (Object)"(A=@n1, B=GetSelectedUnwrappedItem())) -> (Output_Get:bool)
   n2 = Get ShouldTrack()
   > n3 = Branch(Condition=AND Boolean!(A=Get IsSelected() -> (IsSelected:bool), B=NOT Boolean(A=@n2), C:bool=Get IsTrackedQuests()) -> (ReturnValue))
   > n4 = Set QuestToSelect(QuestToSelect:Quest UI Quest View Model Object Reference=GetNextQuest!(self, ItemToFind:"Quest UI Quest View Model Object Reference (by ref)"=Get Quest() -> (Quest:Quest UI Quest View Model Object Reference)) -> (NextQuest:Quest UI Quest View Model Object Reference) {is=title K2Node_CallFunction}) -> (Output_Get:Quest UI Quest View Model Object Reference)
   n5 = Switch on EQuestTrackState!(execute=[@n4.then, @n3.else], Selection:EQuestTrackState Enum=GetQuestTrackState(self=Get Quest()) -> (ReturnValue=AvailableToTrack)) -> (AvailableToTrack:Exec, CurrentlyTracked:Exec, NotTrackable:Exec, AtTrackedQuestLimit:Exec) {is=title K2Node_SwitchEnum}
   Return Node!(execute=@n5.AtTrackedQuestLimit, AtQuestLimit:bool=true, InitialSelection:Quest UI Quest View Model Object Reference)
   n6 = Sequence!(execute=@n5.CurrentlyTracked) -> (then_0, then_1)
   > n7 = Branch(Condition=Get ShouldTrack())
   SetIsUntracked(execute=@n7.else, self:Quest UI Quest View Model Object Reference=Get Quest())
   > FireQuestAnalytics_ElementInteracted!(
    execute
    PlayerController:Player Controller Object Reference=GetOwningPlayer()
    Action:"String (by ref)"=Untrack Quest
    ContentVM:Quest UI Base Content View Model Object Reference=Get Quest()
   ) -> (then)
   > Branch(Condition=Get IsSelected())
   > UpdateTrackedQuestButton(ListItem=Get Quest())
   Return Node(execute=@n6.then_1, AtQuestLimit=false, InitialSelection=Get QuestToSelect() -> (QuestToSelect:Quest UI Quest View Model Object Reference))
   n8 = Sequence(execute=@n5.AvailableToTrack)
   > Branch(Condition=Get ShouldTrack())
   > SetIsTracked(self=Get Quest(), TrackedIndex=-1)
   > FireQuestAnalytics_ElementInteracted(PlayerController=GetOwningPlayer(), Action=Track Quest, ContentVM=Get Quest())
   > Branch(Condition=Get IsSelected())
   > UpdateTrackedQuestButton(ListItem=Get Quest())
   Return Node(execute=@n8.then_1, AtQuestLimit=false)
  ]}
 ShutdownModes {nodes [
   Inputs()
   > n1 = Is Valid(InputObject=Get Dialog_TrackedCap())
   n2 = Branch(execute=@n1.Is Not Valid, Condition=Get IsMoving())
   > n3 = SetMoveMode(IsMoveMode=false, IsShutdown=true) {is=title K2Node_CallFunction}
   Outputs!(Shutdown:Exec=[@n3.then, @n4.then], NoShutdownRequired:Exec=@n2.else)
   n4 = ClearDialog!(execute=@n1.Is Valid, self, ButtonIndex:int=0) -> (then) {is=title K2Node_CallFunction}
  ]}
 StopTrackAllProgress {nodes [
   SetActionProgress!(
    execute
    self:Fort CTAButton Object Reference=Reroute Node(InputPin:Fort CTAButton Object Reference) -> (OutputPin:Fort CTAButton Object Reference)
    HeldPercent:float=0
   ) -> (then)
   n1 = Get HoldSoundAudioComponent!(self:WBP UIKit Button CTA Base Object Reference) -> (HoldSoundAudioComponent:Audio Component Object Reference)
   > Is Valid(InputObject=@n1)
   > Branch(Condition=IsPlaying!(self:Audio Component Object Reference=@n1) -> (ReturnValue:bool))
   > Stop(self:Audio Component Object Reference=@n1)
   StopTrackAllProgress() {is=title K2Node_FunctionEntry}
   Get DecoyButtonTrackAll() -> (DecoyButtonTrackAll:WBP Bottom Bar Decoy Button Object Reference)
  ]}
 UpdateCTAButtons {nodes [
   UpdateCTAButtons() {is=title K2Node_FunctionEntry}
   > UpdateMoveQuestButtonVisibility()
   > UpdateTrackAllButtonState(ForceUntrackCategory=false)
   > UpdateHideDisabledVisibility()
  ]}
 UpdateGamepadDropTarget {nodes [
   UpdateGamepadDropTarget!() -> (then, NewSelectedItem:Object Reference) {is=title K2Node_FunctionEntry}
   > Branch Input Type()
   > Branch(Condition=Get IsMoving())
   > Set OldTarget(OldTarget:WBP Quest Tile Object Reference=Get TargetQuestTile()) -> (Output_Get:WBP Quest Tile Object Reference)
   > n1 = Get Current Selected Widget()
   > Set TargetQuestTile(TargetQuestTile=@n1)
   > DragDropTargetChanged_Visualization(
    OldDropTarget=Get OldTarget() -> (OldTarget:WBP Quest Tile Object Reference)
    DropPlacementEnter=GetDragDropPlacement!(self:WBP Quest Tile Object Reference=Get TargetQuestTile(), SourceTile:WBP Quest Tile Object Reference=Get SourceQuestTile()) -> (Placement:EDragDropPlacementType Enum=NewEnumerator0)
    DropPlacementLeave=GetDragDropPlacement(self=Get OldTarget(), SourceTile=Get SourceQuestTile()) -> (Placement=NewEnumerator0)
   )
  ]}
 UpdateHideDisabledVisibility {nodes [
   UpdateHideDisabledVisibility() {is=title K2Node_FunctionEntry}
   > SetBottomButtonVisibility(
    self=Get DecoyButtonHideDisabled()
    ShouldButtonBeVisible=AND Boolean!(A=IsInTrackedCategory(), B=Get HasDisabledTrackedQuest() -> (HasDisabledTrackedQuest:bool)) -> (ReturnValue)
   )
  ]}
 UpdateInGamePadding {nodes [
   UpdateInGamePadding() {is=title K2Node_FunctionEntry}
   > n1 = Cast To FortPlayerControllerFrontEnd(Object=GetOwningPlayer())
   n2 = Get WBP_UIKit_FrontendLayout()
   > SetUseCollapseTopPadding!(execute, self:WBP UIKit Frontend Layout Object Reference=@n2, bUseCollapseTopBarPadding:bool=false) -> (then)
   SetUseCollapseTopPadding(execute=@n1.CastFailed, self=@n2, bUseCollapseTopBarPadding=true)
  ]}
 UpdateMoveQuestButtonVisibility {nodes [
   UpdateMoveQuestButtonVisibility() {is=title K2Node_FunctionEntry}
   n1 = IsUsingTouch(PlayerController=GetPlayerController!(PlayerIndex:int=0) -> (ReturnValue:Player Controller Object Reference))
   n2 = "Equal (Enum)"(A=GetCurrentInputType(self=Get CommonInputSubsystem() {is=title K2Node_GetBlueprintContextEx}), B=Gamepad)
   > SetBottomButtonVisibility(
    self=Get DecoyButtonMoveQuest() -> (DecoyButtonMoveQuest:WBP Bottom Bar Decoy Button Object Reference)
    ShouldButtonBeVisible=AND Boolean(A=OR Boolean(A=@n2, B=@n1), B=IsInTrackedCategory())
   )
  ]}
 UpdateTabsScrollbar {nodes [
   UpdateTabsScrollbar() {is=title K2Node_FunctionEntry}
   n1 = Get ScrollBox_Tabs(self=Get SideNavigation!(self:WBP Quest V3 Side Navigation Object Reference=Get SideNavigation!() -> (SideNavigation:WBP Quest V3 Side Navigation Object Reference)) -> (SideNavigation:WBP UIKit Side Navigation Object Reference))
   n2 = Get!(self:Scroll Box Object Reference=@n1) -> (
    WidgetStyle_BarThickness:float=~
    WidgetStyle_TopShadowBrush:Slate Brush Structure
    WidgetStyle_BottomShadowBrush:Slate Brush Structure
    WidgetStyle_LeftShadowBrush:Slate Brush Structure
    WidgetStyle_RightShadowBrush:Slate Brush Structure
    WidgetStyle_HorizontalScrolledContentPadding:Margin Structure
    WidgetStyle_VerticalScrolledContentPadding:Margin Structure
   )
   > Set!(
    execute
    WidgetStyle_BarThickness:float=9
    WidgetStyle_TopShadowBrush:Slate Brush Structure=@n2.WidgetStyle_TopShadowBrush
    WidgetStyle_BottomShadowBrush:Slate Brush Structure=@n2.WidgetStyle_BottomShadowBrush
    WidgetStyle_LeftShadowBrush_TintColor:Slate Color Structure=(SpecifiedColor=(R=1,G=1,B=1,A=1),ColorUseRule=UseColor_Specified)
    WidgetStyle_LeftShadowBrush_DrawAs:ESlateBrushDrawType Enum=Image
    WidgetStyle_LeftShadowBrush_Tiling:ESlateBrushTileType Enum=NoTile
    WidgetStyle_LeftShadowBrush_Mirroring:ESlateBrushMirrorType Enum=NoMirror
    WidgetStyle_LeftShadowBrush_ImageSize:Vector 2D Structure=(X=32,Y=32)
    WidgetStyle_LeftShadowBrush_Margin:Margin Structure=(Left=0,Top=0,Right=0,Bottom=0)
    WidgetStyle_LeftShadowBrush_ResourceObject:Object Reference=/QuestUI/QuestScreen/Materials/MI_UI_ItemsFade_Left
    WidgetStyle_LeftShadowBrush_OutlineSettings:Slate Brush Outline Settings Structure=(CornerRadii=(X=0,Y=0,Z=0,W=0),Color=(SpecifiedColor=(R=0,G=0,B=0,A=0),ColorUseRule=UseColor_Specified),Width=0,RoundingType=HalfHeightRadius,bUseBrushTransparency=False)
    WidgetStyle_RightShadowBrush_TintColor:Slate Color Structure=(SpecifiedColor=(R=1,G=1,B=1,A=1),ColorUseRule=UseColor_Specified)
    WidgetStyle_RightShadowBrush_DrawAs:ESlateBrushDrawType Enum=Image
    WidgetStyle_RightShadowBrush_Tiling:ESlateBrushTileType Enum=NoTile
    WidgetStyle_RightShadowBrush_Mirroring:ESlateBrushMirrorType Enum=NoMirror
    WidgetStyle_RightShadowBrush_ImageSize:Vector 2D Structure=(X=32,Y=32)
    WidgetStyle_RightShadowBrush_Margin:Margin Structure=(Left=0,Top=0,Right=0,Bottom=0)
    WidgetStyle_RightShadowBrush_ResourceObject:Object Reference=/QuestUI/QuestScreen/Materials/MI_UI_ItemsFade_Right
    WidgetStyle_RightShadowBrush_OutlineSettings:Slate Brush Outline Settings Structure=(CornerRadii=(X=0,Y=0,Z=0,W=0),Color=(SpecifiedColor=(R=0,G=0,B=0,A=0),ColorUseRule=UseColor_Specified),Width=0,RoundingType=HalfHeightRadius,bUseBrushTransparency=False)
    WidgetStyle_HorizontalScrolledContentPadding:Margin Structure=(Left=0,Top=0,Right=1,Bottom=0)
    WidgetStyle_VerticalScrolledContentPadding:Margin Structure=(Left=0,Top=0,Right=0,Bottom=1)
    self:Scroll Box Object Reference=@n1
   ) -> (then, Output_Get:Scroll Box Style Structure)
  ]}
 UpdateTrackAllButtonState {nodes [
   UpdateTrackAllButtonState!() -> (then, ForceUntrackCategory:bool=~) {is=title K2Node_FunctionEntry}
   > Set IsTrackAll(IsTrackAll:bool=false) -> (Output_Get:bool)
   n1 = IsValid(Object=Get Quest UI Horizontal Punch Card View Model() -> (Quest UI Horizontal Punch Card View Model:Quest UI Horizontal Punch Card View Model Object Reference))
   n2 = GetTrackableQuests()
   n3 = "integer > integer"(A=Length(TargetArray:Array of Quest UI Quest View Model Object References=@n2), B=1)
   > n4 = Branch(Condition=OR Boolean(A=@n3, B=AND Boolean(A=@n3, B=@n1)))
   > SetBottomButtonVisibility(self=Get DecoyButtonTrackAll(), ShouldButtonBeVisible=true)
   > n5 = For Each Loop with Break(Array:Array of Quest UI Quest View Model Object References=@n2, Break=[@n8.then, @n7.then]) -> (Array Element:Quest UI Quest View Model Object Reference)
   > n6 = Branch(Condition="Not Equal (Enum)"(A=GetQuestTrackState(self=@n5) -> (ReturnValue=AvailableToTrack), B=CurrentlyTracked))
   > n7 = Set IsTrackAll(IsTrackAll=true)
   n8 = Branch(execute=@n6.else, Condition=Get ForceUntrackCategory() -> (ForceUntrackCategory:bool))
   n9 = Branch(execute=@n5.Completed, Condition=Get IsTrackAll())
   n10 = Get DecoyButtonTrackAll()
   > SetTriggeringInputAction!(
    execute
    self=@n10
    InputActionRow_DataTable:Data Table Object Reference=&QuestInputActionDataTable
    InputActionRow_RowName:Name=TrackCategory
   ) -> (then)
   > CacheInputAction(self=@n10)
   SetTriggeringInputAction(
    execute=@n9.else
    self=@n10
    InputActionRow_DataTable=&QuestInputActionDataTable
    InputActionRow_RowName=Select(Option 0:Name=UntrackCategory, Option 1:Name=UntrackAll, Index=IsInTrackedCategory()) -> (ReturnValue:Name)
   )
   > CacheInputAction(self=@n10)
   n11 = Get TrackerVM(self=Get QuestUIPlayerVM())
   > SetBottomButtonVisibility(
    self=@n10
    ShouldButtonBeVisible="integer > integer"(A=GetCurrentTrackedQuestCount!(self:Quest UI Tracker View Model Object Reference=@n11) -> (ReturnValue:int=0))
   )
   SetBottomButtonVisibility(execute=@n4.else, self=Get DecoyButtonTrackAll(), ShouldButtonBeVisible=false)
  ]}
 UpdateTrackedQuestButton {nodes [
   UpdateTrackedQuestButton!() -> (then, ListItem:Object Reference) {is=title K2Node_FunctionEntry}
   > n1 = Cast To QuestUIQuestVM(Object=Get ListItem() -> (ListItem:Object Reference))
   > n2 = Switch on EQuestTrackState(Selection=GetQuestTrackState(self=@n1) -> (ReturnValue=AvailableToTrack))
   n3 = Get DecoyButtonTrackQuest() -> (DecoyButtonTrackQuest:WBP Bottom Bar Decoy Button Object Reference)
   n4 = SetTriggeringInputAction(
    execute=@n2.CurrentlyTracked
    self=@n3
    InputActionRow_DataTable=&QuestInputActionDataTable
    InputActionRow_RowName=UntrackQuest
   )
   CacheInputAction(execute=[@n4.then, @n5.then], self=@n3)
   > SetBottomButtonVisibility(self=Get DecoyButtonTrackQuest(), ShouldButtonBeVisible=true)
   n5 = SetTriggeringInputAction(
    execute=[@n2.AtTrackedQuestLimit, @n2.AvailableToTrack]
    self=Get DecoyButtonTrackQuest()
    InputActionRow_DataTable=&QuestInputActionDataTable
    InputActionRow_RowName=TrackQuest
   )
   SetBottomButtonVisibility(execute=@n2.NotTrackable, self=Get DecoyButtonTrackQuest(), ShouldButtonBeVisible=false)
   SetBottomButtonVisibility(execute=@n1.CastFailed, self=Get DecoyButtonTrackQuest(), ShouldButtonBeVisible=false)
  ]}
 UseCorrectedTrackedIndex {nodes [
   UseCorrectedTrackedIndex() {is=title K2Node_FunctionEntry}
   > Return Node!(execute, UseCorrected:bool=IsInZone(self=Get FortGlobalUIContext()))
  ]}
]
variables [
 CategoriesID {type="array<MVVMViewModelBase>", default []}
 CategoryHeader {type=string, default=""}
 CurrentImpressions {type="array<FlowAnalyticsImpressionSectionContent>", default []}
 DebugTabs {type="array<S_UI_TabContainerData>", default []}
 Dialog_TrackedCap {type=WBP_UIKit_Dialog_C, default=null}
 EndListSpaceVM {type="array<Object>", default []}
 HasDisabledTrackedQuest {type=boolean, default=true}
 InitialMoveSelection {type=WBP_QuestTile_C, default=null}
 Input Type {type=ECommonInputType, default=MouseAndKeyboard}
 IsExiting {type=boolean}
 IsHidingDisabled {type=boolean}
 IsMoving {type=boolean}
 IsRestoringSelection {type=boolean}
 IsTrackAll {type=boolean}
 IsTrackedQuests {type=boolean}
 IsTransitioning {type=boolean}
 MultiSourceList {type="array<FortPolyViewSourceData>", default []}
 NewVar {type=integer}
 NoTrackedQuestsMessageVM {type=VM_QuestUISpecialMessage_C, default=null}
 PreviousSideNavCategoryVM {type=MVVMViewModelBase, default=null}
 PreviousSource {type=integer, default=-1}
 Quest UI Horizontal Punch Card View Model {type=QuestUIHorizontalPunchCardVM, default=null}
 ReturnFocusTile {type=UserWidget, default=null}
 SelectedVPunchCard {type=MVVM_PunchCardQuest_C, default=null}
 SelectIslandListener {type=AsyncAction_StartListeningToEvent, default=null}
 ShippingConfig {type=string, default=Shipping}
 SideNavCategoriesVM {type="array<QuestUIBaseContentVM>", default []}
 SourceQuestTile {type=WBP_QuestTile_C, default=null}
 TargetQuestTile {type=WBP_QuestTile_C, default=null}
 TrackedVM_Ref {type=QuestUITrackerVM, default=null}
 VPunchCardSelected Pip {type=integer}
 WBP_QuestTile_Overlay {type=WBP_QuestTile_C, default=null}
]
widgets {
 info {
  inheritedWidgetCount=0
  namedSlotCount=8
  parentClass=/Script/QuestUI.QuestUIScreen
  rootWidgetClass=/FortUILibrary/UIKit/Layout/WBP_UIKit_FrontendLayout.WBP_UIKit_FrontendLayout_C
  widgetCount=34
 }
 widgets [
  WBP_UIKit_FrontendLayout {
   widgetClassPath=/FortUILibrary/UIKit/Layout/WBP_UIKit_FrontendLayout
   bIsVariable=true
   widget=WBP_UIKit_FrontendLayout
   properties {
    bUsePostBufferUpdate=true
    backgroundMaterial=/QuestUI/QuestScreen/Materials/M_UI_TestBackground_Inst
    contentPanelPrimary_SizeX=960
    navigation=&self:WidgetTree.WBP_UIKit_FrontendLayout.WidgetNavigation_0
    safeZone_ContentPadding=<ltrb 96 136 80 0>
    secondaryContentPaddingOverride=<ltrb -96 0 0 0>
   }
  }
  SelectionPanel {
   widgetClassPath=/QuestUI/QuestScreen/Modal/WBP_SelectionPanel
   bIsVariable=true
   namedSlotHost=WBP_UIKit_FrontendLayout
   widget=SelectionPanel
  }
  QuestContent {
   widgetClassPath=&GridPanel
   bIsVariable=true
   namedSlotHost=WBP_UIKit_FrontendLayout
   widget=QuestContent
   properties {columnFill [1,0], rowFill [0,0,1]}
  }
  WBP_QuestsModal_Header {
   widgetClassPath=/QuestUI/QuestScreen/Modal/WBP_QuestsModal_Header
   bIsVariable=true
   parent=QuestContent
   widget=WBP_QuestsModal_Header
   properties {visibility=Collapsed}
   slot {class=GridSlot, properties {layer=2, padding=<ltrb 0 0 56 0>}}
  }
  HeaderSwitcher {
   widgetClassPath=&CommonActivatableWidgetSwitcher
   bIsVariable=true
   parent=QuestContent
   widget=HeaderSwitcher
   properties {transitionDuration=0}
   slot {class=GridSlot, properties {layer=10, padding=<ltrb 30 0 50 0>}}
  }
  WBP_QuestsSeparate_Header {
   widgetClassPath=&/WBP_QuestsSeparate_Header
   bIsVariable=true
   parent=HeaderSwitcher
   widget=WBP_QuestsSeparate_Header
   properties {isEntry=false}
   slot {class=WidgetSwitcherSlot}
  }
  WBP_QuestsSeparate_Tracker {
   widgetClassPath=&/WBP_QuestsSeparate_Tracker
   bIsVariable=true
   parent=HeaderSwitcher
   widget=WBP_QuestsSeparate_Tracker
   slot {class=WidgetSwitcherSlot}
  }
  WBP_QuestsSeparate_CustomHeader {
   widgetClassPath=&/WBP_QuestsSeparate_CustomHeader
   bIsVariable=true
   parent=HeaderSwitcher
   widget=WBP_QuestsSeparate_CustomHeader
   properties {isEntry=false}
   slot {class=WidgetSwitcherSlot}
  }
  SizeBox_Quests {
   widgetClassPath=&SizeBox
   bIsVariable=true
   parent=QuestContent
   widget=SizeBox_Quests
   properties {
    bOverride_WidthOverride=true
    clipping=ClipToBounds
    maxDesiredWidth=902
    navigation=&self:WidgetTree.SizeBox_Quests.WidgetNavigation_0
    widthOverride=997
   }
   slot {class=GridSlot, properties {horizontalAlignment=HAlign_Left, row=2}}
  }
  QuestsV3ListView {
   widgetClassPath=&/WBP_QuestsV3ListView
   bIsVariable=true
   parent=SizeBox_Quests
   widget=QuestsV3ListView
   properties {desiredFocusWidget {widgetName=QuestTilesList}, padding=<ltrb 30 30 50 30>}
   slot {class=SizeBoxSlot, properties {padding=<ltrb 0 0 0 90>}}
  }
  MaskBottom {
   widgetClassPath=&Image
   bIsVariable=true
   parent=QuestContent
   widget=MaskBottom
   properties {
    visibility=SelfHitTestInvisible
    brush {
     bIsDynamicallyLoaded=false
     drawAs=Image
     imageType=FullColor
     mirroring=NoMirror
     resourceName=None
     resourceObject=/FortniteCoreUI/UI/Quests/Materials/MI_RoundedFade_Vertical
     tiling=NoTile
     imageSize=<32 60>
     margin=<ltrb 0 0 0 0>
     outlineSettings {
      bUseBrushTransparency=false
      roundingType=HalfHeightRadius
      width=0
      color {colorUseRule=UseColor_Specified, specifiedColor=<rgba 0 0 0 0>}
      cornerRadii=<0 0 0 0>
     }
     tintColor {colorUseRule=UseColor_Specified, specifiedColor=<rgba 1 1 1 1>}
     uVRegion {bIsValid=false, max=<0 0>, min=<0 0>}
    }
   }
   slot {class=GridSlot, properties {layer=1, row=2, verticalAlignment=VAlign_Bottom}}
  }
  MaskTop {
   widgetClassPath=&Image
   bIsVariable=true
   parent=QuestContent
   widget=MaskTop
   properties {
    visibility=SelfHitTestInvisible
    brush {
     bIsDynamicallyLoaded=false
     drawAs=Image
     imageType=FullColor
     mirroring=NoMirror
     resourceName=None
     resourceObject=/FortniteCoreUI/UI/Quests/Materials/MI_RoundedFade_Top
     tiling=NoTile
     imageSize=<32 50>
     margin=<ltrb 0 0 0 0>
     outlineSettings {
      bUseBrushTransparency=false
      roundingType=HalfHeightRadius
      width=0
      color {colorUseRule=UseColor_Specified, specifiedColor=<rgba 0 0 0 0>}
      cornerRadii=<0 0 0 0>
     }
     tintColor {colorUseRule=UseColor_Specified, specifiedColor=<rgba 1 1 1 1>}
     uVRegion {bIsValid=false, max=<0 0>, min=<0 0>}
    }
   }
   slot {class=GridSlot, properties {layer=1, row=2, verticalAlignment=VAlign_Top, padding=<ltrb 0 -22 0 0>}}
  }
  Overlay_170 {widgetClassPath=&Overlay, namedSlotHost=WBP_UIKit_FrontendLayout, widget=Overlay_170}
  Gradient {
   widgetClassPath=&Image
   bIsVariable=true
   parent=Overlay_170
   widget=Gradient
   properties {
    visibility=SelfHitTestInvisible
    brush {
     bIsDynamicallyLoaded=false
     drawAs=Image
     imageType=FullColor
     mirroring=NoMirror
     resourceName=None
     resourceObject=/QuestUI/QuestScreen/Materials/MI_UI_Quests_VerticalGradient
     tiling=NoTile
     imageSize=<32 128>
     margin=<ltrb 0 0 0 0>
     outlineSettings {
      bUseBrushTransparency=false
      roundingType=HalfHeightRadius
      width=0
      color {colorUseRule=UseColor_Specified, specifiedColor=<rgba 0 0 0 0>}
      cornerRadii=<0 0 0 0>
     }
     tintColor {colorUseRule=UseColor_Specified, specifiedColor=<rgba 1 1 1 1>}
     uVRegion {bIsValid=false, max=<0 0>, min=<0 0>}
    }
    colorAndOpacity=<rgba 1 1 1 0.5>
   }
   slot {class=OverlaySlot, properties {horizontalAlignment=HAlign_Fill, verticalAlignment=VAlign_Bottom}}
  }
  BottomBarWidget {
   widgetClassPath=/Game/UI/Frontend/BottomBar/WBP_AthenaBottomBarWidget
   bIsVariable=true
   parent=Overlay_170
   widget=BottomBarWidget
   slot {class=OverlaySlot, properties {horizontalAlignment=HAlign_Right, verticalAlignment=VAlign_Bottom}}
  }
  WBP_CaptureForPostBufferUpdate_36 {
   widgetClassPath=&WBP_CaptureForPostBufferUpdate
   bIsVariable=true
   namedSlotHost=BottomBarWidget
   widget=WBP_CaptureForPostBufferUpdate_36
   slot=~
  }
  AnimSwitcher_Inputs {widgetClassPath=/Script/CommonUI.CommonAnimatedSwitcher, bIsVariable=true, namedSlotHost=BottomBarWidget, widget=AnimSwitcher_Inputs}
  WBox_Inputs {
   widgetClassPath=&WrapBox
   bIsVariable=true
   parent=AnimSwitcher_Inputs
   widget=WBox_Inputs
   properties {bExplicitWrapSize=true, horizontalAlignment=HAlign_Right, wrapSize=2000, innerSlotPadding=<24 0>}
   slot {class=WidgetSwitcherSlot}
  }
  TODO {
   widgetClassPath=&Overlay
   bIsVariable=true
   parent=WBox_Inputs
   widget=TODO
   slot {class=WrapBoxSlot, properties {horizontalAlignment=HAlign_Left, verticalAlignment=VAlign_Top}}
  }
  DecoyButtonDebugQuest {
   widgetClassPath=/Game/UI/Frontend/BottomBar/WBP_BottomBarDecoyButton
   bIsVariable=true
   parent=TODO
   widget=DecoyButtonDebugQuest
   properties {triggeringInputAction {dataTable=&QuestInputActionDataTable, rowName=DebugQuestUI}}
   slot {class=OverlaySlot}
  }
  DecoyButtonHideDisabled {
   widgetClassPath=&WBP_BottomBarDecoyButton
   bIsVariable=true
   parent=TODO
   widget=DecoyButtonHideDisabled
   properties {priority=1, triggeringInputAction {dataTable=&QuestInputActionDataTable, rowName=HideDisabled}}
   slot {class=OverlaySlot}
  }
  DecoyButtonTrackQuest {
   widgetClassPath=&WBP_BottomBarDecoyButton
   bIsVariable=true
   parent=TODO
   widget=DecoyButtonTrackQuest
   properties {triggeringInputAction {dataTable=&QuestInputActionDataTable, rowName=TrackQuest}}
   slot {class=OverlaySlot}
  }
  DecoyButtonGoToTop {
   widgetClassPath=&WBP_BottomBarDecoyButton
   bIsVariable=true
   parent=TODO
   widget=DecoyButtonGoToTop
   properties {triggeringInputAction {dataTable=&QuestInputActionDataTable, rowName=BackToTop}}
   slot {class=OverlaySlot}
  }
  DecoyButtonTrackAll {
   widgetClassPath=&WBP_BottomBarDecoyButton
   bIsVariable=true
   parent=TODO
   widget=DecoyButtonTrackAll
   properties {
    holdData=&CommonUI_HoldData_Base_C
    triggeringInputAction {dataTable=&QuestInputActionDataTable, rowName=TrackAll}
   }
   slot {class=OverlaySlot}
  }
  DecoyButtonMoveQuest {
   widgetClassPath=&WBP_BottomBarDecoyButton
   bIsVariable=true
   parent=TODO
   widget=DecoyButtonMoveQuest
   properties {triggeringInputAction {dataTable=&QuestInputActionDataTable, rowName=MoveQuest}}
   slot {class=OverlaySlot}
  }
  DecoyButtonExternalLink {
   widgetClassPath=&WBP_BottomBarDecoyButton
   bIsVariable=true
   parent=TODO
   widget=DecoyButtonExternalLink
   properties {priority=10, triggeringInputAction {dataTable=&QuestInputActionDataTable, rowName=ExternalButton}}
   slot {class=OverlaySlot}
  }
  WBox_MoveModeInputs {
   widgetClassPath=&WrapBox
   bIsVariable=true
   parent=AnimSwitcher_Inputs
   widget=WBox_MoveModeInputs
   properties {bExplicitWrapSize=true, horizontalAlignment=HAlign_Right, wrapSize=2000, innerSlotPadding=<24 0>}
   slot {class=WidgetSwitcherSlot}
  }
  Button_ConfirmMove {
   widgetClassPath=/FortUILibrary/UIKit/Buttons/WBP_UIKit_Button_Quiet
   bIsVariable=true
   parent=WBox_MoveModeInputs
   widget=Button_ConfirmMove
   properties {bIsFocusable=false, minWidth=120, mobileMinWidth=200, text=Confirm}
   slot {class=WrapBoxSlot, properties {horizontalAlignment=HAlign_Right}}
  }
  ActionWidget {
   widgetClassPath=&CommonActionWidget
   bIsVariable=true
   namedSlotHost=Button_ConfirmMove
   widget=ActionWidget
   properties {inputActions [{dataTable=&InputActionDataTable, rowName=Confirm}]}
  }
  Button_CancelMove {
   widgetClassPath=&WBP_UIKit_Button_Quiet
   bIsVariable=true
   parent=WBox_MoveModeInputs
   widget=Button_CancelMove
   properties {
    bHideInputActionWithKeyboard=true
    bIsFocusable=false
    minWidth=120
    mobileMinWidth=200
    text=Cancel
    visibility=Visible
    triggeringInputAction {dataTable=&InputActionDataTable, rowName=Cancel}
   }
   slot {class=WrapBoxSlot, properties {horizontalAlignment=HAlign_Right}}
  }
  Overlay_Dialog {
   widgetClassPath=&Overlay
   bIsVariable=true
   parent=Overlay_170
   widget=Overlay_Dialog
   slot {class=OverlaySlot, properties {horizontalAlignment=HAlign_Fill, verticalAlignment=VAlign_Fill}}
  }
  SwipePanel_SideNavigation {
   widgetClassPath=/Script/FortniteUI.FortSwipePanel
   bIsVariable=true
   namedSlotHost=WBP_UIKit_FrontendLayout
   widget=SwipePanel_SideNavigation
  }
  SideNavigation {
   widgetClassPath=&/WBP_QuestV3_SideNavigation
   bIsVariable=true
   parent=SwipePanel_SideNavigation
   widget=SideNavigation
   properties {bIsFocusable=true, navigation=&self:WidgetTree.SideNavigation.WidgetNavigation_0, padding=<ltrb 0 0 0 100>}
   slot {class=FortSwipePanelSlot}
  }
  SeasonPass_Level {
   widgetClassPath=/FortniteCoreUI/MultiProduct/SeasonPass_Level
   bIsVariable=true
   namedSlotHost=WBP_UIKit_FrontendLayout
   widget=SeasonPass_Level
  }
 ]
}
