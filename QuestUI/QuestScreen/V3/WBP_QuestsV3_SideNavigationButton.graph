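# Collapsible side navigation button for the V3 Quests screen, built on the UIKit generic button.
# Icons async-load with a default fallback, applying only if still current and a material instance.
# SetIsCollapsed plays Anim_Collapse forward or backward, resumes interrupted plays mid-way, or skips to the end; GetRemainingAnimTime reports timing.
# The animation fades and shifts Main_Stack and keys a Test value; SetTest maps such values to width.
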WidgetBlueprint /QuestUI/QuestScreen/V3/WBP_QuestsV3_SideNavigationButton
animations [Anim_Collapse {
  binding_count=2
  length_seconds=1
  object=&self:Anim_Collapse
  sequence {bindings [
    Main_Stack {
     id=#1
     possessed_class=StackBox
     tracks [
      {
       type=MovieSceneFloatTrack
       display_name=Render Opacity
       sections [{
         type=MovieSceneFloatSection
         end_frame=20
         start_frame=0
         channels [None {type=MovieSceneScriptingFloatChannel, keys [[0,1],[4,0.896],[8,0],[20,0]]}]
        }]
      }
      {
       type=MovieScene2DTransformTrack
       sections [{
         type=MovieScene2DTransformSection
         end_frame=20
         start_frame=0
         channels [
          Translation.X {type=MovieSceneScriptingFloatChannel, keys [[0,0],[4,-8]]}
          Translation.Y {type=MovieSceneScriptingFloatChannel, const [0,1]}
          Angle {type=MovieSceneScriptingFloatChannel, const [0,1]}
          Scale.X {type=MovieSceneScriptingFloatChannel}
          Scale.Y {type=MovieSceneScriptingFloatChannel}
          Shear.X {type=MovieSceneScriptingFloatChannel, const [0,1]}
          Shear.Y {type=MovieSceneScriptingFloatChannel, const [0,1]}
         ]
        }]
      }
     ]
    }
    SideNavButton_C_40 {
     id=#2
     possessed_class=WBP_Quests_SideNavigationButton_C
     tracks [{
       type=MovieSceneDoubleTrack
       display_name=Test
       sections [{type=MovieSceneDoubleSection, end_frame=20, start_frame=0, channels [None {type=MovieSceneScriptingDoubleChannel, keys [[0,0],[20,1]]}]}]
      }]
    }
   ]}
 }]
graphs [
 EventGraph {nodes [
   n1 = LoadIconTexture!() -> (OutputDelegate, then, Texture:Soft Object Reference) {is=title K2Node_CustomEvent}
   > n2 = Set CurrentIcon(CurrentIcon:Soft Object Reference=Select(
     Option 0:Soft Object Reference=/QuestUI/QuestScreen/Materials/MI_Quests_Icon_Tracked
     Option 1:Soft Object Reference=@n1.Texture
     Index=IsValidSoftObjectReference(SoftObjectReference=@n1.Texture)
    ) -> (ReturnValue:Soft Object Reference)) -> (Output_Get:Soft Object Reference)
   > n3 = Async Load Asset(Asset=@n2)
   Branch(execute=@n3.Completed, Condition="Equal (Object)"(A=@n3, B:Soft Object Reference=Get CurrentIcon() -> (CurrentIcon:Soft Object Reference)))
   > n4 = Cast To MaterialInstance!(execute, Object=@n3) -> (then, CastFailed, AsMaterial Instance:Material Instance Object Reference) {is=title K2Node_DynamicCast}
   > SetIconMaterial!(execute, self:WBP UIKit Button Generic Base Object Reference, Material Instance:Material Instance Object Reference=@n4) -> (then)
   n5 = Event PreConstruct()
   > Parent: PreConstruct(IsDesignTime=@n5.IsDesignTime)
   > SetWidthOverride(self=Get Main_StackSize() -> (Main_StackSize:Size Box Object Reference), InWidthOverride=304)
   Event Construct()
   > Parent: Construct()
   > n6 = Cast To WBP_UIKit_BlockBackground_Tab_Enhanced!(execute, Object=Get Block_Background() -> (Block_Background:WBP UIKit Block Background Base Object Reference)) -> (then, CastFailed, AsWBP UIKit Block Background Tab Enhanced:WBP UIKit Block Background Tab Enhanced Object Reference) {is=title K2Node_DynamicCast}
   > SetPadding!(
    execute
    self:Grid Slot Object Reference=SlotAsGridSlot(Widget=Get Backing!(self:WBP UIKit Block Background Tab Enhanced Object Reference=@n6) -> (Backing:Image Object Reference))
    InPadding:Margin Structure=Get BackgroundPadding() -> (BackgroundPadding:Margin Structure)
   ) -> (then)
   > SetCornerRadiuses!(
    execute
    self:WBP UIKit Block Background Tab Enhanced Object Reference=@n6
    CornerRadius:float=Get BackgroundCornerRadius() -> (BackgroundCornerRadius:float)
    RadiusInPixels:bool=true
   ) -> (then)
   > SetPadding(
    self=SlotAsGridSlot(Widget=Get FocusOutline!(self:WBP UIKit Block Background Tab Enhanced Object Reference=@n6) -> (FocusOutline:Image Object Reference))
    InPadding=Get FocusOutlinePadding() -> (FocusOutlinePadding:Margin Structure)
   )
   n7 = Event SetIsCollapsed!() -> (OutputDelegate, then, IsCollapsed:bool, SkipAnimation:bool) {is=title K2Node_Event}
   > n8 = Branch(Condition=@n7.SkipAnimation)
   n9 = Get Anim_Collapse() -> (Anim_Collapse:Widget Animation Object Reference)
   n10 = Select(Option 0:EUMGSequencePlayMode Enum=Reverse, Option 1:EUMGSequencePlayMode Enum, Index=@n7.IsCollapsed) -> (ReturnValue:EUMGSequencePlayMode Enum)
   > PlayAnimation(InAnimation=@n9, StartAtTime=GetEndTime(self=@n9), PlayMode=@n10)
   n11 = Branch(execute=@n8.else, Condition=IsAnimationPlaying(InAnimation=Get Anim_Collapse()))
   n12 = Get Anim_Collapse()
   > n13 = PauseAnimation!(execute, self:User Widget Object Reference, InAnimation:Widget Animation Object Reference=@n12) -> (then, ReturnValue:float)
   n14 = GetEndTime(self=@n12)
   n15 = Get Playback Speed() -> (Playback Speed:float)
   > PlayAnimation(
    InAnimation=@n12
    StartAtTime=SelectFloat(A=@n13, B=float - float(A=@n14, B=@n13), bPickA=@n7.IsCollapsed)
    PlayMode=@n10
    PlaybackSpeed=@n15
   )
   PlayAnimation(execute=@n11.else, InAnimation=Get Anim_Collapse(), PlayMode=@n10, PlaybackSpeed=@n15)
  ]}
 GetRemainingAnimTime {nodes [
   GetRemainingAnimTime() {is=title K2Node_FunctionEntry}
   n1 = Get Anim_Collapse()
   > n2 = Branch(Condition=IsAnimationPlaying(InAnimation=@n1))
   > n3 = IsAnimationPlayingForward!(execute, self:User Widget Object Reference, InAnimation:Widget Animation Object Reference=@n1) -> (then, ReturnValue:bool)
   > n4 = Branch(Condition=@n3)
   n5 = Get Anim_Collapse()
   n6 = GetEndTime(self=@n5)
   n7 = GetAnimationCurrentTime!(self:User Widget Object Reference, InAnimation:Widget Animation Object Reference=@n5) -> (ReturnValue:float)
   n8 = float / float(A=GetEndTime(self=Get Anim_Collapse()), B=Get Playback Speed())
   > Return Node!(
    execute
    FunctionIsImplemented:bool=true
    RemainingTime:double=float / float(A=float - float(A=@n6, B=@n7), B=Get Playback Speed())
    TotalTime:double=@n8
   )
   n9 = Get Anim_Collapse()
   n10 = GetAnimationCurrentTime(InAnimation=@n9)
   n11 = GetStartTime!(self:Widget Animation Object Reference=@n9) -> (ReturnValue:float)
   Return Node(
    execute=@n4.else
    FunctionIsImplemented=true
    RemainingTime=float / float(A=float - float(A=@n10, B=@n11), B=Get Playback Speed())
    TotalTime=@n8
   )
   Return Node(execute=@n2.else, FunctionIsImplemented=true, RemainingTime=0, TotalTime=0)
   n12 = Get Anim_Collapse()
   n13 = GetStartTime(self=@n12)
   n14 = GetEndTime(self=@n12)
   n15 = GetAnimationCurrentTime(InAnimation=@n12)
   PrintString!(
    execute
    InString=Append!(
     A="Start: "
     B="To String (Float)"(InDouble=@n13)
     C:String=" || Current: "
     D:String="To String (Float)"(InDouble=@n15)
     E:String=" || End: "
     F:String="To String (Float)"(InDouble=@n14)
    ) -> (ReturnValue)
    bPrintToScreen
    bPrintToLog
    TextColor
    Duration
    Key
   ) -> (then)
  ]}
 Sequencer Events {links [], nodes []}
 SetIconSoft {nodes [
   n1 = SetIconSoft!() -> (then, Icon:Soft Object Reference) {is=title K2Node_FunctionEntry}
   > Set QuestUITrackerVM(QuestUITrackerVM:Quest UI Tracker View Model Object Reference) -> (Output_Get:Quest UI Tracker View Model Object Reference)
   > LoadIconTexture!(execute, self, Texture:Soft Object Reference=@n1) -> (then) {is=title K2Node_CallFunction}
  ]}
 SetIconSoft_0 {nodes [
   n1 = SetIconSoft_0!() -> (then, Icon:Soft Object Reference) {is=title K2Node_FunctionEntry}
   > Set QuestUIBucketVM(QuestUIBucketVM:Quest UI Bucket View Model Object Reference) -> (Output_Get:Quest UI Bucket View Model Object Reference)
   > LoadIconTexture(Texture=@n1)
  ]}
 SetTest {nodes [
   n1 = SetTest!() -> (then, Value:double=~) {is=title K2Node_FunctionEntry}
   n2 = Get!() -> (CachedContentSize_X:double, CachedContentSize_Y:double)
   > SetWidthOverride(
    self=Get Main_StackSize()
    InWidthOverride=Ease!(Function:EEasingFunc Enum=Linear, Alpha:double=@n1, A:double=float - float(A=@n2, B=40), B:double) -> (Result:double=~) {is=title K2Node_EaseFunction}
   )
  ]}
]
variables [
 BackgroundCornerRadius {type=number, default=12}
 BackgroundPadding {type=Margin, default=<ltrb 0 0 0 0>}
 CachedContentSize {type=Vector2D, default=<344 48>}
 CurrentIcon {type=Object, default=null}
 FocusOutlinePadding {type=Margin, default=<ltrb -15 -15 -15 -15>}
 Playback Speed {type=number, default=3.0303}
 test {type=number}
]
widgets {
 info {
  inheritedWidgetCount=0
  parentClass=/FortUILibrary/Systems/ModularButton/WBP_UIKit_ButtonGeneric_Base.WBP_UIKit_ButtonGeneric_Base_C
  rootWidgetClass=null
  widgetCount=0
 }
 widgets []
}
